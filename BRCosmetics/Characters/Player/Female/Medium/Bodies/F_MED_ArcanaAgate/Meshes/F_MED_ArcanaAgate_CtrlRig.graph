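# Control Rig for the ArcanaAgate female medium body cosmetic that procedurally corrects skirt and scarf bones as the character poses.
# After the shared DeformRig function, Variables remaps thigh, calf and spine rotations and pelvis height into weights.
# Skirt blends thigh-pelvis constraints and weighted offsets, including smoothed crouch, slide and rail states plus axe-specific fixes.
# Scarf offsets follow neck spherical pose readers, left-arm raise, right-hand proximity and front-end state.
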ControlRigBlueprint /BRCosmetics/Characters/Player/Female/Medium/Bodies/F_MED_ArcanaAgate/Meshes/F_MED_ArcanaAgate_CtrlRig
graphs [
 Multiply_2_ContainedGraph {nodes [
   n1 = RigVMFunctionEntryNode()
   RigVMFunctionReturnNode(Result=&RigVMFunction_MathDoubleMul(B=@n1.C, A=&RigVMFunction_MathDoubleMul(A=@n1.A, B=@n1.B).Result).Result)
  ]}
 Multiply_6_ContainedGraph {nodes [
   n1 = RigVMFunctionEntryNode()
   RigVMFunctionReturnNode(Result=&RigVMFunction_MathDoubleMul(B=@n1.C, A=&RigVMFunction_MathDoubleMul(A=@n1.A, B=@n1.B).Result).Result)
  ]}
 Multiply_6_ContainedGraph {nodes [
   n1 = RigVMFunctionEntryNode()
   RigVMFunctionReturnNode(Result=&RigVMFunction_MathDoubleMul(B=@n1.C, A=&RigVMFunction_MathDoubleMul(A=@n1.A, B=@n1.B).Result).Result)
  ]}
 Multiply_6_ContainedGraph {nodes [
   n1 = RigVMFunctionEntryNode()
   RigVMFunctionReturnNode(Result=&RigVMFunction_MathDoubleMul(B=@n1.C, A=&RigVMFunction_MathDoubleMul(A=@n1.A, B=@n1.B).Result).Result)
  ]}
 New Function_ContainedGraph {nodes [
   RigVMFunctionEntryNode()
   > RigVMFunction_Sequence = RigVMAggregateNode()
   &RigUnit_ParentConstraint(
    AdvancedSettings=(RotationOrderForFilter=XZY)
    Child=(Type=Bone,Name="base_skirtSide_l")
    Filter=(TranslationFilter=(bX=false,bY=false,bZ=false),RotationFilter=(bX=True,bY=True,bZ=True),ScaleFilter=(bX=True,bY=True,bZ=True))
    Parents=((Item=(Type=Bone,Name="thigh_twist_01_l"),Weight=0.2),(Item=(Type=Bone,Name="pelvis"),Weight=0.8))
    Weight=1
    ExecutePin=@RigVMFunction_Sequence.A
   )
   > &RigUnit_ParentConstraint(
    AdvancedSettings=(RotationOrderForFilter=XZY)
    Child=(Type=Bone,Name="base_skirtFrontSide_l")
    Filter=(TranslationFilter=(bX=false,bY=false,bZ=false),RotationFilter=(bX=True,bY=True,bZ=True),ScaleFilter=(bX=True,bY=True,bZ=True))
    Parents=((Item=(Type=Bone,Name="thigh_twist_01_l"),Weight=0.25),(Item=(Type=Bone,Name="pelvis"),Weight=0.75))
    Weight=1
   )
   > &RigUnit_ModifyTransforms(
    ItemToModify="((Item=(Type=Bone,Name=\"base_skirtFrontSide_l\"),Transform=(Rotation=(X=0.041,Y=0.0856,Z=0.498,W=-0.862),Translation=(X=0,Y=0,Z=0),Scale3D=(X=1,Y=1,Z=1))),(Item=(Type=Bone,Name=\"base_skirtSide_l\"),Transform=(Rotation=(X=0,Y=0,Z=0.0564,W=-0.9984),Translation ...(truncated)"
    Weight=RigVMDispatchNode(BoneName=root, Default=1, Name=deformRig_l_thigh_up_alpha) {template="DISPATCH_RigDispatch_GetAnimAttribute(in Name,in BoneName,in Default,out Value,out Found)"}.Value
   )
   n1 = RigVMVariableNode() {variable=Thigh_L_SideRotation}
   > &RigUnit_ParentConstraint(
    AdvancedSettings=(RotationOrderForFilter=XZY)
    Child=(Type=Bone,Name="base_skirtSide_l")
    Filter=(TranslationFilter=(bX=false,bY=false,bZ=false),RotationFilter=(bX=True,bY=True,bZ=True),ScaleFilter=(bX=True,bY=True,bZ=True))
    Parents=((Item=(Type=Bone,Name="thigh_twist_01_l"),Weight=0.8),(Item=(Type=Bone,Name="pelvis"),Weight=0.2))
    Weight=@n1.Value
   )
   > &RigUnit_ParentConstraint(
    AdvancedSettings=(RotationOrderForFilter=XZY)
    Child=(Type=Bone,Name="base_skirtFrontSide_l")
    Filter=(TranslationFilter=(bX=false,bY=false,bZ=false),RotationFilter=(bX=True,bY=True,bZ=True),ScaleFilter=(bX=True,bY=True,bZ=True))
    Parents=((Item=(Type=Bone,Name="thigh_twist_01_l"),Weight=0.8),(Item=(Type=Bone,Name="pelvis"),Weight=0.2))
    Weight=@n1.Value
   )
   > &RigUnit_ModifyTransforms(
    ItemToModify="((Item=(Type=Bone,Name=\"base_skirtSide_l\"),Transform=(Rotation=(X=0.0028,Y=-0.0169,Z=0.2779,W=-0.9605),Translation=(X=0,Y=0,Z=0),Scale3D=(X=1,Y=1,Z=1))),(Item=(Type=Bone,Name=\"base_skirtFrontSide_l\"),Transform=(Rotation=(X=-0.0008,Y=-0.0906,Z=0.0053,W=0.99 ...(truncated)"
    Weight=@n1.Value
   )
   > &RigUnit_ModifyTransforms(
    ItemToModify="((Item=(Type=Bone,Name=\"base_skirtBack\"),Transform=(Rotation=(X=0,Y=0,Z=-0.2588,W=0.9659),Translation=(X=0,Y=0,Z=0),Scale3D=(X=1,Y=1,Z=1))),(Item=(Type=Bone,Name=\"base_skirtBackSide_l\"),Transform=(Rotation=(X=-0.0137,Y=0.3449,Z=-0.3628,W=0.8656),Translatio ...(truncated)"
    WeightMinimum=0
    Weight=RigVMVariableNode() {variable=Thigh_L_BackRotation}.Value
   )
   &RigUnit_ParentConstraint(
    AdvancedSettings=(RotationOrderForFilter=XZY)
    Child=(Type=Bone,Name="base_skirtSide_r")
    Filter=(TranslationFilter=(bX=false,bY=false,bZ=false),RotationFilter=(bX=True,bY=True,bZ=True),ScaleFilter=(bX=True,bY=True,bZ=True))
    Parents=((Item=(Type=Bone,Name="thigh_twist_01_r"),Weight=0.2),(Item=(Type=Bone,Name="pelvis"),Weight=0.8))
    Weight=1
    ExecutePin=@RigVMFunction_Sequence.B
   )
   > &RigUnit_ParentConstraint(
    AdvancedSettings=(RotationOrderForFilter=XZY)
    Child=(Type=Bone,Name="base_skirtFrontSide_r")
    Filter=(TranslationFilter=(bX=false,bY=false,bZ=false),RotationFilter=(bX=True,bY=True,bZ=True),ScaleFilter=(bX=True,bY=True,bZ=True))
    Parents=((Item=(Type=Bone,Name="thigh_twist_01_r"),Weight=0.25),(Item=(Type=Bone,Name="pelvis"),Weight=0.75))
    Weight=1
   )
   n2 = RigVMDispatchNode(BoneName=root, Default=1, Name=deformRig_r_thigh_up_alpha) {template="DISPATCH_RigDispatch_GetAnimAttribute(in Name,in BoneName,in Default,out Value,out Found)"}
   > &RigUnit_ModifyTransforms(
    ItemToModify="((Item=(Type=Bone,Name=\"base_skirtFrontSide_r\"),Transform=(Rotation=(X=0.1635,Y=-0.2003,Z=0.3621,W=-0.8956),Translation=(X=0,Y=0,Z=0),Scale3D=(X=1,Y=1,Z=1))),(Item=(Type=Bone,Name=\"base_skirtSide_r\"),Transform=(Rotation=(X=0.3016,Y=-0.0641,Z=0.0898,W=-0.94 ...(truncated)"
    Weight=@n2.Value
   )
   n3 = RigVMVariableNode() {variable=isCrouching_CR}
   Multiply_3 = RigVMAggregateNode(
    B=&RigVMFunction_MathBoolToFloat(Value=RigVMVariableNode() {variable=isArcanaAgateAxeEquipped}.Value).Result
    C=&RigVMFunction_MathBoolToFloat(Value=&RigVMFunction_MathBoolNot(Value=@n3.Value).Result).Result
    A=@n2.Value
   )
   > &RigUnit_ModifyTransforms(
    ItemToModify="((Item=(Type=Bone,Name=\"base_skirtFrontSide_r\"),Transform=(Rotation=(X=-0.0233,Y=-0.0271,Z=-0.1997,W=-0.9792),Translation=(X=0,Y=0,Z=0),Scale3D=(X=1,Y=1,Z=1))),(Item=(Type=Bone,Name=\"base_skirtSide_r\"),Transform=(Rotation=(X=0.0131,Y=0.0001,Z=0.0087,W=0.99 ...(truncated)"
    Weight=@Multiply_3.Result
   )
   n4 = RigVMVariableNode() {variable=Thigh_R_SideRotation}
   > &RigUnit_ParentConstraint(
    AdvancedSettings=(RotationOrderForFilter=XZY)
    Child=(Type=Bone,Name="base_skirtSide_r")
    Filter=(TranslationFilter=(bX=false,bY=false,bZ=false),RotationFilter=(bX=True,bY=True,bZ=True),ScaleFilter=(bX=True,bY=True,bZ=True))
    Parents=((Item=(Type=Bone,Name="thigh_twist_01_r"),Weight=0.8),(Item=(Type=Bone,Name="pelvis"),Weight=0.2))
    Weight=@n4.Value
   )
   > &RigUnit_ParentConstraint(
    AdvancedSettings=(RotationOrderForFilter=XZY)
    Child=(Type=Bone,Name="base_skirtFrontSide_r")
    Filter=(TranslationFilter=(bX=false,bY=false,bZ=false),RotationFilter=(bX=True,bY=True,bZ=True),ScaleFilter=(bX=True,bY=True,bZ=True))
    Parents=((Item=(Type=Bone,Name="thigh_twist_01_r"),Weight=0.8),(Item=(Type=Bone,Name="pelvis"),Weight=0.2))
    Weight=@n4.Value
   )
   > &RigUnit_ModifyTransforms(
    ItemToModify="((Item=(Type=Bone,Name=\"base_skirtSide_r\"),Transform=(Rotation=(X=-0.0143,Y=-0.0523,Z=-0.1847,W=0.9813),Translation=(X=0,Y=0,Z=0),Scale3D=(X=1,Y=1,Z=1))),(Item=(Type=Bone,Name=\"base_skirtFrontSide_r\"),Transform=(Rotation=(X=-0.007,Y=-0.0101,Z=0.0508,W=-0.9 ...(truncated)"
    Weight=@n4.Value
   )
   Cast = &RigVMFunction_MathBoolToFloat(Value=RigVMVariableNode() {variable=isArcanaAgateAxeEquipped}.Value)
   RigVMFunction_MathBoolNot = &RigVMFunction_MathBoolNot(Value=RigVMVariableNode() {variable=isCrouching_CR}.Value)
   Multiply_2 = RigVMAggregateNode(B=@Cast.Result, C=&RigVMFunction_MathBoolToFloat(Value=@RigVMFunction_MathBoolNot.Result).Result, A=@n4.Value)
   > &RigUnit_ModifyTransforms(
    ItemToModify="((Item=(Type=Bone,Name=\"base_skirtSide_r\"),Transform=(Rotation=(X=0,Y=0,Z=0.1736,W=0.9848),Translation=(X=0,Y=0,Z=0),Scale3D=(X=1,Y=1,Z=1))),(Item=(Type=Bone,Name=\"base_skirtFrontSide_r\"),Transform=(Rotation=(X=0.0703,Y=-0.2047,Z=-0.4289,W=-0.877),Translat ...(truncated)"
    Weight=@Multiply_2.Result
   )
   n5 = RigVMVariableNode() {variable=Thigh_R_BackRotation}
   > &RigUnit_ModifyTransforms(
    ItemToModify="((Item=(Type=Bone,Name=\"base_skirtBack\"),Transform=(Rotation=(X=0,Y=0,Z=-0.2588,W=0.9659),Translation=(X=0,Y=0,Z=0),Scale3D=(X=1,Y=1,Z=1))),(Item=(Type=Bone,Name=\"base_skirtBackSide_r\"),Transform=(Rotation=(X=-0.1167,Y=-0.3217,Z=-0.4149,W=0.843),Translatio ...(truncated)"
    WeightMinimum=0
    Weight=@n5.Value
   )
   n6 = RigVMVariableNode() {variable=isCrouching_CR}
   Multiply_5 = RigVMAggregateNode(
    B=&RigVMFunction_MathBoolToFloat(Value=RigVMVariableNode() {variable=isArcanaAgateAxeEquipped}.Value).Result
    C=&RigVMFunction_MathBoolToFloat(Value=&RigVMFunction_MathBoolNot(Value=@n6.Value).Result).Result
    A=@n5.Value
   )
   > &RigUnit_ModifyTransforms(
    ItemToModify="((Item=(Type=Bone,Name=\"base_skirtBack\"),Transform=(Rotation=(X=-0.169,Y=-0.085,Z=-0.0215,W=0.9817),Translation=(X=0,Y=0,Z=0),Scale3D=(X=1,Y=1,Z=1))),(Item=(Type=Bone,Name=\"base_skirtBackSide_r\"),Transform=(Rotation=(X=-0.0811,Y=-0.0192,Z=0.0365,W=0.9959), ...(truncated)"
    WeightMinimum=0
    Weight=@Multiply_5.Result
   )
   n7 = RigVMDispatchNode(BoneName=root, Default=1, Name=deformRig_l_thigh_up_alpha) {template="DISPATCH_RigDispatch_GetAnimAttribute(in Name,in BoneName,in Default,out Value,out Found)"}
   &RigUnit_ParentConstraint(
    AdvancedSettings=(RotationOrderForFilter=XZY)
    Child=(Type=Bone,Name="base_skirtFront")
    Filter=(TranslationFilter=(bX=false,bY=false,bZ=false),RotationFilter=(bX=True,bY=True,bZ=True),ScaleFilter=(bX=True,bY=True,bZ=True))
    Parents=((Item=(Type=Bone,Name="thigh_twist_01_l"),Weight=0.55),(Item=(Type=Bone,Name="pelvis"),Weight=0.45))
    Weight=@n7.Value
    ExecutePin=@RigVMFunction_Sequence.C
   )
   > &RigUnit_ModifyTransforms(
    ItemToModify="((Item=(Type=Bone,Name=\"None\"),Transform=(Rotation=(X=0,Y=0,Z=0,W=1),Translation=(X=0,Y=0,Z=0),Scale3D=(X=1,Y=1,Z=1))),(Item=(Type=Bone,Name=\"base_skirtFront\"),Transform=(Rotation=(X=-0.0971,Y=0.0483,Z=-0.3672,W=0.9238),Translation=(X=0,Y=0,Z=0),Scale3D=(X ...(truncated)"
    WeightMinimum=0
    Weight=@n7.Value
   )
   > &RigUnit_ModifyTransforms(
    ItemToModify="((Item=(Type=Bone,Name=\"None\"),Transform=(Rotation=(X=0,Y=0,Z=0,W=1),Translation=(X=0,Y=0,Z=0),Scale3D=(X=1,Y=1,Z=1))),(Item=(Type=Bone,Name=\"base_skirtFront\"),Transform=(Rotation=(X=0,Y=0,Z=0,W=1),Translation=(X=0,Y=0,Z=0),Scale3D=(X=1,Y=1,Z=1))),(Item=(T ...(truncated)"
    WeightMinimum=0
    Weight=RigVMVariableNode() {variable=Thigh_L_SideRotation}.Value
   )
   n8 = RigVMDispatchNode(BoneName=root, Default=1, Name=deformRig_r_thigh_up_alpha) {template="DISPATCH_RigDispatch_GetAnimAttribute(in Name,in BoneName,in Default,out Value,out Found)"}
   &RigUnit_ParentConstraint(
    AdvancedSettings=(RotationOrderForFilter=XZY)
    Child=(Type=Bone,Name="base_skirtFront")
    Filter=(TranslationFilter=(bX=false,bY=false,bZ=false),RotationFilter=(bX=True,bY=True,bZ=True),ScaleFilter=(bX=True,bY=True,bZ=True))
    Parents=((Item=(Type=Bone,Name="thigh_twist_01_r"),Weight=0.55),(Item=(Type=Bone,Name="pelvis"),Weight=0.45))
    Weight=@n8.Value
    ExecutePin=@RigVMFunction_Sequence.D
   )
   > &RigUnit_ModifyTransforms(
    ItemToModify="((Item=(Type=Bone,Name=\"None\"),Transform=(Rotation=(X=0,Y=0,Z=0,W=1),Translation=(X=0,Y=0,Z=0),Scale3D=(X=1,Y=1,Z=1))),(Item=(Type=Bone,Name=\"base_skirtFront\"),Transform=(Rotation=(X=0.0078,Y=0.0397,Z=0.2367,W=-0.9707),Translation=(X=0,Y=0,Z=0),Scale3D=(X= ...(truncated)"
    WeightMinimum=0
    Weight=@n8.Value
   )
   > &RigUnit_ModifyTransforms(
    ItemToModify="((Item=(Type=Bone,Name=\"None\"),Transform=(Rotation=(X=0,Y=0,Z=0,W=1),Translation=(X=0,Y=0,Z=0),Scale3D=(X=1,Y=1,Z=1))),(Item=(Type=Bone,Name=\"base_skirtFront\"),Transform=(Rotation=(X=0,Y=0,Z=0,W=1),Translation=(X=0,Y=0,Z=0),Scale3D=(X=1,Y=1,Z=1))),(Item=(T ...(truncated)"
    WeightMinimum=0
    Weight=RigVMVariableNode() {variable=Thigh_R_SideRotation}.Value
   )
   Delta Rotation Z_1 = RigVMFunctionReferenceNode(Child=thigh_l, Parent=thigh_l_null, Source Maximum=-150, Source Minimum=-80, ExecuteContext=@RigVMFunction_Sequence.F)
   > &RigUnit_ModifyTransforms(
    ItemToModify="((Item=(Type=Bone,Name=\"base_skirtFrontSide_l\"),Transform=(Rotation=(X=-0.0179,Y=-0.0701,Z=-0.2468,W=0.9664),Translation=(X=0,Y=0,Z=0),Scale3D=(X=1,Y=1,Z=1))),(Item=(Type=Bone,Name=\"base_skirtFront\"),Transform=(Rotation=(X=0,Y=0,Z=-0.1305,W=0.9914),Transla ...(truncated)"
    WeightMinimum=0
    Weight=@"Delta Rotation Z_1".Delta Rotation
   )
   > Delta Rotation Z_2 = RigVMFunctionReferenceNode(Child=thigh_l, Parent=thigh_l_null, Source Maximum=-150, Source Minimum=-100)
   > &RigUnit_ModifyTransforms(
    ItemToModify="((Item=(Type=Bone,Name=\"base_skirtFrontSide_l\"),Transform=(Rotation=(X=0.0229,Y=0.101,Z=-0.2196,W=0.9701),Translation=(X=-0.5055,Y=0.2413,Z=-2.6599),Scale3D=(X=1,Y=1,Z=1))),(Item=(Type=Bone,Name=\"base_skirtSide_l\"),Transform=(Rotation=(X=-0.0037,Y=0.0257,Z ...(truncated)"
    WeightMinimum=0
    Weight=@"Delta Rotation Z_2".Delta Rotation
   )
   > Delta Rotation Z_3 = RigVMFunctionReferenceNode(Child=thigh_r, Parent=thigh_r_null, Source Maximum=-100, Source Minimum=-70)
   > &RigUnit_ModifyTransforms(
    ItemToModify=((Item=(Type=Bone,Name="base_skirtFrontSide_r"),Transform=(Rotation=(X=0,Y=0,Z=-0.2588,W=0.9659),Translation=(X=-0.5055,Y=0.2413,Z=-2.6599),Scale3D=(X=1,Y=1,Z=1))))
    WeightMinimum=0
    Weight=@"Delta Rotation Z_3".Delta Rotation
   )
   n9 = &RigVMFunction_AlphaInterp(
    InterpSpeedDecreasing=10
    InterpSpeedIncreasing=10
    OutRange=(Min=0,Max=1.4)
    Scale=1
    bClampResult=true
    Value=&RigVMFunction_MathBoolToFloat(Value=RigVMVariableNode() {variable=isCrouching_CR}.Value).Result
   )
   Multiply = &RigVMFunction_MathDoubleMul(B=@n9.Result, A=RigVMVariableNode() {variable=crouchWalking}.Value)
   &RigUnit_ModifyTransforms(
    ItemToModify="((Item=(Type=Bone,Name=\"base_skirtFrontSide_l\"),Transform=(Rotation=(X=0.0539,Y=-0.0646,Z=-0.0486,W=0.9953),Translation=(X=0,Y=0,Z=0),Scale3D=(X=1,Y=1,Z=1))),(Item=(Type=Bone,Name=\"base_skirtSide_l\"),Transform=(Rotation=(X=0,Y=0,Z=0,W=1),Translation=(X=0,Y ...(truncated)"
    WeightMinimum=0
    Weight=@Multiply.Result
    ExecutePin=@RigVMFunction_Sequence.G
   )
   Multiply_1 = RigVMAggregateNode(B=@n9.Result, C=RigVMVariableNode() {variable=onehandedPickaxe_CR}.Value, A=RigVMVariableNode() {variable=spineTwist_R}.Value)
   > &RigUnit_ModifyTransforms(
    ItemToModify="((Item=(Type=Bone,Name=\"base_skirtFrontSide_l\"),Transform=(Rotation=(X=-0.0075,Y=-0.156,Z=-0.0693,W=0.9853),Translation=(X=0,Y=0,Z=0),Scale3D=(X=1,Y=1,Z=1))),(Item=(Type=Bone,Name=\"base_skirtSide_l\"),Transform=(Rotation=(X=-0.0013,Y=-0.0314,Z=-0.0406,W=0.9 ...(truncated)"
    WeightMinimum=0
    Weight=@Multiply_1.Result
   )
   RigVMFunctionReturnNode(ExecuteContext=@RigVMFunction_Sequence.H)
   &RigUnit_ModifyTransforms(
    ItemToModify="((Item=(Type=Bone,Name=\"None\"),Transform=(Rotation=(X=0,Y=0,Z=0,W=1),Translation=(X=0,Y=0,Z=0),Scale3D=(X=1,Y=1,Z=1))),(Item=(Type=Bone,Name=\"deform_internalSkirt_l\"),Transform=(Rotation=(X=0,Y=0,Z=0,W=1),Translation=(X=-9.2653,Y=-0.9626,Z=7.3541),Scale3D= ...(truncated)"
    WeightMinimum=0
    Weight=RigVMDispatchNode(BoneName=root, Default=1, Name=deformRig_l_thigh_up_alpha) {template="DISPATCH_RigDispatch_GetAnimAttribute(in Name,in BoneName,in Default,out Value,out Found)"}.Value
    ExecutePin=@RigVMFunction_Sequence.I
   )
   > &RigUnit_ModifyTransforms(
    ItemToModify="((Item=(Type=Bone,Name=\"None\"),Transform=(Rotation=(X=0,Y=0,Z=0,W=1),Translation=(X=0,Y=0,Z=0),Scale3D=(X=1,Y=1,Z=1))),(Item=(Type=Bone,Name=\"deform_internalSkirt_r\"),Transform=(Rotation=(X=0,Y=0,Z=0,W=1),Translation=(X=-6.2662,Y=-8.5128,Z=-14.4666),Scale3 ...(truncated)"
    WeightMinimum=0
    Weight=RigVMDispatchNode(BoneName=root, Default=1, Name=deformRig_r_thigh_up_alpha) {template="DISPATCH_RigDispatch_GetAnimAttribute(in Name,in BoneName,in Default,out Value,out Found)"}.Value
   )
   &RigUnit_ModifyTransforms(
    ItemToModify="((Item=(Type=Bone,Name=\"base_skirtFrontSide_l\"),Transform=(Rotation=(X=-0.0012,Y=-0.0173,Z=-0.044,W=-0.9989),Translation=(X=0,Y=0,Z=0),Scale3D=(X=1,Y=1,Z=1))),(Item=(Type=Bone,Name=\"base_skirtSide_l\"),Transform=(Rotation=(X=-0.0378,Y=0.1896,Z=0.0595,W=-0.9 ...(truncated)"
    WeightMinimum=0
    Weight=@n9.Result
    ExecutePin=@RigVMFunction_Sequence.E
   )
   n10 = &RigVMFunction_AlphaInterp(
    InterpSpeedDecreasing=10
    InterpSpeedIncreasing=10
    OutRange=(Min=0,Max=1.4)
    Scale=1
    bClampResult=true
    Value=&RigVMFunction_MathBoolToFloat(Value=RigVMVariableNode() {variable=isBaseballSliding_CR}.Value).Result
   )
   > &RigUnit_ModifyTransforms(
    ItemToModify="((Item=(Type=Bone,Name=\"base_skirtFrontSide_l\"),Transform=(Rotation=(X=0.0112,Y=-0.069,Z=-0.0631,W=-0.9956),Translation=(X=0,Y=0,Z=0),Scale3D=(X=1,Y=1,Z=1))),(Item=(Type=Bone,Name=\"base_skirtSide_l\"),Transform=(Rotation=(X=-0.0213,Y=-0.0542,Z=-0.1178,W=-0. ...(truncated)"
    WeightMinimum=0
    Weight=@n10.Result
   )
   > &RigUnit_ModifyTransforms(
    ItemToModify="((Item=(Type=Bone,Name=\"base_skirtFront\"),Transform=(Rotation=(X=-0.3976,Y=0.1292,Z=0.1298,W=-0.8991),Translation=(X=0,Y=0,Z=0),Scale3D=(X=1,Y=1,Z=1))),(Item=(Type=Bone,Name=\"base_skirtFrontSide_r\"),Transform=(Rotation=(X=0.1052,Y=-0.0707,Z=-0.317,W=0.9399 ...(truncated)"
    WeightMinimum=0
    Weight=&RigVMFunction_MathDoubleMul(B=@n10.Result, A=RigVMVariableNode() {variable=spineTwist_R}.Value).Result
   )
   n11 = &RigVMFunction_AlphaInterp(
    InterpSpeedDecreasing=10
    InterpSpeedIncreasing=10
    OutRange=(Min=0,Max=1.4)
    Scale=1
    bClampResult=true
    Value=&RigVMFunction_MathBoolToFloat(Value=RigVMVariableNode() {variable=isKneeSliding_CR}.Value).Result
   )
   > &RigUnit_ModifyTransforms(
    ItemToModify="((Item=(Type=Bone,Name=\"base_skirtFrontSide_l\"),Transform=(Rotation=(X=-0.0177,Y=0.1027,Z=-0.0272,W=-0.9942),Translation=(X=0,Y=0,Z=0),Scale3D=(X=1,Y=1,Z=1))),(Item=(Type=Bone,Name=\"base_skirtSide_l\"),Transform=(Rotation=(X=-0.0162,Y=-0.1492,Z=0.187,W=-0.9 ...(truncated)"
    WeightMinimum=0
    Weight=@n11.Result
   )
   n12 = &RigVMFunction_AlphaInterp(
    InterpSpeedDecreasing=10
    InterpSpeedIncreasing=10
    OutRange=(Min=0,Max=1.4)
    Scale=1
    bClampResult=true
    Value=&RigVMFunction_MathBoolToFloat(Value=RigVMVariableNode() {variable=isOnRail_CR}.Value).Result
   )
   > &RigUnit_ModifyTransforms(
    ItemToModify="((Item=(Type=Bone,Name=\"base_skirtFrontSide_l\"),Transform=(Rotation=(X=0,Y=0,Z=0.0411,W=0.9992),Translation=(X=0,Y=0,Z=0),Scale3D=(X=1,Y=1,Z=1))),(Item=(Type=Bone,Name=\"base_skirtSide_l\"),Transform=(Rotation=(X=0.0016,Y=-0.0259,Z=-0.0599,W=0.9979),Translat ...(truncated)"
    WeightMinimum=0
    Weight=@n12.Result
   )
  ]}
 New Function_ContainedGraph {nodes [
   n1 = RigVMFunctionEntryNode()
   RigVMFunction_ControlFlowBranch = &RigVMFunction_ControlFlowBranch(Condition=@n1.Debug Controls, ExecuteContext=@n1.ExecuteContext)
   RigVMFunctionReturnNode(ExecuteContext=@RigVMFunction_ControlFlowBranch.Completed)
   RigUnit_GetTransform = &RigUnit_GetTransform(Item=(Type=Control,Name="thigh_r_ctrl"), bInitial=False)
   &RigUnit_SetTransform(
    Item=(Type=Bone,Name="thigh_r")
    Weight=1
    bInitial=False
    Value=@RigUnit_GetTransform.Transform
    ExecutePin=@RigVMFunction_ControlFlowBranch.True
   )
   > &RigUnit_SetTransform(
    Item=(Type=Bone,Name="thigh_l")
    Weight=1
    bInitial=False
    Value=&RigUnit_GetTransform(Item=(Type=Control,Name="thigh_l_ctrl"), bInitial=False).Transform
   )
   > &RigUnit_SetTransform(
    Item=(Type=Bone,Name="calf_l")
    Weight=1
    bInitial=False
    Value=&RigUnit_GetTransform(Item=(Type=Control,Name="calf_l_ctrl"), bInitial=False).Transform
   )
   > &RigUnit_SetTransform(
    Item=(Type=Bone,Name="calf_r")
    Weight=1
    bInitial=False
    Value=&RigUnit_GetTransform(Item=(Type=Control,Name="calf_r_ctrl"), bInitial=False).Transform
   )
   &RigUnit_SetTransform(
    ExecutePin=()
    Item=(Type=Bone,Name="upperarm_l")
    Weight=1
    bInitial=False
    Value=&RigUnit_GetTransform(Item=(Type=Control,Name="upperarm_l_ctrl"), bInitial=False).Transform
   )
   > &RigUnit_SetTransform(
    Item=(Type=Bone,Name="upperarm_r")
    Weight=1
    bInitial=False
    Value=&RigUnit_GetTransform(Item=(Type=Control,Name="upperarm_r_ctrl"), bInitial=False).Transform
   )
   &RigUnit_SetTransform(
    ExecutePin=()
    Item=(Type=Bone,Name="spine_05")
    Weight=1
    bInitial=False
    Value=&RigUnit_GetTransform(Item=(Type=Control,Name="spine_05_ctrl"), bInitial=False).Transform
   )
  ]}
 New Function_ContainedGraph {nodes [
   RigVMFunctionEntryNode()
   > RigVMFunction_Sequence = RigVMAggregateNode()
   Delta Rotation Y_2 = RigVMFunctionReferenceNode(Child=thigh_l, Parent=thigh_l_null, Source Maximum=-75, Source Minimum=0, ExecuteContext=@RigVMFunction_Sequence.A)
   > RigVMVariableNode(Value=@"Delta Rotation Y_2".Delta Rotation) {variable=Thigh_L_SideRotation}
   > Delta Rotation Y = RigVMFunctionReferenceNode(Child=thigh_r, Parent=thigh_r_null, Source Maximum=-75, Source Minimum=0)
   > RigVMVariableNode(Value=@"Delta Rotation Y".Delta Rotation) {variable=Thigh_R_SideRotation}
   > Delta Rotation Z_2 = RigVMFunctionReferenceNode(Child=thigh_l, Parent=thigh_l_null, Source Maximum=70, Source Minimum=5)
   > RigVMVariableNode(Value=@"Delta Rotation Z_2".Delta Rotation) {variable=Thigh_L_BackRotation}
   > Delta Rotation Z = RigVMFunctionReferenceNode(Child=thigh_r, Parent=thigh_r_null, Source Maximum=70, Source Minimum=5)
   > RigVMVariableNode(Value=@"Delta Rotation Z".Delta Rotation) {variable=Thigh_R_BackRotation}
   > Delta Rotation Z_1 = RigVMFunctionReferenceNode(Child=calf_l, Parent=calf_l_null, Source Maximum=90, Source Minimum=30)
   > RigVMVariableNode(Value=@"Delta Rotation Z_1".Delta Rotation) {variable=Calf_L_BackRotation}
   > Delta Rotation Z_3 = RigVMFunctionReferenceNode(Child=calf_r, Parent=calf_r_null, Source Maximum=90, Source Minimum=30)
   > RigVMVariableNode(Value=@"Delta Rotation Z_3".Delta Rotation) {variable=Calf_R_BackRotation}
   > Delta Rotation X = RigVMFunctionReferenceNode(Child=spine_05, Parent=spine_05_null, Source Maximum=-35, Source Minimum=-15)
   > RigVMVariableNode(Value=@"Delta Rotation X".Delta Rotation) {variable=spineTwist_R}
   n1 = &RigVMFunction_AnimEvalRichCurve(
    Curve=(EditorCurveData=(Keys=((),(Time=0.5,Value=1),(Time=0.8))))
    SourceMaximum=2
    SourceMinimum=0
    TargetMinimum=0
    Value=&RigVMFunction_MathDoubleRemap(
     SourceMaximum=75
     SourceMinimum=45
     TargetMaximum=1
     TargetMinimum=0
     bClamp=false
     Value=&RigUnit_GetRelativeTransformForItem(Child=(Type=Bone,Name="pelvis"), Parent=(Type=Bone,Name="root"), bChildInitial=False, bParentInitial=False).RelativeTransform.Translation.Z
    ).Result
   )
   > RigVMVariableNode(Value=@n1.Result) {variable=crouchWalking}
   RigVMFunctionReturnNode(ExecuteContext=@RigVMFunction_Sequence.C)
  ]}
 New Function_ContainedGraph {nodes [
   n1 = RigVMFunctionEntryNode()
   RigVMFunction_MathQuaternionToEuler = &RigVMFunction_MathQuaternionToEuler(
    RotationOrder=XYZ
    Value=&RigUnit_GetRelativeTransformForItem(
     Child=(Type=Bone,Name="hand_l")
     Parent=(Type=Null,Name="lowerarm_l")
     bChildInitial=False
     bParentInitial=False
     Child.Name=@n1.Child
     Parent.Name=@n1.Parent
    ).RelativeTransform.Rotation
   )
   Remap = &RigVMFunction_MathDoubleRemap(
    TargetMaximum=1
    TargetMinimum=0
    Value=@RigVMFunction_MathQuaternionToEuler.Result.Y
    SourceMaximum=@n1.Source Maximum
    SourceMinimum=@n1.Source Minimum
   )
   > RigVMFunctionReturnNode(Delta Rotation=@Remap.Result)
  ]}
 New Function_ContainedGraph {nodes [
   n1 = RigVMFunctionEntryNode()
   RigVMFunction_MathQuaternionToEuler = &RigVMFunction_MathQuaternionToEuler(
    RotationOrder=XYZ
    Value=&RigUnit_GetRelativeTransformForItem(
     Child=(Type=Bone,Name="hand_l")
     Parent=(Type=Null,Name="lowerarm_l")
     bChildInitial=False
     bParentInitial=False
     Child.Name=@n1.Child
     Parent.Name=@n1.Parent
    ).RelativeTransform.Rotation
   )
   Remap = &RigVMFunction_MathDoubleRemap(
    TargetMaximum=1
    TargetMinimum=0
    Value=@RigVMFunction_MathQuaternionToEuler.Result.Z
    SourceMaximum=@n1.Source Maximum
    SourceMinimum=@n1.Source Minimum
   )
   > RigVMFunctionReturnNode(Delta Rotation=@Remap.Result)
  ]}
 New Function_ContainedGraph {nodes [
   n1 = RigVMFunctionEntryNode()
   RigVMFunction_MathQuaternionToEuler = &RigVMFunction_MathQuaternionToEuler(
    RotationOrder=XYZ
    Value=&RigUnit_GetRelativeTransformForItem(
     Child=(Type=Bone,Name="hand_l")
     Parent=(Type=Null,Name="lowerarm_l")
     bChildInitial=False
     bParentInitial=False
     Child.Name=@n1.Child
     Parent.Name=@n1.Parent
    ).RelativeTransform.Rotation
   )
   Remap = &RigVMFunction_MathDoubleRemap(
    TargetMaximum=1
    TargetMinimum=0
    Value=@RigVMFunction_MathQuaternionToEuler.Result.X
    SourceMaximum=@n1.Source Maximum
    SourceMinimum=@n1.Source Minimum
   )
   > RigVMFunctionReturnNode(Delta Rotation=@Remap.Result)
  ]}
 New Function_ContainedGraph {nodes [
   RigVMFunctionEntryNode()
   > RigVMFunction_Sequence = RigVMAggregateNode()
   n1 = &RigUnit_SphericalPoseReader(
    ActiveRegionScaleFactors=(PositiveWidth=1,NegativeWidth=1,PositiveHeight=1,NegativeHeight=1)
    ActiveRegionSize=0.15
    Debug=(bDrawDebug=false,bDraw2D=false,bDrawLocalAxes=False,DebugScale=25,DebugSegments=20,DebugThickness=0.25)
    DriverAxis=(X=1,Y=0,Z=0)
    DriverItem=(Type=Bone,Name="neck_01")
    FalloffRegionScaleFactors=(PositiveWidth=1,NegativeWidth=1,PositiveHeight=1,NegativeHeight=1)
    FalloffSize=0.23
    FlipHeightScaling=False
    FlipWidthScaling=False
    OptionalParentItem=(Type=Bone,Name="spine_04")
    RotationOffset=(X=91.1538,Y=0.8291,Z=-61.7963)
    ExecutePin=@RigVMFunction_Sequence.A
   )
   n2 = &RigVMFunction_AlphaInterp(
    InterpSpeedDecreasing=10
    InterpSpeedIncreasing=10
    OutRange=(Min=0,Max=1.4)
    Scale=1
    bClampResult=true
    Value=&RigVMFunction_MathBoolToFloat(Value=RigVMVariableNode() {variable=isFrontEnd_CR}.Value).Result
   )
   Multiply = &RigVMFunction_MathDoubleMul(
    B=&RigVMFunction_MathDoubleRemap(SourceMaximum=1, SourceMinimum=0, TargetMaximum=0, TargetMinimum=1, Value=@n2.Result).Result
    A=@n1.OutputParam
   )
   > &RigUnit_ModifyTransforms(
    ItemToModify="((Item=(Type=Bone,Name=\"dyn_scarfTop_01\"),Transform=(Rotation=(X=0.009,Y=0.0804,Z=-0.1523,W=0.985),Translation=(X=0,Y=0,Z=0),Scale3D=(X=1,Y=1,Z=1))),(Item=(Type=Bone,Name=\"dyn_scarfBottom_01\"),Transform=(Rotation=(X=0,Y=0,Z=-0.1676,W=0.9859),Translation=(X ...(truncated)"
    WeightMinimum=0
    Weight=@Multiply.Result
   )
   > n3 = &RigUnit_SphericalPoseReader(
    ActiveRegionScaleFactors=(PositiveWidth=1,NegativeWidth=1,PositiveHeight=1,NegativeHeight=1)
    ActiveRegionSize=0.15
    Debug=(bDrawDebug=false,bDraw2D=false,bDrawLocalAxes=False,DebugScale=25,DebugSegments=20,DebugThickness=0.25)
    DriverAxis=(X=1,Y=0,Z=0)
    DriverItem=(Type=Bone,Name="neck_01")
    FalloffRegionScaleFactors=(PositiveWidth=1,NegativeWidth=1,PositiveHeight=1,NegativeHeight=1)
    FalloffSize=0.23
    FlipHeightScaling=False
    FlipWidthScaling=False
    OptionalParentItem=(Type=Bone,Name="spine_04")
    RotationOffset=(X=67.9484,Y=0.6208,Z=-134.1748)
   )
   > &RigUnit_ModifyTransforms(
    ItemToModify="((Item=(Type=Bone,Name=\"dyn_scarfTop_01\"),Transform=(Rotation=(X=0.0353,Y=0.0397,Z=-0.0855,W=0.9949),Translation=(X=0,Y=0,Z=0),Scale3D=(X=1,Y=1,Z=1))),(Item=(Type=Bone,Name=\"dyn_scarfBottom_01\"),Transform=(Rotation=(X=0,Y=0,Z=-0.0623,W=0.9981),Translation= ...(truncated)"
    WeightMinimum=0
    Weight=@n3.OutputParam
   )
   > &RigUnit_ModifyTransforms(
    ItemToModify="((Item=(Type=Bone,Name=\"dyn_scarfTop_01\"),Transform=(Rotation=(X=0.1232,Y=0.019,Z=0.0359,W=0.9916),Translation=(X=-0.0364,Y=-1.0338,Z=-1.0664),Scale3D=(X=1,Y=1,Z=1))),(Item=(Type=Bone,Name=\"dyn_scarfBottom_01\"),Transform=(Rotation=(X=0,Y=0,Z=0,W=1),Transla ...(truncated)"
    WeightMinimum=0
    Weight=RigVMDispatchNode(BoneName=root, Default=1, Name=deformRig_l_upperarm_up_alpha) {template="DISPATCH_RigDispatch_GetAnimAttribute(in Name,in BoneName,in Default,out Value,out Found)"}.Value
   )
   > &RigUnit_ModifyTransforms(
    ItemToModify="((Item=(Type=Bone,Name=\"dyn_scarfTop_01\"),Transform=(Rotation=(X=-0.0815,Y=-0.0763,Z=0.0667,W=-0.9915),Translation=(X=0,Y=0,Z=0),Scale3D=(X=1,Y=1,Z=1))),(Item=(Type=Bone,Name=\"dyn_scarfBottom_01\"),Transform=(Rotation=(X=0,Y=0,Z=0.0515,W=-0.9987),Translatio ...(truncated)"
    WeightMinimum=0
    Weight=@n2.Result
   )
   > Get Normalized Distance Between_1 = RigVMFunctionReferenceNode(Debug=false, Max Distance=30, Min Distance=11, Source=(Type=Bone,Name="hand_r"), Target=(Type=Bone,Name="base_scarfTop"))
   > &RigUnit_ModifyTransforms(
    ItemToModify="((Item=(Type=Bone,Name=\"dyn_scarfTop_01\"),Transform=(Rotation=(X=0,Y=0.1351,Z=0,W=-0.9908),Translation=(X=0,Y=0,Z=0),Scale3D=(X=1,Y=1,Z=1))),(Item=(Type=Bone,Name=\"dyn_scarfBottom_02\"),Transform=(Rotation=(X=-0.0329,Y=0.0004,Z=-0.0874,W=-0.9956),Translatio ...(truncated)"
    WeightMinimum=0
    Weight=@"Get Normalized Distance Between_1".Result
   )
   RigVMFunctionReturnNode(ExecuteContext=@RigVMFunction_Sequence.D)
  ]}
 RigVMFunctionLibrary {nodes [
   Control Setup = RigVMCollapseNode(Debug Controls=False, ExecuteContext=())
   Delta Rotation X = RigVMCollapseNode(ExecuteContext=())
   Delta Rotation Y = RigVMCollapseNode(ExecuteContext=())
   Delta Rotation Z = RigVMCollapseNode(ExecuteContext=())
   Scarf = RigVMCollapseNode(ExecuteContext=())
   Skirt = RigVMCollapseNode(ExecuteContext=())
   Variables = RigVMCollapseNode(ExecuteContext=())
  ]}
 RigVMFunction_Sequence_1_ContainedGraph {nodes [
   n1 = RigVMFunctionEntryNode()
   RigVMFunction_Sequence = &RigVMFunction_Sequence(ExecuteContext=@n1.ExecuteContext)
   n2 = &RigVMFunction_Sequence(ExecuteContext=@RigVMFunction_Sequence.B)
   n3 = &RigVMFunction_Sequence(ExecuteContext=@n2.B)
   n4 = &RigVMFunction_Sequence(ExecuteContext=@n3.B)
   n5 = &RigVMFunction_Sequence(ExecuteContext=@n4.B)
   n6 = &RigVMFunction_Sequence(ExecuteContext=@n5.B)
   n7 = &RigVMFunction_Sequence(ExecuteContext=@n6.B)
   n8 = &RigVMFunction_Sequence(ExecuteContext=@n7.B)
   RigVMFunctionReturnNode(A=@RigVMFunction_Sequence.A, B=@n2.A, C=@n3.A, D=@n4.A, E=@n5.A, F=@n6.A, G=@n7.A, H=@n8.A, I=@n8.B)
  ]}
 RigVMFunction_Sequence_1_ContainedGraph {nodes [
   n1 = RigVMFunctionEntryNode()
   RigVMFunction_Sequence = &RigVMFunction_Sequence(ExecuteContext=@n1.ExecuteContext)
   n2 = &RigVMFunction_Sequence(ExecuteContext=@RigVMFunction_Sequence.B)
   RigVMFunctionReturnNode(A=@RigVMFunction_Sequence.A, B=@n2.A, C=@n2.B)
  ]}
 RigVMFunction_Sequence_1_ContainedGraph {nodes [
   n1 = RigVMFunctionEntryNode()
   RigVMFunction_Sequence = &RigVMFunction_Sequence(ExecuteContext=@n1.ExecuteContext)
   n2 = &RigVMFunction_Sequence(ExecuteContext=@RigVMFunction_Sequence.B)
   n3 = &RigVMFunction_Sequence(ExecuteContext=@n2.B)
   RigVMFunctionReturnNode(A=@RigVMFunction_Sequence.A, B=@n2.A, C=@n3.A, D=@n3.B)
  ]}
 RigVMModel {nodes [
   RigUnit_BeginExecution = &RigUnit_BeginExecution()
   RigVMFunction_Sequence = &RigVMFunction_Sequence(ExecuteContext=@RigUnit_BeginExecution.ExecutePin)
   Control Setup = RigVMFunctionReferenceNode(Debug Controls=false, ExecuteContext=@RigVMFunction_Sequence.A)
   FN_DeformRig_FNC = RigVMFunctionReferenceNode(
    DeformRig_ Curve="(deformRig_l_upperarm_up_alpha_72_#1=true,deformRig_r_upperarm_up_alpha_73_#2=true,deformRig_l_upperarm_fr_alpha_74_#3=true,deformRig_r_upperarm_fr_alpha_75_CB51B80C4 ...(truncated)"
    DeformRig_Bool="(upperarm_1_#4=True,elbow_3_#5=True,wrist_7_#6=True,groin_8_#7=true,glute_17_#8=true,knee_9_40C408324CA261 ...(truncated)"
    DeformRig_Debug=(upperarm_debug_18_#4=False,wrist_debug_20_#6=False,lat_debug_26_#9=False,groinGlute_debug_27_#7=false)
    DeformRig_Value="(upperarm_up_18_#10=(Rotation=(X=0.0,Y=-0.0,Z=0.0,W=1.0),Translation=(X=0.0,Y=0.0,Z=8.0),Scale3D=(X=0.70,Y=1.0,Z=1.0)),upperarm_fwd_fr_19_#11=(Rotation=(X=0.0,Y=-0.0,Z=0.0,W=1.0),Translation=(X=0.0, ...(truncated)"
    ExecuteContext=@RigVMFunction_Sequence.B
   )
   > Variables = RigVMFunctionReferenceNode()
   > Skirt = RigVMFunctionReferenceNode()
   > Scarf = RigVMFunctionReferenceNode()
   Scarf_1 = RigVMFunctionReferenceNode(ExecuteContext=())
   Scarf_2 = RigVMFunctionReferenceNode(ExecuteContext=())
  ]}
]
hierarchy [
 root {loc=<0 0 0>, rot=<0 0 0 1>}
 pelvis {parent=root, loc=<0 -0.5301 98.0571>, rot=<0.0127 0.707 -0.0127 -0.707>}
 spine_01 {parent=pelvis, loc=<6.1208 -0.9328 0>, rot=<0 0 0.034 -0.9994>}
 spine_02 {parent=spine_01, loc=<6.5997 -0.2095 0>, rot=<0 0 0.0238 -0.9997>}
 spine_03 {parent=spine_02, loc=<6.4914 -0.1237 0>, rot=<0 0 0.0149 -0.9999>}
 spine_04 {parent=spine_03, loc=<6.9598 0.0794 0>, rot=<0 0 -0.0132 -0.9999>}
 spine_05 {parent=spine_04, loc=<6.9987 0 0>, rot=<0 0 -0.0015 -1>}
 neck_01 {parent=spine_05, loc=<11.3138 0.1529 0>, rot=<0 0 -0.2003 -0.9797>}
 neck_02 {parent=neck_01, loc=<4.5627 0 0>, rot=<0 0 0.0504 -0.9987>}
 head {parent=neck_02, loc=<5.3914 0 0>, rot=<0 0 0.115 -0.9934>}
 clavicle_l {parent=spine_05, loc=<5.4831 -0.1053 -3.2946>, rot=<0.0465 -0.7289 -0.009 -0.6829>}
 upperarm_l {parent=clavicle_l, loc=<9.8658 -1.2136 -0.4105>, rot=<0.0164 -0.3245 -0.0269 -0.9454>}
 lowerarm_l {parent=upperarm_l, loc=<26.465 -0.8345 0.259>, rot=<0.0036 0.0161 -0.2298 -0.9731>}
 hand_l {parent=lowerarm_l, loc=<22.7355 -0.3203 -0.6105>, rot=<-0.6784 -0.0223 0.0735 -0.7307>}
 thumb_01_l {parent=hand_l, loc=<1.3118 -0.3191 -2.459>, rot=<0.6546 -0.4042 0.0664 -0.6354>}
 thumb_02_l {parent=thumb_01_l, loc=<5.9336 -0.0448 0.0401>, rot=<-0.0037 0.0213 0.298 -0.9543>}
 thumb_03_l {parent=thumb_02_l, loc=<2.2223 0.3682 -0.0907>, rot=<0.0118 -0.013 0.0066 -0.9998>}
 index_metacarpal_l {parent=hand_l, loc=<1.6535 0.0183 -2.96>, rot=<0.0158 -0.0124 -0.0173 -0.9996>}
 index_01_l {parent=index_metacarpal_l, loc=<8.3008 0.0357 -0.0524>, rot=<-0.0044 0.0485 0.0924 -0.9945>}
 index_02_l {parent=index_01_l, loc=<3.3587 -0.3004 0.0452>, rot=<-0.002 0.0027 0.3225 -0.9466>}
 index_03_l {parent=index_02_l, loc=<2.5578 -0.123 -0.0252>, rot=<-0.002 0.0067 0.2558 -0.9667>}
 middle_metacarpal_l {parent=hand_l, loc=<2.1856 -0.1725 -0.5984>, rot=<0.006 0.0168 -0.0329 -0.9993>}
 middle_01_l {parent=middle_metacarpal_l, loc=<7.3275 -0.2447 0.0129>, rot=<-0.0208 0.0042 0.0858 -0.9961>}
 middle_02_l {parent=middle_01_l, loc=<4.0123 0.1401 0.1175>, rot=<-0.0218 0.0258 0.3512 -0.9357>}
 middle_03_l {parent=middle_02_l, loc=<2.6077 0.0534 0.1373>, rot=<0.0082 0.0188 0.1708 -0.9851>}
 ring_metacarpal_l {parent=hand_l, loc=<2.2074 -0.3625 1.3638>, rot=<-0.0284 0.0722 -0.0458 -0.9959>}
 ring_01_l {parent=ring_metacarpal_l, loc=<7.0295 -0.6197 -0.1446>, rot=<-0.0489 -0.0252 0.1892 -0.9804>}
 ring_02_l {parent=ring_01_l, loc=<3.6174 0.4647 -0.1027>, rot=<0 0 0.2806 -0.9598>}
 ring_03_l {parent=ring_02_l, loc=<2.4152 0.114 0.0199>, rot=<0 0 0.1189 -0.9929>}
 pinky_metacarpal_l {parent=hand_l, loc=<1.7616 -0.9753 3.3814>, rot=<-0.0397 0.0814 -0.069 -0.9935>}
 pinky_01_l {parent=pinky_metacarpal_l, loc=<7.1709 -0.4831 -0.1094>, rot=<-0.0654 -0.0035 0.2242 -0.9723>}
 pinky_02_l {parent=pinky_01_l, loc=<2.8274 0.0893 -0.1043>, rot=<0.0037 0.0011 0.2854 -0.9584>}
 pinky_03_l {parent=pinky_02_l, loc=<2.2046 0.0307 -0.0078>, rot=<-0.0002 0.0098 0.1445 -0.9895>}
 deform_wrist_dn_base_l {parent=hand_l, loc=<0 -1.5326 0>, rot=<0 0 0 -1>}
 deform_wrist_dn_l {parent=deform_wrist_dn_base_l, loc=<0 0 0>, rot=<0 0 0 1>}
 deform_wrist_up_base_l {parent=hand_l, loc=<0 2.3009 0>, rot=<0 0 0 -1>}
 deform_wrist_up_l {parent=deform_wrist_up_base_l, loc=<0 0 0>, rot=<0 0 0 1>}
 lowerarm_twist_02_l {parent=lowerarm_l, loc=<10.6707 0 0>, rot=<0 0 0 1>}
 lowerarm_twist_01_l {parent=lowerarm_l, loc=<18.316 0 0>, rot=<0 0 0 1>}
 deform_elbow_in_base_l {parent=lowerarm_l, loc=<0.1834 3.6062 -0.2572>, rot=<0 0 0 -1>}
 deform_elbow_in_l {parent=deform_elbow_in_base_l, loc=<0 0 0>, rot=<0 0 0 1>}
 deform_elbow_out_base_l {parent=lowerarm_l, loc=<0 -4 0>, rot=<0 0 0 1>}
 deform_elbow_out_l {parent=deform_elbow_out_base_l, loc=<0 0 0>, rot=<0 0 0 1>}
 upperarm_twist_01_l {parent=upperarm_l, loc=<7.2135 -0.1134 0>, rot=<0 0 0 1>}
 deform_upperarm_bk_base_l {parent=upperarm_twist_01_l, loc=<-6.4053 -4.3746 0.314>, rot=<0 0 0 -1>}
 deform_upperarm_bk_l {parent=deform_upperarm_bk_base_l, loc=<0 0 0>, rot=<0 0 0 1>}
 deform_upperarm_fr_base_l {parent=upperarm_twist_01_l, loc=<-5.6312 5.092 0.3354>, rot=<0 0 0 -1>}
 deform_upperarm_fr_l {parent=deform_upperarm_fr_base_l, loc=<0 0 0>, rot=<0 0 0 1>}
 deform_upperarm_up_base_l {parent=upperarm_twist_01_l, loc=<-8.891 0.4825 3.9517>, rot=<0 0 0 -1>}
 deform_upperarm_up_l {parent=deform_upperarm_up_base_l, loc=<0 0 0>, rot=<0 0 0 1>}
 upperarm_twist_02_l {parent=upperarm_l, loc=<19.7998 -0.1134 0>, rot=<0 0 0 1>}
 clavicle_r {parent=spine_05, loc=<5.4831 -0.1053 3.2946>, rot=<0.7289 0.0465 -0.6829 0.009>}
 upperarm_r {parent=clavicle_r, loc=<-9.8658 1.2136 0.4105>, rot=<0.0164 -0.3245 -0.0269 -0.9454>}
 lowerarm_r {parent=upperarm_r, loc=<-26.465 0.8345 -0.259>, rot=<0.0036 0.0161 -0.2298 -0.9731>}
 hand_r {parent=lowerarm_r, loc=<-22.7355 0.3203 0.6106>, rot=<-0.6784 -0.0223 0.0735 -0.7307>}
 thumb_01_r {parent=hand_r, loc=<-1.3118 0.3191 2.459>, rot=<0.6546 -0.4042 0.0664 -0.6354>}
 thumb_02_r {parent=thumb_01_r, loc=<-5.9336 0.0448 -0.0401>, rot=<-0.0037 0.0213 0.298 -0.9543>}
 thumb_03_r {parent=thumb_02_r, loc=<-2.2223 -0.3682 0.0907>, rot=<0.0118 -0.013 0.0066 -0.9998>}
 index_metacarpal_r {parent=hand_r, loc=<-1.6535 -0.0183 2.96>, rot=<0.0158 -0.0124 -0.0173 -0.9996>}
 index_01_r {parent=index_metacarpal_r, loc=<-8.3008 -0.0357 0.0524>, rot=<-0.0044 0.0485 0.0924 -0.9945>}
 index_02_r {parent=index_01_r, loc=<-3.3587 0.3004 -0.0452>, rot=<-0.002 0.0027 0.3225 -0.9466>}
 index_03_r {parent=index_02_r, loc=<-2.5578 0.123 0.0252>, rot=<-0.002 0.0067 0.2558 -0.9667>}
 middle_metacarpal_r {parent=hand_r, loc=<-2.1856 0.1725 0.5984>, rot=<0.006 0.0168 -0.0329 -0.9993>}
 middle_01_r {parent=middle_metacarpal_r, loc=<-7.3275 0.2447 -0.0129>, rot=<-0.0208 0.0042 0.0858 -0.9961>}
 middle_02_r {parent=middle_01_r, loc=<-4.0123 -0.1401 -0.1175>, rot=<-0.0218 0.0258 0.3512 -0.9357>}
 middle_03_r {parent=middle_02_r, loc=<-2.6077 -0.0534 -0.1373>, rot=<0.0082 0.0188 0.1708 -0.9851>}
 ring_metacarpal_r {parent=hand_r, loc=<-2.2074 0.3625 -1.3638>, rot=<-0.0284 0.0722 -0.0458 -0.9959>}
 ring_01_r {parent=ring_metacarpal_r, loc=<-7.0295 0.6197 0.1446>, rot=<-0.0489 -0.0252 0.1892 -0.9804>}
 ring_02_r {parent=ring_01_r, loc=<-3.6174 -0.4647 0.1027>, rot=<0 0 0.2806 -0.9598>}
 ring_03_r {parent=ring_02_r, loc=<-2.4152 -0.114 -0.0199>, rot=<0 0 0.1189 -0.9929>}
 pinky_metacarpal_r {parent=hand_r, loc=<-1.7616 0.9753 -3.3814>, rot=<-0.0397 0.0814 -0.069 -0.9935>}
 pinky_01_r {parent=pinky_metacarpal_r, loc=<-7.1709 0.4831 0.1094>, rot=<-0.0654 -0.0035 0.2242 -0.9723>}
 pinky_02_r {parent=pinky_01_r, loc=<-2.8274 -0.0893 0.1043>, rot=<0.0037 0.0011 0.2854 -0.9584>}
 pinky_03_r {parent=pinky_02_r, loc=<-2.2046 -0.0307 0.0078>, rot=<-0.0002 0.0098 0.1445 -0.9895>}
 deform_wrist_dn_base_r {parent=hand_r, loc=<-0.0001 1.5328 0>, rot=<0 0 0 -1>}
 deform_wrist_dn_r {parent=deform_wrist_dn_base_r, loc=<0 0 0>, rot=<0 0 0 1>}
 deform_wrist_up_base_r {parent=hand_r, loc=<-0.0002 -2.3007 -0.0001>, rot=<0 0 0 -1>}
 deform_wrist_up_r {parent=deform_wrist_up_base_r, loc=<0 0 0>, rot=<0 0 0 1>}
 lowerarm_twist_02_r {parent=lowerarm_r, loc=<-10.6707 0 0>, rot=<0 0 0 -1>}
 lowerarm_twist_01_r {parent=lowerarm_r, loc=<-18.316 0 0>, rot=<0 0 0 -1>}
 deform_elbow_in_base_r {parent=lowerarm_r, loc=<-0.1834 -3.6062 0.2572>, rot=<0 0 0 -1>}
 deform_elbow_in_r {parent=deform_elbow_in_base_r, loc=<0 0 0>, rot=<0 0 0 1>}
 deform_elbow_out_base_r {parent=lowerarm_r, loc=<0 4 -0.0001>, rot=<0 0 0 -1>}
 deform_elbow_out_r {parent=deform_elbow_out_base_r, loc=<0 0 0>, rot=<0 0 0 1>}
 upperarm_twist_01_r {parent=upperarm_r, loc=<-7.2135 0.1134 0>, rot=<0 0 0 -1>}
 deform_upperarm_bk_base_r {parent=upperarm_twist_01_r, loc=<6.4053 4.3746 -0.314>, rot=<0 0 0 -1>}
 deform_upperarm_bk_r {parent=deform_upperarm_bk_base_r, loc=<0 0 0>, rot=<0 0 0 1>}
 deform_upperarm_fr_base_r {parent=upperarm_twist_01_r, loc=<5.6312 -5.092 -0.3354>, rot=<0 0 0 -1>}
 deform_upperarm_fr_r {parent=deform_upperarm_fr_base_r, loc=<0 0 0>, rot=<0 0 0 1>}
 deform_upperarm_up_base_r {parent=upperarm_twist_01_r, loc=<8.891 -0.4825 -3.9517>, rot=<0 0 0 -1>}
 deform_upperarm_up_r {parent=deform_upperarm_up_base_r, loc=<0 0 0>, rot=<0 0 0 1>}
 upperarm_twist_02_r {parent=upperarm_r, loc=<-19.7998 0.1134 0>, rot=<0 0 0 -1>}
 attach_backpack {parent=spine_05, loc=<-0.1183 -13.4745 -0.0001>, rot=<-0.0092 -0.707 0.0093 -0.707>}
 Backpack_BR {parent=attach_backpack, type=Null, loc=<0 0 0>, rot=<0 0 0 1>}
 attach_cape {parent=spine_05, loc=<9.0206 -11.7152 -0.0001>, rot=<0.0616 -0.7044 -0.0616 -0.7044>}
 cape {parent=attach_cape, type=Null, loc=<0 -1.455 0>, rot=<0 0 0 1>}
 cape_0 {parent=attach_cape, type=Null, loc=<0 -0.4512 0>, rot=<0 0 0 1>}
 base_scarfTop {parent=spine_05, loc=<7.252 8.1363 5.5243>, rot=<0.0522 -0.1966 0.9627 0.1785>}
 dyn_scarfTop_01 {parent=base_scarfTop, loc=<0 0 0>, rot=<0 0 0 1>}
 dyn_scarfTop_02 {parent=dyn_scarfTop_01, loc=<4.3678 0 0>, rot=<0 0 0 1>}
 dyn_scarfTop_03 {parent=dyn_scarfTop_02, loc=<4.3678 0 0>, rot=<0 0 0 1>}
 dyn_scarfTop_04 {parent=dyn_scarfTop_03, loc=<4.9373 0 0>, rot=<0 0 -0.0673 -0.9977>}
 base_scarfBottom {parent=spine_05, loc=<4.4668 8.8541 3.6645>, rot=<-0.0552 -0.0148 0.9824 0.1781>}
 dyn_scarfBottom_01 {parent=base_scarfBottom, loc=<0 0 0>, rot=<0 0 0 1>}
 dyn_scarfBottom_02 {parent=dyn_scarfBottom_01, loc=<4.7235 0 0>, rot=<0 0 0 1>}
 dyn_scarfBottom_03 {parent=dyn_scarfBottom_02, loc=<4.7235 0 0>, rot=<0 0 -0.1402 -0.9901>}
 dyn_scarfBottom_04 {parent=dyn_scarfBottom_03, loc=<4.9527 0 0>, rot=<0 0 -0.2161 -0.9764>}
 deform_pec_aim_base_l {parent=spine_04, loc=<3.2456 9.3309 -1.2237>, rot=<0.0603 -0.4033 0.3318 -0.8507>}
 deform_pec_aim_l {parent=deform_pec_aim_base_l, loc=<0 0 0>, rot=<0 0 0 1>}
 deform_pec_base_l {parent=deform_pec_aim_l, loc=<9.3665 0 0>, rot=<0 0 0 1>}
 deform_pec_l {parent=deform_pec_base_l, loc=<0 0 0>, rot=<0 0 0 1>}
 deform_pec_aim_base_r {parent=spine_04, loc=<3.2452 9.3309 1.2237>, rot=<-0.4033 -0.0603 0.8507 0.3318>}
 deform_pec_aim_r {parent=deform_pec_aim_base_r, loc=<0 0 0>, rot=<0 0 0 1>}
 deform_pec_base_r {parent=deform_pec_aim_r, loc=<-9.3667 -0.0003 -0.0004>, rot=<0 0 0 1>}
 deform_pec_r {parent=deform_pec_base_r, loc=<0 0 0>, rot=<0 0 0 1>}
 spine_05_null {parent=spine_01, type=Null, loc=<-29.2097 -6.2232 -2.1145>, rot=<0.2081 0.0255 0.2307 0.9502>}
 thigh_l {parent=pelvis, loc=<-1.0855 -0.2109 -8.7053>, rot=<0.0009 0.0355 0.0155 -0.9993>}
 calf_l {parent=thigh_l, loc=<-42.5284 -0.0019 -0.7297>, rot=<0.0006 -0.0161 -0.0391 -0.9991>}
 foot_l {parent=calf_l, loc=<-46.7401 -2.5014 -0.592>, rot=<0.0427 -0.0238 0.0353 -0.9982>}
 ball_l {parent=foot_l, loc=<-3.5658 10.8066 -0.1544>, rot=<0.0001 -0.0001 -0.7063 -0.7079>}
 calf_twist_02_l {parent=calf_l, loc=<-19.1546 -1.7943 -0.2638>, rot=<0.0237 0.0095 -0.0035 -0.9997>}
 calf_twist_01_l {parent=calf_l, loc=<-35.8483 -1.9197 -0.5794>, rot=<0.0237 0.0095 -0.0035 -0.9997>}
 deform_knee_in_base_l {parent=calf_l, loc=<0.8031 -5.0231 0.2456>, rot=<0 0 0 -1>}
 deform_knee_in_l {parent=deform_knee_in_base_l, loc=<0 0 0>, rot=<0 0 0 1>}
 deform_knee_out_base_l {parent=calf_l, loc=<0 4 0>, rot=<0 0 0 -1>}
 deform_knee_out_l {parent=deform_knee_out_base_l, loc=<0 0 0>, rot=<0 0 0 1>}
 thigh_twist_01_l {parent=thigh_l, loc=<-27.952 0 0>, rot=<0.0013 0 0 -1>}
 deform_groin_base_l {parent=thigh_l, loc=<0.4801 5.6776 -0.2486>, rot=<0 0 0 -1>}
 deform_groin_l {parent=deform_groin_base_l, loc=<0 0 0>, rot=<0 0 0 1>}
 calf_l_null {parent=thigh_l, type=Null, loc=<-42.5284 -0.0019 -0.7297>, rot=<0.0006 -0.0161 -0.0391 -0.9991>}
 thigh_r {parent=pelvis, loc=<-1.0855 -0.2109 8.7053>, rot=<0.0355 -0.0009 0.9993 0.0155>}
 calf_r {parent=thigh_r, loc=<42.5284 0.0019 0.7297>, rot=<0.0006 -0.0161 -0.0391 -0.9991>}
 foot_r {parent=calf_r, loc=<46.7401 2.5014 0.592>, rot=<0.0427 -0.0238 0.0353 -0.9982>}
 ball_r {parent=foot_r, loc=<3.5658 -10.8066 0.1544>, rot=<0.0001 -0.0001 -0.7063 -0.7079>}
 calf_twist_02_r {parent=calf_r, loc=<19.1546 1.7943 0.2638>, rot=<0.0237 0.0095 -0.0035 -0.9997>}
 calf_twist_01_r {parent=calf_r, loc=<35.8483 1.9197 0.5794>, rot=<0.0237 0.0095 -0.0035 -0.9997>}
 deform_knee_in_base_r {parent=calf_r, loc=<-0.8031 5.0231 -0.2457>, rot=<0 0 0 -1>}
 deform_knee_in_r {parent=deform_knee_in_base_r, loc=<0 0 0>, rot=<0 0 0 1>}
 deform_knee_out_base_r {parent=calf_r, loc=<0 -4 0>, rot=<0 0 0 -1>}
 deform_knee_out_r {parent=deform_knee_out_base_r, loc=<0 0 0>, rot=<0 0 0 1>}
 thigh_twist_01_r {parent=thigh_r, loc=<27.952 0 0>, rot=<0.0013 0 0 -1>}
 deform_groin_base_r {parent=thigh_r, loc=<-0.4801 -5.6776 0.2486>, rot=<0 0 0 -1>}
 deform_groin_r {parent=deform_groin_base_r, loc=<0 0 0>, rot=<0 0 0 1>}
 calf_r_null {parent=thigh_r, type=Null, loc=<42.5284 0.0019 0.7297>, rot=<0.0006 -0.0161 -0.0391 -0.9991>}
 deform_glute_base_l {parent=pelvis, loc=<-7.3466 -9.0213 -9.1245>, rot=<0.0172 0.031 -0.4477 -0.8935>}
 deform_glute_l {parent=deform_glute_base_l, loc=<0 0 0>, rot=<0 0 0 1>}
 deform_glute_base_r {parent=pelvis, loc=<-7.3464 -9.0213 9.1245>, rot=<0.031 -0.0172 0.8935 -0.4477>}
 deform_glute_r {parent=deform_glute_base_r, loc=<0 0 0>, rot=<0 0 0 1>}
 base_skirtFront {parent=pelvis, loc=<9.0482 7.9087 0>, rot=<0 0 0.9986 0.0523>}
 dyn_skirtFront_01 {parent=base_skirtFront, loc=<0.2796 0 0>, rot=<0 0 0 1>}
 dyn_skirtFront_02 {parent=dyn_skirtFront_01, loc=<10.2873 0 0>, rot=<0 0 0 1>}
 dyn_skirtFront_03 {parent=dyn_skirtFront_02, loc=<9.8209 0 0>, rot=<0 0 0 1>}
 base_skirtBack {parent=pelvis, loc=<10.2806 -8.6926 0>, rot=<0.1355 -0.9908 0 0>}
 dyn_skirtBack_01 {parent=base_skirtBack, loc=<0 0 0>, rot=<0 0 0 1>}
 dyn_skirtBack_02 {parent=dyn_skirtBack_01, loc=<10.2999 0 0>, rot=<0 0 -0.0819 -0.9966>}
 dyn_skirtBack_03 {parent=dyn_skirtBack_02, loc=<10.7067 0 0>, rot=<0 0 -0.0452 -0.999>}
 base_skirtSide_l {parent=pelvis, loc=<9.5304 -0.0958 -11.6614>, rot=<-0.0896 0.6968 0.7014 0.1205>}
 dyn_skirtSide_01_l {parent=base_skirtSide_l, loc=<0 0 0>, rot=<0 0 0 1>}
 dyn_skirtSide_02_l {parent=dyn_skirtSide_01_l, loc=<10.2873 0 0>, rot=<0 0 0 1>}
 dyn_skirtSide_03_l {parent=dyn_skirtSide_02_l, loc=<10.123 0 0>, rot=<0 0 0 1>}
 base_skirtFrontSide_l {parent=pelvis, loc=<9.076 5.8832 -8.0503>, rot=<-0.0639 0.3069 0.9449 0.094>}
 dyn_skirtFrontSide_01_l {parent=base_skirtFrontSide_l, loc=<0 0 0>, rot=<0 0 0 1>}
 dyn_skirtFrontSide_02_l {parent=dyn_skirtFrontSide_01_l, loc=<10.2873 0 0>, rot=<0 0 0 1>}
 dyn_skirtFrontSide_03_l {parent=dyn_skirtFrontSide_02_l, loc=<9.8209 0 0>, rot=<0 0 0 1>}
 base_skirtBackSide_l {parent=pelvis, loc=<10.1148 -6.8043 -7.3456>, rot=<0.1752 -0.9269 -0.3077 -0.1248>}
 dyn_skirtBackSide_01_l {parent=base_skirtBackSide_l, loc=<0 0 0>, rot=<0 0 0 1>}
 dyn_skirtBackSide_02_l {parent=dyn_skirtBackSide_01_l, loc=<10.3179 0 0>, rot=<0 0 -0.1176 -0.9931>}
 dyn_skirtBackSide_03_l {parent=dyn_skirtBackSide_02_l, loc=<10.7043 0 0>, rot=<0 0 -0.0461 -0.9989>}
 base_skirtSide_r {parent=pelvis, loc=<9.5301 -0.0957 11.5016>, rot=<0.0896 -0.6968 0.7014 0.1205>}
 dyn_skirtSide_01_r {parent=base_skirtSide_r, loc=<0 0 0>, rot=<0 0 0 1>}
 dyn_skirtSide_02_r {parent=dyn_skirtSide_01_r, loc=<10.2873 0 0>, rot=<0 0 0 1>}
 dyn_skirtSide_03_r {parent=dyn_skirtSide_02_r, loc=<10.123 0 0>, rot=<0 0 0 1>}
 base_skirtFrontSide_r {parent=pelvis, loc=<9.076 5.8831 8.0506>, rot=<0.064 -0.3061 0.9452 0.094>}
 dyn_skirtFrontSide_01_r {parent=base_skirtFrontSide_r, loc=<0 0 0>, rot=<0 0 0 1>}
 dyn_skirtFrontSide_02_r {parent=dyn_skirtFrontSide_01_r, loc=<10.2873 0 0>, rot=<0 0 0 1>}
 dyn_skirtFrontSide_03_r {parent=dyn_skirtFrontSide_02_r, loc=<9.8209 0 0>, rot=<0 0 0 1>}
 base_skirtBackSide_r {parent=pelvis, loc=<10.1149 -6.8037 7.3466>, rot=<0.1748 -0.9279 0.3045 0.1253>}
 dyn_skirtBackSide_01_r {parent=base_skirtBackSide_r, loc=<0 0 0>, rot=<0 0 0 1>}
 dyn_skirtBackSide_02_r {parent=dyn_skirtBackSide_01_r, loc=<10.3179 0 0>, rot=<0 0 -0.1176 -0.9931>}
 dyn_skirtBackSide_03_r {parent=dyn_skirtBackSide_02_r, loc=<10.7043 0 0>, rot=<0 0 -0.0461 -0.9989>}
 thigh_l_null {parent=pelvis, type=Null, loc=<-1.0855 -0.2109 -8.7053>, rot=<0.0009 0.0355 0.0155 -0.9993>}
 thigh_r_null {parent=pelvis, type=Null, loc=<-1.0855 -0.2109 8.7053>, rot=<0.0355 -0.0009 0.9993 0.0155>}
 deform_internalSkirt_r {parent=pelvis, loc=<2.4623 6.1909 9.1522>, rot=<0 0 0 1>}
 deform_internalSkirt_l {parent=pelvis, loc=<2.4623 6.1909 -9.152>, rot=<0 0 0 1>}
 thigh_r_ctrl {controlType=EULER_TRANSFORM, type=Control, loc=<0 0 0>, rot=<0 0 0 1>}
 calf_r_ctrl {controlType=EULER_TRANSFORM, parent=thigh_r_ctrl, type=Control, loc=<0 0 0>, rot=<0 0 0 1>}
 thigh_l_ctrl {controlType=EULER_TRANSFORM, type=Control, loc=<0 0 0>, rot=<0 0 0 1>}
 calf_l_ctrl {controlType=EULER_TRANSFORM, parent=thigh_l_ctrl, type=Control, loc=<0 0 0>, rot=<0 0 0 1>}
]
variables [
 Calf_L_BackRotation {type=double, default=0}
 Calf_R_BackRotation {type=double, default=0}
 Thigh_L_BackRotation {type=double, default=0}
 Thigh_L_SideRotation {type=double, default=0}
 Thigh_R_BackRotation {type=double, default=0}
 Thigh_R_SideRotation {type=double, default=0}
 crouchWalking {type=double, default=0}
 isArcanaAgateAxeEquipped {type=bool, default=False}
 isBaseballSliding_CR {type=bool, default=False}
 isCrouching_CR {type=bool, default=False}
 isFrontEnd_CR {type=bool, default=False}
 isKneeSliding_CR {type=bool, default=False}
 isOnRail_CR {type=bool, default=False}
 onehandedPickaxe_CR {type=double, default=0}
 spineTwist_R {type=double, default=0}
]
