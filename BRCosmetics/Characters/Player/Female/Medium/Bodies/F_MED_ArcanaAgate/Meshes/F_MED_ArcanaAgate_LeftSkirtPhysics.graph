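# Physics asset for the left-side skirt of the ArcanaAgate medium female player body cosmetic.
# Three chains of three bones (front-side, side, back-side) link by constraints with tight swing limits and locked twist, without cross-links or constraints above each first bone.
# Lower skirt bones use large capsules offset sideways from the bone.
# Capsules on both thighs, calves and glutes supply body shapes, with only calf-to-thigh constraints among them.
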PhysicsAsset /BRCosmetics/Characters/Player/Female/Medium/Bodies/F_MED_ArcanaAgate/Meshes/F_MED_ArcanaAgate_LeftSkirtPhysics
bodies {
 calf_l [{shapeType=Capsule, length=22.5921, radius=4.8367, center=<-12.901 -1.4473 -0.2096>, rotation=<rot -0.3773 93.1793 -87.4762>}]
 calf_r [{shapeType=Capsule, length=22.5921, radius=4.8367, center=<12.901 1.4473 0.2096>, rotation=<rot -0.3773 93.1793 92.5238>}]
 deform_glute_l [{
   shapeType=Capsule
   length=0
   radius=8.4992
   shapeName=deform_glute_l_capsule
   center=<8.0204 0.2535 1.9291>
   rotation=<rot 15.5528 -90 -32.4501>
  }]
 deform_glute_r [{
   shapeType=Capsule
   length=0
   radius=8.4992
   shapeName=deform_glute_r_capsule
   center=<-6.3691 -1.5273 -1.8208>
   rotation=<rot 15.5528 -90 147.5499>
  }]
 dyn_skirtBackSide_01_l [{
   shapeType=Capsule
   length=4.0538
   radius=2.7654
   shapeName=dyn_skirtBackSide_01_l_capsule
   center=<4.4579 0.6801 0.2119>
   rotation=<rot -1.5549 -90 9.7728>
  }]
 dyn_skirtBackSide_02_l [{shapeType=Capsule, length=14.9277, radius=8.6697, shapeName=dyn_skirtFront_03_capsule, center=<6.0134 -8.5345 0>, rotation=<rot 0 -90 0>}]
 dyn_skirtBackSide_03_l [{shapeType=Capsule, length=13.6983, radius=9.9614, shapeName=dyn_skirtFront_03_capsule, center=<9.4492 -6.3226 0>, rotation=<rot 0 -90 0>}]
 dyn_skirtFrontSide_01_l [{
   shapeType=Capsule
   length=4.441
   radius=3.0469
   shapeName=dyn_skirtFrontSide_01_l_capsule
   center=<4.9046 0.629 0.44>
   rotation=<rot 8.4906 -90 -14.1792>
  }]
 dyn_skirtFrontSide_02_l [{shapeType=Capsule, length=7.1728, radius=8.6697, shapeName=dyn_skirtFront_03_capsule, center=<6.0134 -9.1695 0>, rotation=<rot 0 -90 0>}]
 dyn_skirtFrontSide_03_l [{shapeType=Capsule, length=18.464, radius=9.9614, shapeName=dyn_skirtFront_03_capsule, center=<7.3446 -8.4272 0>, rotation=<rot 0 -90 0>}]
 dyn_skirtSide_01_l [{
   shapeType=Capsule
   length=3.9218
   radius=3.5886
   shapeName=dyn_skirtSide_01_l_capsule
   center=<3.8278 0.6605 -0.2524>
   rotation=<rot 0 -101.6246 -53.5897>
  }]
 dyn_skirtSide_02_l [{shapeType=Capsule, length=7.1728, radius=8.6697, shapeName=dyn_skirtFront_03_capsule, center=<6.0134 -8.8257 0>, rotation=<rot 0 -90 0>}]
 dyn_skirtSide_03_l [{shapeType=Capsule, length=13.6983, radius=9.9614, shapeName=dyn_skirtFront_03_capsule, center=<9.4426 -7.0194 0>, rotation=<rot 0 -90 0>}]
 thigh_l [
  {shapeType=Capsule, length=33.2791, radius=7.3994, center=<-21.0438 0.2728 0.0803>, rotation=<rot 0.0324 -87.5344 -90.7484>}
  {
   shapeType=Capsule
   length=5.3613
   radius=4.3773
   shapeName=thigh_l_capsule
   center=<-26.527 3.3389 -0.0624>
   rotation=<rot -87.019 -179.9999 179.9997>
  }
  {
   shapeType=Capsule
   length=0
   radius=8.4448
   shapeName=thigh_l_capsule
   center=<-4.3128 -3.2108 2.8174>
   rotation=<rot -0.1598 -86.7475 -96.7386>
  }
 ]
 thigh_r [
  {shapeType=Capsule, length=33.2791, radius=7.2012, center=<21.0716 -0.9184 -0.0799>, rotation=<rot 0 -90.0153 89.2509>}
  {
   shapeType=Capsule
   length=5.3613
   radius=4.3773
   shapeName=thigh_r_capsule
   center=<26.527 -3.3389 0.0624>
   rotation=<rot -87.019 -179.9999 -0.0003>
  }
  {
   shapeType=Capsule
   length=0
   radius=8.4448
   shapeName=thigh_r_capsule
   center=<4.3128 3.2108 -2.8174>
   rotation=<rot -0.1598 -86.7475 83.2614>
  }
 ]
}
constraints [
 {bone1Name=calf_l, bone2Name=thigh_l, swing1LimitDegrees=45, swing2LimitDegrees=45, twistLimitDegrees=45}
 {bone1Name=calf_r, bone2Name=thigh_r, swing1LimitDegrees=45, swing2LimitDegrees=45, twistLimitDegrees=45}
 {
  bone1Name=dyn_skirtBackSide_02_l
  bone2Name=dyn_skirtBackSide_01_l
  swing1LimitDegrees=15
  swing2LimitDegrees=10
  twistLimitDegrees=45
  twistMotion=Locked
 }
 {
  bone1Name=dyn_skirtBackSide_03_l
  bone2Name=dyn_skirtBackSide_02_l
  swing1LimitDegrees=20
  swing2LimitDegrees=10
  twistLimitDegrees=45
  twistMotion=Locked
 }
 {
  bone1Name=dyn_skirtFrontSide_02_l
  bone2Name=dyn_skirtFrontSide_01_l
  swing1LimitDegrees=14
  swing2LimitDegrees=4
  twistLimitDegrees=45
  twistMotion=Locked
 }
 {
  bone1Name=dyn_skirtFrontSide_03_l
  bone2Name=dyn_skirtFrontSide_02_l
  swing1LimitDegrees=12
  swing2LimitDegrees=4
  twistLimitDegrees=45
  twistMotion=Locked
 }
 {
  bone1Name=dyn_skirtSide_02_l
  bone2Name=dyn_skirtSide_01_l
  swing1LimitDegrees=13
  swing2LimitDegrees=10
  twistLimitDegrees=45
  twistMotion=Locked
 }
 {
  bone1Name=dyn_skirtSide_03_l
  bone2Name=dyn_skirtSide_02_l
  swing1LimitDegrees=15
  swing2LimitDegrees=10
  twistLimitDegrees=45
  twistMotion=Locked
 }
]
