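# Post-process-style AnimBP for the ArcanaAgate female body cosmetic, adding secondary motion to scarf and skirt.
# Copied parent pose runs through a Control Rig, scarf constraints and AnimDynamics, per-thigh skirt RigidBody sims, averaging constraints, then corrective bone offsets.
# EventGraph updates gravity override, upright state and skirt alpha, and detects one-handed pickaxes and the tagged ArcanaAgate axe.
# The axe-driven right/back skirt offset nodes are disconnected from the output.
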AnimBlueprint /BRCosmetics/Characters/Player/Female/Medium/Bodies/F_MED_ArcanaAgate/Meshes/F_MED_ArcanaAgate_AnimBP
target_skeleton=&/F_MED_ArcanaAgate_Skeleton
graphs [
 AnimGraph {nodes [
   n1 = Get ArcanaAgateAxeEquipped() -> (ArcanaAgateAxeEquipped:bool)
   n2 = "Transform (Modify) Bone - Bone: base_skirtSide_r"!(ComponentPose, bAlphaBoolEnabled:bool=@n1, Translation="0,0,0", Rotation="0,2,0", Scale="1,1,1") -> (Pose) {
    alphaInputType=Bool
    rotationMode=BMM_Additive
    rotationSpace=BCS_BoneSpace
    translationMode=BMM_Additive
    translationSpace=BCS_BoneSpace
    alphaBoolBlend {blendInTime=1, blendOption=Linear, blendOutTime=1}
    boneToModify {boneName=base_skirtSide_r}
    rotation {yaw=2}
   }
   n3 = "Transform (Modify) Bone - Bone: base_skirtBackSide_r"!(ComponentPose=@n2, bAlphaBoolEnabled:bool=@n1, Translation="0,0,0", Rotation="0,2,0", Scale="1,1,1") -> (Pose) {
    alphaInputType=Bool
    rotationMode=BMM_Additive
    rotationSpace=BCS_BoneSpace
    translationMode=BMM_Additive
    translationSpace=BCS_BoneSpace
    alphaBoolBlend {blendInTime=1, blendOption=Linear, blendOutTime=1}
    boneToModify {boneName=base_skirtBackSide_r}
    rotation {yaw=2}
   }
   "Transform (Modify) Bone - Bone: base_skirtBack"!(ComponentPose=@n3, bAlphaBoolEnabled:bool=@n1, Translation="0,0,0", Rotation="0,5,0", Scale="1,1,1") -> (Pose) {
    alphaInputType=Bool
    rotationMode=BMM_Additive
    rotationSpace=BCS_BoneSpace
    translationMode=BMM_Additive
    translationSpace=BCS_BoneSpace
    alphaBoolBlend {blendInTime=1, blendOption=Linear, blendOutTime=1}
    boneToModify {boneName=base_skirtBack}
    rotation {yaw=5}
   }
   "Transform (Modify) Bone - Bone: dyn_skirtSide_03_r"!(ComponentPose, bAlphaBoolEnabled:bool=@n1, Translation="0,0,0", Rotation="0,3,0", Scale="1,1,1") -> (Pose) {
    alphaInputType=Bool
    rotationMode=BMM_Additive
    rotationSpace=BCS_BoneSpace
    translationMode=BMM_Additive
    translationSpace=BCS_BoneSpace
    alphaBoolBlend {blendInTime=1, blendOption=Linear, blendOutTime=1}
    boneToModify {boneName=dyn_skirtSide_03_r}
    rotation {yaw=3}
   }
   n4 = Control Rig!(
    Source=Copy Pose From Mesh() {bUseAttachedParent=true}
    Alpha
    isCrouching_CR:bool=Get bIsCrouching()
    isBaseballSliding_CR:bool=Get bIsBaseballSliding()
    isKneeSliding_CR:bool=Get bIsKneeSliding()
    isFrontEnd_CR:bool=Get bIsInFrontEnd()
    isOnRail_CR:bool=Get bIsOnGrindRail()
    onehandedPickaxe_CR:double=Get OneHandedPickaxe() -> (OneHandedPickaxe:double)
    isArcanaAgateAxeEquipped:bool=Get ArcanaAgateAxeEquipped()
   ) -> (Pose) {
    alphaBoolBlend {blendOption=Linear}
    controlRigAssetReference {blueprintRigClass=&/F_MED_ArcanaAgate_CtrlRig.F_MED_ArcanaAgate_CtrlRig_C}
   }
   n5 = "Constraint - base_scarfTop (PPPPP)"!(
    ComponentPose=Local To Component(LocalPose=@n4)
    Alpha=1
    ConstraintWeights_0=0.7981
    ConstraintWeights_1:float=0.1831
    ConstraintWeights_2:float=0.0133
    ConstraintWeights_3:float=0.0028
    ConstraintWeights_4:float=0.0026
   ) -> (Pose) {
    alphaBoolBlend {blendOption=Linear}
    boneToModify {boneName=base_scarfTop}
    constraintSetup [
     {targetBone {boneName=spine_05}}
     {targetBone {boneName=clavicle_r}}
     {targetBone {boneName=spine_04}}
     {targetBone {boneName=deform_pec_r}}
     {targetBone {boneName=neck_01}}
    ]
    constraintWeights [1,1,1,1,1]
   }
   n6 = "Constraint - base_scarfBottom (PPPPP)"!(
    ComponentPose=@n5
    Alpha=1
    ConstraintWeights_0=0.6834
    ConstraintWeights_1:float=0.1977
    ConstraintWeights_2:float=0.0622
    ConstraintWeights_3:float=0.0556
    ConstraintWeights_4:float=0.001
   ) -> (Pose) {
    alphaBoolBlend {blendOption=Linear}
    boneToModify {boneName=base_scarfBottom}
    constraintSetup [
     {targetBone {boneName=spine_05}}
     {targetBone {boneName=spine_04}}
     {targetBone {boneName=clavicle_r}}
     {targetBone {boneName=deform_pec_r}}
     {targetBone {boneName=upperarm_twist_01_r}}
    ]
    constraintWeights [1,1,1,1,1]
   }
   n7 = Anim Dynamics - Bone: dyn_scarfTop_01(ComponentPose=@n6, Alpha=1, ExternalForce="0,0,0") {
    angularDampingOverride=1
    angularSpringConstant=1000
    bAngularSpring=true
    bOverrideAngularDamping=true
    bUseGravityOverride=true
    gravityScale=0.35
    simulationSpace=RootRelative
    alphaBoolBlend {blendOption=Linear}
    boundBone {boneName=dyn_scarfTop_01}
    componentAppliedLinearAccClamp=<6000 6000 6000>
    componentLinearAccScale=<0.4 0.4 0.4>
    componentLinearVelScale=<0.5 0.5 0.5>
    gravityOverride {z=-980}
    physicsBodyDefinitions [{
      boundBone {boneName=dyn_scarfTop_01}
      boxExtents=<20 20 20>
      constraintSetup {angularLimitsMax=<15 8 0>, angularLimitsMin=<-15 -8 -15>, angularTarget=<1 0 0>}
      localJointOffset=<4 0 0>
     }]
    simSpaceSettings {maxAngularAcceleration=40, maxAngularVelocity=8, simSpaceAngularAlpha=1}
   }
   n8 = Anim Dynamics - Bone: dyn_scarfTop_02(ComponentPose=@n7, Alpha=1, ExternalForce="0,0,0") {
    angularDampingOverride=1
    angularSpringConstant=1000
    bAngularSpring=true
    bOverrideAngularDamping=true
    bUseGravityOverride=true
    gravityScale=0.35
    simulationSpace=RootRelative
    alphaBoolBlend {blendOption=Linear}
    boundBone {boneName=dyn_scarfTop_02}
    componentAppliedLinearAccClamp=<6000 6000 6000>
    componentLinearAccScale=<0.4 0.4 0.4>
    componentLinearVelScale=<0.5 0.5 0.5>
    gravityOverride {z=-980}
    physicsBodyDefinitions [{
      boundBone {boneName=dyn_scarfTop_02}
      boxExtents=<20 20 20>
      constraintSetup {angularLimitsMax=<15 8 0>, angularLimitsMin=<-15 -8 -15>, angularTarget=<1 0 0>}
      localJointOffset=<4 0 0>
     }]
    simSpaceSettings {maxAngularAcceleration=40, maxAngularVelocity=8, simSpaceAngularAlpha=1}
   }
   n9 = Anim Dynamics - Bone: dyn_scarfTop_03(ComponentPose=@n8, Alpha=1, ExternalForce="0,0,0") {
    angularDampingOverride=1
    angularSpringConstant=1000
    bAngularSpring=true
    bOverrideAngularDamping=true
    bUseGravityOverride=true
    gravityScale=0.35
    simulationSpace=RootRelative
    alphaBoolBlend {blendOption=Linear}
    boundBone {boneName=dyn_scarfTop_03}
    componentAppliedLinearAccClamp=<6000 6000 6000>
    componentLinearAccScale=<0.4 0.4 0.4>
    componentLinearVelScale=<0.5 0.5 0.5>
    gravityOverride {z=-980}
    physicsBodyDefinitions [{
      boundBone {boneName=dyn_scarfTop_03}
      boxExtents=<20 20 20>
      constraintSetup {angularLimitsMax=<15 8 0>, angularLimitsMin=<-15 -8 -15>, angularTarget=<1 0 0>}
      localJointOffset=<4 0 0>
     }]
    simSpaceSettings {maxAngularAcceleration=40, maxAngularVelocity=8, simSpaceAngularAlpha=1}
   }
   n10 = Anim Dynamics - Bone: dyn_scarfBottom_01(ComponentPose=@n9, Alpha=1, ExternalForce="0,0,0") {
    angularDampingOverride=1
    angularSpringConstant=1000
    bAngularSpring=true
    bOverrideAngularDamping=true
    bUseGravityOverride=true
    gravityScale=0.35
    simulationSpace=RootRelative
    alphaBoolBlend {blendOption=Linear}
    boundBone {boneName=dyn_scarfBottom_01}
    componentAppliedLinearAccClamp=<6000 6000 6000>
    componentLinearAccScale=<0.4 0.4 0.4>
    componentLinearVelScale=<0.5 0.5 0.5>
    gravityOverride {z=-980}
    physicsBodyDefinitions [{
      boundBone {boneName=dyn_scarfBottom_01}
      boxExtents=<20 20 20>
      constraintSetup {angularLimitsMax=<15 8 0>, angularLimitsMin=<-15 -8 -15>, angularTarget=<1 0 0>}
      localJointOffset=<4 0 0>
     }]
    simSpaceSettings {maxAngularAcceleration=40, maxAngularVelocity=8, simSpaceAngularAlpha=1}
   }
   n11 = Anim Dynamics - Bone: dyn_scarfBottom_02(ComponentPose=@n10, Alpha=1, ExternalForce="0,0,0") {
    angularDampingOverride=1
    angularSpringConstant=1000
    bAngularSpring=true
    bOverrideAngularDamping=true
    bUseGravityOverride=true
    gravityScale=0.35
    simulationSpace=RootRelative
    alphaBoolBlend {blendOption=Linear}
    boundBone {boneName=dyn_scarfBottom_02}
    componentAppliedLinearAccClamp=<6000 6000 6000>
    componentLinearAccScale=<0.4 0.4 0.4>
    componentLinearVelScale=<0.5 0.5 0.5>
    gravityOverride {z=-980}
    physicsBodyDefinitions [{
      boundBone {boneName=dyn_scarfBottom_02}
      boxExtents=<20 20 20>
      constraintSetup {angularLimitsMax=<15 8 0>, angularLimitsMin=<-15 -8 -15>, angularTarget=<1 0 0>}
      localJointOffset=<4 0 0>
     }]
    simSpaceSettings {maxAngularAcceleration=40, maxAngularVelocity=8, simSpaceAngularAlpha=1}
   }
   n12 = Anim Dynamics - Bone: dyn_scarfBottom_03(ComponentPose=@n11, Alpha=1, ExternalForce="0,0,0") {
    angularDampingOverride=1
    angularSpringConstant=1000
    bAngularSpring=true
    bOverrideAngularDamping=true
    bUseGravityOverride=true
    gravityScale=0.35
    simulationSpace=RootRelative
    alphaBoolBlend {blendOption=Linear}
    boundBone {boneName=dyn_scarfBottom_03}
    componentAppliedLinearAccClamp=<6000 6000 6000>
    componentLinearAccScale=<0.4 0.4 0.4>
    componentLinearVelScale=<0.5 0.5 0.5>
    gravityOverride {z=-980}
    physicsBodyDefinitions [{
      boundBone {boneName=dyn_scarfBottom_03}
      boxExtents=<20 20 20>
      constraintSetup {angularLimitsMax=<15 8 0>, angularLimitsMin=<-15 -8 -15>, angularTarget=<1 0 0>}
      localJointOffset=<4 0 0>
     }]
    simSpaceSettings {maxAngularAcceleration=40, maxAngularVelocity=8, simSpaceAngularAlpha=1}
   }
   n13 = RigidBody!(ComponentPose=@n12, Alpha=0.8, OverrideWorldGravity:Vector="0,0,0", ExternalForce) -> (Pose) {
    bOverrideWorldGravity=true
    overridePhysicsAsset=&/F_MED_ArcanaAgate_RightSkirtPhysics
    simulationSpace=BaseBoneSpace
    alphaBoolBlend {blendInTime=2, blendOption=Linear, blendOutTime=0.5}
    alphaScaleBiasClamp {bMapRange=true}
    baseBoneRef {boneName=thigh_r}
    componentAppliedLinearAccClamp=<4000 4000 4000>
    simSpaceSettings {
     dampingAlpha=0
     maxAngularAcceleration=0.2
     maxAngularVelocity=0.8
     maxLinearAcceleration=900
     maxLinearVelocity=300
     velocityScaleZ=0.6
     worldAlpha=0.8
     externalLinearDragV=<0.2 0.2 0.2>
    }
   }
   n14 = RigidBody(ComponentPose=@n13, Alpha=0.8, OverrideWorldGravity="0,0,0") {
    bOverrideWorldGravity=true
    overridePhysicsAsset=&/F_MED_ArcanaAgate_LeftSkirtPhysics
    simulationSpace=BaseBoneSpace
    alphaBoolBlend {blendInTime=2, blendOption=Linear, blendOutTime=0.5}
    alphaScaleBiasClamp {bMapRange=true}
    baseBoneRef {boneName=thigh_l}
    componentAppliedLinearAccClamp=<4000 4000 4000>
    simSpaceSettings {
     dampingAlpha=0
     maxAngularAcceleration=0.2
     maxAngularVelocity=0.8
     maxLinearAcceleration=900
     maxLinearVelocity=300
     velocityScaleZ=0.6
     worldAlpha=0.8
     externalLinearDragV=<0.2 0.2 0.2>
    }
   }
   n15 = "Constraint - dyn_skirtFront_02 (RR)"!(ComponentPose=@n14, ConstraintWeights_0=0.5, ConstraintWeights_1:float=0.5) -> (Pose) {
    alpha=0.5
    alphaBoolBlend {blendOption=Linear}
    boneToModify {boneName=dyn_skirtFront_02}
    constraintSetup [
     {transformType=Rotation, targetBone {boneName=dyn_skirtFrontSide_02_l}}
     {transformType=Rotation, targetBone {boneName=dyn_skirtFrontSide_02_r}}
    ]
    constraintWeights [1,1]
   }
   n16 = "Constraint - dyn_skirtFront_03 (RR)"!(ComponentPose=@n15, ConstraintWeights_0=0.5, ConstraintWeights_1:float=0.5) -> (Pose) {
    alpha=0.5
    alphaBoolBlend {blendOption=Linear}
    boneToModify {boneName=dyn_skirtFront_03}
    constraintSetup [
     {transformType=Rotation, targetBone {boneName=dyn_skirtFrontSide_03_l}}
     {transformType=Rotation, targetBone {boneName=dyn_skirtFrontSide_03_r}}
    ]
    constraintWeights [1,1]
   }
   n17 = "Constraint - dyn_skirtBack_02 (RR)"!(ComponentPose=@n16, ConstraintWeights_0=0.5, ConstraintWeights_1:float=0.5) -> (Pose) {
    alpha=0.5
    alphaBoolBlend {blendOption=Linear}
    boneToModify {boneName=dyn_skirtBack_02}
    constraintSetup [
     {transformType=Rotation, targetBone {boneName=dyn_skirtBackSide_02_l}}
     {transformType=Rotation, targetBone {boneName=dyn_skirtBackSide_02_r}}
    ]
    constraintWeights [1,1]
   }
   n18 = "Constraint - dyn_skirtBack_03 (RR)"!(ComponentPose=@n17, ConstraintWeights_0=0.5, ConstraintWeights_1:float=0.5) -> (Pose) {
    alpha=0.5
    alphaBoolBlend {blendOption=Linear}
    boneToModify {boneName=dyn_skirtBack_03}
    constraintSetup [
     {transformType=Rotation, targetBone {boneName=dyn_skirtBackSide_03_l}}
     {transformType=Rotation, targetBone {boneName=dyn_skirtBackSide_03_r}}
    ]
    constraintWeights [1,1]
   }
   n19 = "Transform (Modify) Bone - Bone: deform_glute_l"!(ComponentPose=@n18, bAlphaBoolEnabled:bool=True, Translation="13.387,-1.874,7.577", Rotation="0,0,0", Scale="1,1,1") -> (Pose) {
    alphaInputType=Bool
    translationMode=BMM_Additive
    translationSpace=BCS_BoneSpace
    alphaBoolBlend {blendOption=Linear}
    boneToModify {boneName=deform_glute_l}
    translation=<13.387 -1.874 7.577>
   }
   n20 = "Transform (Modify) Bone - Bone: deform_glute_r"!(ComponentPose=@n19, bAlphaBoolEnabled:bool=True, Translation="-10.548,-0.384,-6.779", Rotation="0,0,0", Scale="1,1,1") -> (Pose) {
    alphaInputType=Bool
    translationMode=BMM_Additive
    translationSpace=BCS_BoneSpace
    alphaBoolBlend {blendOption=Linear}
    boneToModify {boneName=deform_glute_r}
    translation=<-10.548 -0.384 -6.779>
   }
   n21 = "Transform (Modify) Bone - Bone: deform_glute_r"!(ComponentPose=@n20, Alpha:float=1, Translation="-1.574,-2.245,-2.373", Rotation="0,0,0", Scale="1,1,1") -> (Pose) {
    rotationMode=BMM_Additive
    rotationSpace=BCS_BoneSpace
    translationMode=BMM_Additive
    translationSpace=BCS_BoneSpace
    alphaBoolBlend {blendOption=Linear}
    boneToModify {boneName=deform_glute_r}
    translation=<-1.574 -2.245 -2.373>
   }
   n22 = "Transform (Modify) Bone - Bone: deform_glute_l"!(ComponentPose=@n21, Alpha:float=1, Translation="2.771,2.298,1.533", Rotation="0,0,0", Scale="1,1,1") -> (Pose) {
    rotationMode=BMM_Additive
    rotationSpace=BCS_BoneSpace
    translationMode=BMM_Additive
    translationSpace=BCS_BoneSpace
    alphaBoolBlend {blendOption=Linear}
    boneToModify {boneName=deform_glute_l}
    translation=<2.771 2.298 1.533>
   }
   n23 = Get isUpright() -> (isUpright:bool)
   n24 = "Transform (Modify) Bone - Bone: base_skirtFrontSide_l"!(ComponentPose=@n22, bAlphaBoolEnabled:bool=@n23, Translation="0,0,0", Rotation="0,-6,0", Scale="1,1,1") -> (Pose) {
    alphaInputType=Bool
    rotationMode=BMM_Additive
    rotationSpace=BCS_BoneSpace
    translationMode=BMM_Additive
    translationSpace=BCS_BoneSpace
    alphaBoolBlend {blendInTime=1, blendOption=Linear, blendOutTime=1}
    boneToModify {boneName=base_skirtFrontSide_l}
    rotation {yaw=-6}
   }
   n25 = "Transform (Modify) Bone - Bone: base_skirtFrontSide_r"!(ComponentPose=@n24, bAlphaBoolEnabled:bool=@n23, Translation="0,0,0", Rotation="0.157,-7,0.377", Scale="1,1,1") -> (Pose) {
    alphaInputType=Bool
    rotationMode=BMM_Additive
    rotationSpace=BCS_BoneSpace
    translationMode=BMM_Additive
    translationSpace=BCS_BoneSpace
    alphaBoolBlend {blendInTime=1, blendOption=Linear, blendOutTime=1}
    boneToModify {boneName=base_skirtFrontSide_r}
    rotation=<rot 0.157 -7 0.377>
   }
   n26 = "Transform (Modify) Bone - Bone: base_skirtFront"!(ComponentPose=@n25, bAlphaBoolEnabled:bool=@n23, Translation="0,0,0", Rotation="0,-12.61,0", Scale="1,1,1") -> (Pose) {
    alphaInputType=Bool
    rotationMode=BMM_Additive
    rotationSpace=BCS_BoneSpace
    translationMode=BMM_Additive
    translationSpace=BCS_BoneSpace
    alphaBoolBlend {blendInTime=1, blendOption=Linear, blendOutTime=1}
    boneToModify {boneName=base_skirtFront}
    rotation {yaw=-12.61}
   }
   n27 = "Transform (Modify) Bone - Bone: dyn_skirtBack_02"!(ComponentPose=@n26, bAlphaBoolEnabled:bool=True, Translation="0.075,1.798,-0.304", Rotation="0.008,-9.408,-0.010", Scale="1,1,1") -> (Pose) {
    alphaInputType=Bool
    rotationMode=BMM_Additive
    rotationSpace=BCS_BoneSpace
    translationMode=BMM_Additive
    translationSpace=BCS_BoneSpace
    alphaBoolBlend {blendInTime=0.1, blendOption=Linear, blendOutTime=0.3}
    boneToModify {boneName=dyn_skirtBack_02}
    rotation=<rot 0.008 -9.4079 -0.01>
    translation=<0.075 1.798 -0.304>
   }
   n28 = "Transform (Modify) Bone - Bone: dyn_skirtBackSide_02_l"!(ComponentPose=@n27, bAlphaBoolEnabled:bool=True, Translation="0.845,3.880,-2.575", Rotation="0.000,-17.684,-0.001", Scale="1,1,1") -> (Pose) {
    alphaInputType=Bool
    rotationMode=BMM_Additive
    rotationSpace=BCS_BoneSpace
    translationMode=BMM_Additive
    translationSpace=BCS_BoneSpace
    alphaBoolBlend {blendInTime=0.1, blendOption=Linear, blendOutTime=0.3}
    boneToModify {boneName=dyn_skirtBackSide_02_l}
    rotation=<rot 0.0001 -17.6839 -0.0011>
    translation=<0.845 3.88 -2.575>
   }
   n29 = "Transform (Modify) Bone - Bone: dyn_skirtBackSide_02_r"!(ComponentPose=@n28, bAlphaBoolEnabled:bool=True, Translation="0.484,2.372,0.016", Rotation="0.001,-17.280,-0.004", Scale="1,1,1") -> (Pose) {
    alphaInputType=Bool
    rotationMode=BMM_Additive
    rotationSpace=BCS_BoneSpace
    translationMode=BMM_Additive
    translationSpace=BCS_BoneSpace
    alphaBoolBlend {blendInTime=0.1, blendOption=Linear, blendOutTime=0.3}
    boneToModify {boneName=dyn_skirtBackSide_02_r}
    rotation=<rot 0.001 -17.28 -0.004>
    translation=<0.484 2.372 0.016>
   }
   n30 = "Transform (Modify) Bone - Bone: base_skirtFrontSide_l"(ComponentPose=@n29, bAlphaBoolEnabled=false, Translation="0,0,0", Rotation="-4.384,-14.040,0.634", Scale="1,1,1") {
    alphaInputType=Bool
    rotationMode=BMM_Additive
    rotationSpace=BCS_BoneSpace
    translationMode=BMM_Additive
    translationSpace=BCS_BoneSpace
    alphaBoolBlend {blendInTime=1, blendOption=Linear, blendOutTime=1}
    boneToModify {boneName=base_skirtFrontSide_l}
    rotation=<rot -4.384 -14.0401 0.634>
   }
   n31 = "Transform (Modify) Bone - Bone: base_skirtFrontSide_r"(ComponentPose=@n30, bAlphaBoolEnabled=false, Translation="0,0,0", Rotation="-9.223,-15.927,1.995", Scale="1,1,1") {
    alphaInputType=Bool
    rotationMode=BMM_Additive
    rotationSpace=BCS_BoneSpace
    translationMode=BMM_Additive
    translationSpace=BCS_BoneSpace
    alphaBoolBlend {blendInTime=1, blendOption=Linear, blendOutTime=1}
    boneToModify {boneName=base_skirtFrontSide_r}
    rotation=<rot -9.2231 -15.9271 1.9951>
   }
   n32 = "Transform (Modify) Bone - Bone: base_skirtFront"(ComponentPose=@n31, bAlphaBoolEnabled=Get bIsTacticalSprinting(), Translation="0,0,0", Rotation="-0.026,-12.608,-0.113", Scale="1,1,1") {
    alphaInputType=Bool
    rotationMode=BMM_Additive
    rotationSpace=BCS_BoneSpace
    translationMode=BMM_Additive
    translationSpace=BCS_BoneSpace
    alphaBoolBlend {blendInTime=1, blendOption=Linear, blendOutTime=1}
    boneToModify {boneName=base_skirtFront}
    rotation=<rot -0.026 -12.608 -0.113>
   }
   Output Pose(Result=Component To Local(ComponentPose=@n32)) {layerGroup=None}
  ]}
 EventGraph {nodes [
   Event BlueprintUpdateAnimation()
   > n1 = Sequence!(execute) -> (then_0, then_1, then_2:Exec)
   > n2 = "Gravity Override State Function (Parameter Asset) (Nativized)"(
    CharacterPartAnimInstance=Self-Reference()
    GameTimeInSeconds=GetGameTimeInSeconds()
    DebugJointName=spine_05
    Parameters=&/F_MED_ArcanaAgate_Parameters
   )
   > Set Gravity Override(Gravity Override=@n2)
   n3 = Get bIsOnGrindRail()
   n4 = Get bIsCrouching()
   n5 = Get bIsBaseballSliding()
   n6 = Get bIsSkyDiving()
   Set isUpright(execute=@n1.then_1, isUpright:bool=NOT Boolean(A=OR Boolean!(A=@n4, B=@n3, C:bool=@n6, D:bool=@n5) -> (ReturnValue))) -> (Output_Get:bool)
   > n7 = Branch(Condition=Get bIsBaseballSliding())
   > Set skirtRBANAlpha(skirtRBANAlpha:double=0.2) -> (Output_Get:double)
   Set skirtRBANAlpha(execute=@n7.else, skirtRBANAlpha=0.9)
   n8 = Event OnCharacterCustomizationLoaded()
   n9 = Cast To FortPlayerAnimInstance(execute=[@n8.then, @n11.then], Object=GetAnimInstance(self=GetOwnerSkeletalMeshComponent()))
   > Set OneHandedPickaxe(OneHandedPickaxe:double=0) -> (Output_Get:double)
   > Set ArcanaAgateAxeEquipped(ArcanaAgateAxeEquipped:bool=PickaxeItemDefinitionHasAnyMetaTag(Tags=Make Array("[0]"=(TagName="Cosmetics.Metatags.ArcanaAgate")), bRequiresEquipped=true)) -> (Output_Get:bool)
   > n10 = Switch on EFortWeaponCoreAnimation!(execute, Selection:EFortWeaponCoreAnimation Enum=Get WeaponCoreAnim(self=@n9)) -> (
    Melee:Exec
    Pistol:Exec
    Shotgun:Exec
    PaperBlueprint:Exec
    Rifle:Exec
    MeleeOneHand:Exec
    MachinePistol:Exec
    RocketLauncher:Exec
    GrenadeLauncher:Exec
    GoingCommando:Exec
    AssaultRifle:Exec
    TacticalShotgun:Exec
    SniperRifle:Exec
    TrapPlacement:Exec
    ShoulderLauncher:Exec
    AbilityDecoTool:Exec
    Crossbow:Exec
    C4:Exec
    RemoteControl:Exec
    DualWield:Exec
    AR_BullPup:Exec
    AR_ForwardGrip:Exec
    MedPackPaddles:Exec
    SMG_P90:Exec
    AR_DrumGun:Exec
    Consumable_Small:Exec
    Consumable_Large:Exec
    Balloon:Exec
    MountedTurret:Exec
    CreativeTool:Exec
    ExplosiveBow:Exec
    AshtonIndigo:Exec
    AshtonChicago:Exec
    MeleeDualWield:Exec
    Unarmed:Exec
   )
   Set OneHandedPickaxe(execute=@n10.MeleeOneHand, OneHandedPickaxe=1)
   n11 = Event OnWeaponChanged()
  ]}
]
variables [
 ArcanaAgateAxeEquipped {type=boolean}
 Gravity Override {type=Vector}
 isUpright {type=boolean}
 OneHandedPickaxe {type=number}
 skirtRBANAlpha {type=number}
]
