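# Control Rig for the ArcanaAgate female face accessory that poses back and front-side hair bones around head, arm and gameplay poses.
# It smooths crouch state, then branches on WearingBackblingAndNoRBAN between ConstraintsBackpackNoRBAN and Hair functions.
# The backbling path parent-constrains hair to weighted head, neck, spine and arm bones before applying correctives.
# Both paths layer transform offsets weighted by spherical pose readers and state variables like glider and emote.
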ControlRigBlueprint /BRCosmetics/Characters/Player/Female/Medium/Bodies/F_MED_ArcanaAgate/Meshes/Parts/F_MED_ArcanaAgate_FaceAcc_CtrlRig
graphs [
 Multiply_11_ContainedGraph {nodes [
   n1 = RigVMFunctionEntryNode()
   RigVMFunctionReturnNode(Result=&RigVMFunction_MathDoubleMul(B=@n1.C, A=&RigVMFunction_MathDoubleMul(A=@n1.A, B=@n1.B).Result).Result)
  ]}
 Multiply_7_ContainedGraph {nodes [
   n1 = RigVMFunctionEntryNode()
   n2 = &RigVMFunction_MathDoubleMul(A=@n1.A, B=@n1.B)
   RigVMFunctionReturnNode(Result=&RigVMFunction_MathDoubleMul(B=@n1.D, A=&RigVMFunction_MathDoubleMul(B=@n1.C, A=@n2.Result).Result).Result)
  ]}
 Multiply_7_ContainedGraph {nodes [
   n1 = RigVMFunctionEntryNode()
   n2 = &RigVMFunction_MathDoubleMul(A=@n1.A, B=@n1.B)
   RigVMFunctionReturnNode(Result=&RigVMFunction_MathDoubleMul(B=@n1.D, A=&RigVMFunction_MathDoubleMul(B=@n1.C, A=@n2.Result).Result).Result)
  ]}
 New Function_ContainedGraph {nodes [
   n1 = RigVMFunctionEntryNode()
   RigVMFunction_MathQuaternionToEuler = &RigVMFunction_MathQuaternionToEuler(
    RotationOrder=ZXY
    Value=&RigUnit_GetRelativeTransformForItem(
     Child=(Type=Bone,Name="None")
     Parent=(Type=Bone,Name="None")
     bChildInitial=False
     bParentInitial=False
     Child.Name=@n1.Child
     Parent.Name=@n1.Parent
    ).RelativeTransform.Rotation
   )
   Remap = &RigVMFunction_MathDoubleRemap(
    TargetMaximum=1
    TargetMinimum=0
    Value=@RigVMFunction_MathQuaternionToEuler.Result.Z
    SourceMinimum=@n1.SourceMin
    SourceMaximum=@n1.SourceMax
   )
   > RigVMFunctionReturnNode(Delta Z=@Remap.Result)
  ]}
 New Function_ContainedGraph {nodes [
   n1 = RigVMFunctionEntryNode()
   RigVMFunction_MathQuaternionToEuler = &RigVMFunction_MathQuaternionToEuler(
    RotationOrder=ZYX
    Value=&RigUnit_GetRelativeTransformForItem(
     Child=(Type=Bone,Name="None")
     Parent=(Type=Bone,Name="None")
     bChildInitial=False
     bParentInitial=False
     Child.Name=@n1.Child
     Parent.Name=@n1.Parent
    ).RelativeTransform.Rotation
   )
   Remap = &RigVMFunction_MathDoubleRemap(
    TargetMaximum=1
    TargetMinimum=0
    Value=@RigVMFunction_MathQuaternionToEuler.Result.Y
    SourceMinimum=@n1.SourceMin
    SourceMaximum=@n1.SourceMax
   )
   > RigVMFunctionReturnNode(Delta Y=@Remap.Result)
  ]}
 New Function_ContainedGraph {nodes [
   RigVMFunctionEntryNode()
   > RigVMFunction_Sequence = RigVMAggregateNode()
   n1 = &RigUnit_SphericalPoseReader(
    ActiveRegionScaleFactors=(PositiveWidth=0.3,NegativeWidth=0.3,PositiveHeight=0.6,NegativeHeight=1)
    ActiveRegionSize=0.15
    Debug=(bDrawDebug=false,bDraw2D=false,bDrawLocalAxes=true,DebugScale=25,DebugSegments=20,DebugThickness=0.25)
    DriverAxis=(X=1,Y=0,Z=0)
    DriverItem=(Type=Bone,Name="head")
    FalloffRegionScaleFactors=(PositiveWidth=0.8,NegativeWidth=1,PositiveHeight=1.3,NegativeHeight=1)
    FalloffSize=0.4
    FlipHeightScaling=false
    FlipWidthScaling=true
    OptionalParentItem=(Type=Bone,Name="spine_04")
    RotationOffset=(X=0,Y=-62,Z=-90)
    ExecutePin=@RigVMFunction_Sequence.A
   )
   n2 = &RigVMFunction_AlphaInterp(
    InterpSpeedDecreasing=10
    InterpSpeedIncreasing=10
    Scale=1
    bClampResult=true
    Value=&RigVMFunction_MathBoolToFloat(Value=RigVMVariableNode() {variable=Floating_CR}.Value).Result
   )
   > &RigUnit_ModifyTransforms(
    ItemToModify="((Item=(Type=Bone,Name=\"base_hairB_C_01\"),Transform=(Rotation=(X=-0.0052,Y=-0.0173,Z=0.2855,W=-0.9582),Translation=(X=-1.5119,Y=0.9887,Z=0.109),Scale3D=(X=1,Y=1,Z=1))),(Item=(Type=Bone,Name=\"base_hairB_L_01\"),Transform=(Rotation=(X=-0.0042,Y=-0.0215,Z=-0.1 ...(truncated)"
    WeightMinimum=0
    Weight=&RigVMFunction_MathDoubleMul(
     B=&RigVMFunction_MathDoubleRemap(SourceMaximum=1, SourceMinimum=0, TargetMaximum=0, TargetMinimum=1, Value=@n2.Result).Result
     A=@n1.OutputParam
    ).Result
   )
   > &RigUnit_ModifyTransforms(
    ItemToModify="((Item=(Type=Bone,Name=\"dyn_hairB_FR_02\"),Transform=(Rotation=(X=0,Y=0,Z=0,W=-1),Translation=(X=1.9761,Y=0.4771,Z=0),Scale3D=(X=1,Y=1,Z=1))),(Item=(Type=Bone,Name=\"base_hairB_R_01\"),Transform=(Rotation=(X=-0.0088,Y=0.2722,Z=-0.0221,W=0.9619),Translation=(X ...(truncated)"
    WeightMinimum=0
    Weight=&RigVMFunction_MathDoubleMul(A=@n1.OutputParam, B=RigVMVariableNode() {variable=Pistol_CR}.Value).Result
   )
   > n3 = &RigUnit_SphericalPoseReader(
    ActiveRegionScaleFactors=(PositiveWidth=1.2,NegativeWidth=1,PositiveHeight=1,NegativeHeight=1)
    ActiveRegionSize=0.05
    Debug=(bDrawDebug=false,bDraw2D=false,bDrawLocalAxes=true,DebugScale=25,DebugSegments=20,DebugThickness=0.25)
    DriverAxis=(X=-1,Y=0,Z=0)
    DriverItem=(Type=Bone,Name="upperarm_r")
    FalloffRegionScaleFactors=(PositiveWidth=2,NegativeWidth=1.6,PositiveHeight=1,NegativeHeight=1)
    FalloffSize=0.15
    FlipHeightScaling=false
    FlipWidthScaling=true
    OptionalParentItem=(Type=Bone,Name="spine_05")
    RotationOffset=(X=454.3811,Y=-139.5663,Z=198.5291)
   )
   > &RigUnit_ModifyTransforms(
    ItemToModify="((Item=(Type=Bone,Name=\"base_hairB_R_01\"),Transform=(Rotation=(X=0.1485,Y=-0.0269,Z=0.1737,W=-0.9732),Translation=(X=-2.2498,Y=1.4495,Z=0.8092),Scale3D=(X=1,Y=1,Z=1))),(Item=(Type=Bone,Name=\"dyn_hairB_FR_02\"),Transform=(Rotation=(X=0,Y=0,Z=0,W=-1),Translat ...(truncated)"
    WeightMinimum=0
    Weight=@n3.OutputParam
   )
   > n4 = &RigUnit_SphericalPoseReader(
    ActiveRegionScaleFactors=(PositiveWidth=1,NegativeWidth=1,PositiveHeight=1,NegativeHeight=1.5)
    ActiveRegionSize=0.05
    Debug=(bDrawDebug=false,bDraw2D=false,bDrawLocalAxes=true,DebugScale=25,DebugSegments=20,DebugThickness=0.25)
    DriverAxis=(X=1,Y=0.5,Z=0)
    DriverItem=(Type=Bone,Name="head")
    FalloffRegionScaleFactors=(PositiveWidth=1,NegativeWidth=1,PositiveHeight=1,NegativeHeight=2)
    FalloffSize=0.15
    FlipHeightScaling=false
    FlipWidthScaling=false
    OptionalParentItem=(Type=Bone,Name="spine_04")
    RotationOffset=(X=114.1051,Y=-69.8745,Z=-57.838)
   )
   n5 = RigVMVariableNode() {variable=Pistol_CR}
   n6 = &RigVMFunction_MathDoubleRemap(SourceMaximum=1, SourceMinimum=0, TargetMaximum=0.2, TargetMinimum=1, Value=@n5.Value)
   n7 = &RigVMFunction_AlphaInterp(
    InterpSpeedDecreasing=10
    InterpSpeedIncreasing=10
    Scale=1
    bClampResult=true
    Value=&RigVMFunction_MathBoolToFloat(Value=RigVMVariableNode() {variable=isEmoting_CR}.Value).Result
   )
   n8 = &RigVMFunction_MathDoubleRemap(SourceMaximum=1, SourceMinimum=0, TargetMaximum=0, TargetMinimum=1, Value=@n7.Result)
   n9 = &RigVMFunction_MathBoolToFloat(Value=RigVMVariableNode() {variable=isFrontEnd_CR}.Value)
   n10 = &RigVMFunction_MathDoubleRemap(SourceMaximum=1, SourceMinimum=0, TargetMaximum=0, TargetMinimum=1, Value=@n9.Result)
   Multiply_3 = RigVMAggregateNode(B=@n6.Result, C=@n8.Result, D=@n10.Result, A=@n4.OutputParam)
   > &RigUnit_ModifyTransforms(
    ItemToModify="((Item=(Type=Bone,Name=\"base_hairB_R_01\"),Transform=(Rotation=(X=0,Y=0,Z=0,W=1),Translation=(X=-0.6815,Y=0,Z=0),Scale3D=(X=1,Y=1,Z=1))),(Item=(Type=Bone,Name=\"dyn_hairB_FR_02\"),Transform=(Rotation=(X=0,Y=0,Z=0,W=1),Translation=(X=0,Y=0,Z=0),Scale3D=(X=1,Y= ...(truncated)"
    WeightMinimum=0
    Weight=@Multiply_3.Result
   )
   > &RigUnit_ModifyTransforms(
    ItemToModify=((Item=(Type=Bone,Name="base_hairB_R_01"),Transform=(Rotation=(X=0,Y=0,Z=0,W=1),Translation=(X=-0.7938,Y=0,Z=0),Scale3D=(X=1,Y=1,Z=1))),(Item=(Type=Bone,Name="None"),Transform=(Rotation=(X=0,Y=0,Z=0,W=1),Translation=(X=0,Y=0,Z=0),Scale3D=(X=1,Y=1,Z=1))))
    WeightMinimum=0
    Weight=&RigVMFunction_MathDoubleMul(A=@n4.OutputParam, B=@n5.Value).Result
   )
   > n11 = &RigUnit_SphericalPoseReader(
    ActiveRegionScaleFactors=(PositiveWidth=0.3,NegativeWidth=0.3,PositiveHeight=0.6,NegativeHeight=1)
    ActiveRegionSize=0.15
    Debug=(bDrawDebug=false,bDraw2D=false,bDrawLocalAxes=true,DebugScale=25,DebugSegments=20,DebugThickness=0.25)
    DriverAxis=(X=1,Y=0,Z=0)
    DriverItem=(Type=Bone,Name="head")
    FalloffRegionScaleFactors=(PositiveWidth=0.8,NegativeWidth=1,PositiveHeight=1.3,NegativeHeight=1)
    FalloffSize=0.2
    FlipHeightScaling=false
    FlipWidthScaling=true
    OptionalParentItem=(Type=Bone,Name="spine_04")
    RotationOffset=(X=36.8185,Y=-22.943,Z=-90)
   )
   Multiply_7 = RigVMAggregateNode(B=@n6.Result, C=@n8.Result, D=@n10.Result, A=@n11.OutputParam)
   > &RigUnit_ModifyTransforms(
    ItemToModify="((Item=(Type=Bone,Name=\"base_hairB_C_01\"),Transform=(Rotation=(X=0,Y=-0.0836,Z=0,W=-0.9965),Translation=(X=-0.9295,Y=-2.0013,Z=-1.9505),Scale3D=(X=1,Y=1,Z=1))),(Item=(Type=Bone,Name=\"base_hairB_L_01\"),Transform=(Rotation=(X=0,Y=-0.0268,Z=0,W=-0.9996),Trans ...(truncated)"
    WeightMinimum=0
    Weight=@Multiply_7.Result
   )
   > n12 = &RigUnit_SphericalPoseReader(
    ActiveRegionScaleFactors=(PositiveWidth=1,NegativeWidth=1,PositiveHeight=1,NegativeHeight=1.5)
    ActiveRegionSize=0.05
    Debug=(bDrawDebug=false,bDraw2D=false,bDrawLocalAxes=true,DebugScale=25,DebugSegments=20,DebugThickness=0.25)
    DriverAxis=(X=-1,Y=0,Z=0)
    DriverItem=(Type=Bone,Name="clavicle_r")
    FalloffRegionScaleFactors=(PositiveWidth=1,NegativeWidth=1,PositiveHeight=1,NegativeHeight=1)
    FalloffSize=0.15
    FlipHeightScaling=false
    FlipWidthScaling=true
    OptionalParentItem=(Type=Bone,Name="spine_04")
    RotationOffset=(X=1.3554,Y=107.9066,Z=0)
   )
   Multiply_10 = RigVMAggregateNode(C=@n6.Result, B=@n4.OutputParam, A=@n12.OutputParam)
   > &RigUnit_ModifyTransforms(
    ItemToModify="((Item=(Type=Bone,Name=\"base_hairB_R_01\"),Transform=(Rotation=(X=0,Y=0,Z=0,W=-1),Translation=(X=-1.9047,Y=-1.8213,Z=-1.8705),Scale3D=(X=1,Y=1,Z=1))),(Item=(Type=Bone,Name=\"dyn_hairB_FR_02\"),Transform=(Rotation=(X=0,Y=0,Z=0,W=1),Translation=(X=0,Y=0,Z=0),Sc ...(truncated)"
    WeightMinimum=0
    Weight=@Multiply_10.Result
   )
   RigVMFunctionReturnNode(ExecuteContext=@RigVMFunction_Sequence.E)
   &RigUnit_ModifyTransforms(
    ItemToModify="((Item=(Type=Bone,Name=\"base_hairB_C_01\"),Transform=(Rotation=(X=0,Y=0,Z=0.1706,W=-0.9853),Translation=(X=0,Y=0,Z=0),Scale3D=(X=1,Y=1,Z=1))),(Item=(Type=Bone,Name=\"base_hairB_R_01\"),Transform=(Rotation=(X=0.2189,Y=-0.1518,Z=0.1308,W=-0.9549),Translation=(X ...(truncated)"
    WeightMinimum=0
    Weight=RigVMVariableNode() {variable=Glider_CR}.Value
    ExecutePin=@RigVMFunction_Sequence.C
   )
   n13 = &RigVMFunction_AlphaInterp(
    InterpSpeedDecreasing=10
    InterpSpeedIncreasing=10
    Scale=1
    bClampResult=true
    Value=&RigVMFunction_MathBoolToFloat(Value=RigVMVariableNode() {variable=isDBNO_CR}.Value).Result
   )
   > &RigUnit_ModifyTransforms(
    ItemToModify="((Item=(Type=Bone,Name=\"base_hairB_C_01\"),Transform=(Rotation=(X=0,Y=0,Z=0,W=1),Translation=(X=0,Y=0,Z=0),Scale3D=(X=1,Y=1,Z=1))),(Item=(Type=Bone,Name=\"base_hairB_R_01\"),Transform=(Rotation=(X=0,Y=0,Z=0,W=1),Translation=(X=0,Y=0,Z=0),Scale3D=(X=1,Y=1,Z=1) ...(truncated)"
    WeightMinimum=0
    Weight=@n13.Result
   )
   > n14 = &RigUnit_SphericalPoseReader(
    ActiveRegionScaleFactors=(PositiveWidth=1,NegativeWidth=1,PositiveHeight=1,NegativeHeight=1)
    ActiveRegionSize=0.1
    Debug=(bDrawDebug=false,bDraw2D=false,bDrawLocalAxes=true,DebugScale=25,DebugSegments=20,DebugThickness=0.25)
    DriverAxis=(X=1,Y=0,Z=0)
    DriverItem=(Type=Bone,Name="head")
    FalloffRegionScaleFactors=(PositiveWidth=0.8,NegativeWidth=1,PositiveHeight=1.3,NegativeHeight=1)
    FalloffSize=0.25
    FlipHeightScaling=false
    FlipWidthScaling=true
    OptionalParentItem=(Type=Bone,Name="spine_04")
    RotationOffset=(X=107.6801,Y=-18.1754,Z=-120.0052)
   )
   n15 = &RigVMFunction_AlphaInterp(
    InterpSpeedDecreasing=10
    InterpSpeedIncreasing=10
    Scale=1
    bClampResult=true
    Value=&RigVMFunction_MathBoolToFloat(Value=RigVMVariableNode() {variable=isOnRail_CR}.Value).Result
   )
   > &RigUnit_ModifyTransforms(
    ItemToModify="((Item=(Type=Bone,Name=\"base_hairB_R_01\"),Transform=(Rotation=(X=0,Y=-0.0375,Z=0,W=0.9993),Translation=(X=-0.1459,Y=0,Z=1.9406),Scale3D=(X=1,Y=1,Z=1))),(Item=(Type=Bone,Name=\"dyn_hairB_FR_02\"),Transform=(Rotation=(X=0,Y=-0.0731,Z=0,W=-0.9973),Translation=( ...(truncated)"
    WeightMinimum=0
    Weight=&RigVMFunction_MathDoubleMul(B=@n15.Result, A=@n14.OutputParam).Result
   )
   n16 = &RigVMFunction_AlphaInterp(
    InterpSpeedDecreasing=10
    InterpSpeedIncreasing=10
    Scale=1
    bClampResult=true
    Value=&RigVMFunction_MathBoolToFloat(Value=RigVMVariableNode() {variable=isAttackingPickaxe}.Value).Result
   )
   > &RigUnit_ModifyTransforms(
    ItemToModify="((Item=(Type=Bone,Name=\"dyn_hairB_FL_02\"),Transform=(Rotation=(X=0.0028,Y=0.0371,Z=0.0754,W=-0.9965),Translation=(X=1.0573,Y=-0.5899,Z=0.2924),Scale3D=(X=1,Y=1,Z=1))),(Item=(Type=Bone,Name=\"base_hairB_L_01\"),Transform=(Rotation=(X=0,Y=0,Z=0,W=1),Translatio ...(truncated)"
    WeightMinimum=0
    Weight=@n16.Result
   )
   &RigUnit_ModifyTransforms(
    ItemToModify="((Item=(Type=Bone,Name=\"base_hairB_C_01\"),Transform=(Rotation=(X=-0.0098,Y=-0.0031,Z=0.3112,W=-0.9503),Translation=(X=-0.3415,Y=-1.7538,Z=0.0132),Scale3D=(X=1,Y=1,Z=1))),(Item=(Type=Bone,Name=\"base_hairB_R_01\"),Transform=(Rotation=(X=-0.0309,Y=-0.0727,Z=0. ...(truncated)"
    WeightMinimum=0
    Weight=@n2.Result
    ExecutePin=@RigVMFunction_Sequence.B
   )
   > &RigUnit_ModifyTransforms(
    ItemToModify="((Item=(Type=Bone,Name=\"dyn_hairB_L_01\"),Transform=(Rotation=(X=0.0559,Y=0,Z=0,W=-0.9984),Translation=(X=0,Y=0,Z=0),Scale3D=(X=1,Y=1,Z=1))),(Item=(Type=Bone,Name=\"base_hairB_R_01\"),Transform=(Rotation=(X=0,Y=0,Z=0,W=1),Translation=(X=0,Y=-2.3286,Z=0),Scale ...(truncated)"
    WeightMinimum=0
    Weight=@n7.Result
   )
   n17 = &RigVMFunction_AlphaInterp(
    InterpSpeedDecreasing=10
    InterpSpeedIncreasing=10
    Scale=1
    bClampResult=true
    Value=RigVMVariableNode() {variable=CelebrationEmoteRBAN_CR}.Value
   )
   > &RigUnit_ModifyTransforms(
    ItemToModify="((Item=(Type=Bone,Name=\"dyn_hairB_FL_02\"),Transform=(Rotation=(X=0,Y=0,Z=-0.0936,W=0.9956),Translation=(X=-0.738,Y=-3.8916,Z=-0.0001),Scale3D=(X=1,Y=1,Z=1))),(Item=(Type=Bone,Name=\"dyn_hairB_FR_02\"),Transform=(Rotation=(X=-0.0362,Y=0.0013,Z=0.0178,W=-0.999 ...(truncated)"
    WeightMinimum=0
    Weight=@n17.Result
   )
   > n18 = &RigUnit_SphericalPoseReader(
    ActiveRegionScaleFactors=(PositiveWidth=1,NegativeWidth=1,PositiveHeight=1,NegativeHeight=1.5)
    ActiveRegionSize=0.1
    Debug=(bDrawDebug=false,bDraw2D=false,bDrawLocalAxes=true,DebugScale=25,DebugSegments=20,DebugThickness=0.25)
    DriverAxis=(X=1,Y=0.5,Z=0)
    DriverItem=(Type=Bone,Name="head")
    FalloffRegionScaleFactors=(PositiveWidth=1,NegativeWidth=1,PositiveHeight=1,NegativeHeight=2)
    FalloffSize=0.25
    FlipHeightScaling=false
    FlipWidthScaling=true
    OptionalParentItem=(Type=Bone,Name="spine_04")
    RotationOffset=(X=106.9754,Y=-66.2566,Z=-68.2639)
   )
   Multiply = &RigVMFunction_MathDoubleMul(A=@n18.OutputParam, B=@n9.Result)
   > &RigUnit_ModifyTransforms(
    ItemToModify="((Item=(Type=Bone,Name=\"base_hairB_R_01\"),Transform=(Rotation=(X=0,Y=0,Z=0,W=1),Translation=(X=0,Y=0,Z=0),Scale3D=(X=1,Y=1,Z=1))),(Item=(Type=Bone,Name=\"dyn_hairB_FR_02\"),Transform=(Rotation=(X=0,Y=0,Z=0,W=-1),Translation=(X=-2.1558,Y=0,Z=0),Scale3D=(X=1,Y ...(truncated)"
    WeightMinimum=0
    Weight=@Multiply.Result
   )
  ]}
 New Function_ContainedGraph {nodes [
   RigVMFunctionEntryNode()
   > RigVMFunction_Sequence = RigVMAggregateNode()
   &RigUnit_ParentConstraint(
    AdvancedSettings=(RotationOrderForFilter=XZY)
    Child=(Type=Bone,Name="dyn_hairB_R_02")
    Parents=((Item=(Type=Bone,Name="head"),Weight=0.064),(Item=(Type=Bone,Name="spine_05"),Weight=0.163),(Item=(Type=Bone,Name="neck_01"),Weight=0.355),(Item=(Type=Bone,Name="neck_02"),Weight=0.418))
    Weight=1
    ExecutePin=@RigVMFunction_Sequence.A
   )
   > &RigUnit_ParentConstraint(
    AdvancedSettings=(RotationOrderForFilter=XZY)
    Child=(Type=Bone,Name="dyn_hairB_R_03")
    Parents=((Item=(Type=Bone,Name="upperarm_r"),Weight=0.0139),(Item=(Type=Bone,Name="neck_01"),Weight=0.0148),(Item=(Type=Bone,Name="clavicle_r"),Weight=0.3586),(Item=(Type=Bone,Name="spine_05"),Weight=0.6127))
    Weight=1
   )
   > &RigUnit_ParentConstraint(
    AdvancedSettings=(RotationOrderForFilter=XZY)
    Child=(Type=Bone,Name="dyn_hairB_FR_02")
    Parents=((Item=(Type=Bone,Name="clavicle_r"),Weight=0.003),(Item=(Type=Bone,Name="head"),Weight=0.076),(Item=(Type=Bone,Name="spine_05"),Weight=0.171),(Item=(Type=Bone,Name="neck_01"),Weight=0.356),(Item=(Type=Bone,Name="neck_02"),Weight=0.394))
    Weight=1
   )
   > &RigUnit_ParentConstraint(
    AdvancedSettings=(RotationOrderForFilter=XZY)
    Child=(Type=Bone,Name="dyn_hairB_C_02")
    Parents=((Item=(Type=Bone,Name="head"),Weight=0.0849),(Item=(Type=Bone,Name="spine_05"),Weight=0.1289),(Item=(Type=Bone,Name="neck_01"),Weight=0.3796),(Item=(Type=Bone,Name="neck_02"),Weight=0.4066))
    Weight=1
   )
   > &RigUnit_ParentConstraint(
    AdvancedSettings=(RotationOrderForFilter=XZY)
    Child=(Type=Bone,Name="dyn_hairB_C_03")
    Parents=((Item=(Type=Bone,Name="neck_02"),Weight=0.005),(Item=(Type=Bone,Name="clavicle_l"),Weight=0.034),(Item=(Type=Bone,Name="clavicle_r"),Weight=0.034),(Item=(Type=Bone,Name="neck_01"),Weight=0.079),(Item=(Type=Bone,Name="spine_05"),Weight=0.848))
    Weight=1
   )
   > &RigUnit_ParentConstraint(
    AdvancedSettings=(RotationOrderForFilter=XZY)
    Child=(Type=Bone,Name="dyn_hairB_L_02")
    Parents=((Item=(Type=Bone,Name="head"),Weight=0.0571),(Item=(Type=Bone,Name="spine_05"),Weight=0.1662),(Item=(Type=Bone,Name="neck_01"),Weight=0.3594),(Item=(Type=Bone,Name="neck_02"),Weight=0.4174))
    Weight=1
   )
   > &RigUnit_ParentConstraint(
    AdvancedSettings=(RotationOrderForFilter=XZY)
    Child=(Type=Bone,Name="dyn_hairB_L_03")
    Parents=((Item=(Type=Bone,Name="upperarm_l"),Weight=0.0139),(Item=(Type=Bone,Name="neck_01"),Weight=0.0148),(Item=(Type=Bone,Name="clavicle_l"),Weight=0.3586),(Item=(Type=Bone,Name="spine_05"),Weight=0.6127))
    Weight=1
   )
   > &RigUnit_ParentConstraint(
    AdvancedSettings=(RotationOrderForFilter=XZY)
    Child=(Type=Bone,Name="dyn_hairB_FL_02")
    Parents=((Item=(Type=Bone,Name="clavicle_l"),Weight=0.008),(Item=(Type=Bone,Name="head"),Weight=0.069),(Item=(Type=Bone,Name="spine_05"),Weight=0.19),(Item=(Type=Bone,Name="neck_01"),Weight=0.353),(Item=(Type=Bone,Name="neck_02"),Weight=0.38))
    Weight=1
   )
   n1 = &RigUnit_SphericalPoseReader(
    ActiveRegionScaleFactors=(PositiveWidth=0.5,NegativeWidth=0.5,PositiveHeight=0.6,NegativeHeight=1)
    ActiveRegionSize=0.15
    Debug=(bDrawDebug=false,bDraw2D=false,bDrawLocalAxes=true,DebugScale=25,DebugSegments=20,DebugThickness=0.25)
    DriverAxis=(X=1,Y=0,Z=0)
    DriverItem=(Type=Bone,Name="head")
    FalloffRegionScaleFactors=(PositiveWidth=0.8,NegativeWidth=1,PositiveHeight=0.5,NegativeHeight=1)
    FalloffSize=0.4
    FlipHeightScaling=false
    FlipWidthScaling=true
    OptionalParentItem=(Type=Bone,Name="spine_04")
    RotationOffset=(X=21.5906,Y=-79.6095,Z=-96.3359)
    ExecutePin=@RigVMFunction_Sequence.B
   )
   > &RigUnit_ModifyTransforms(
    ItemToModify="((Item=(Type=Bone,Name=\"base_hairB_C_01\"),Transform=(Rotation=(X=0,Y=0,Z=-0.4056,W=0.914),Translation=(X=0,Y=0,Z=0),Scale3D=(X=1,Y=1,Z=1))),(Item=(Type=Bone,Name=\"base_hairB_L_01\"),Transform=(Rotation=(X=0,Y=0,Z=0.1495,W=-0.9888),Translation=(X=-2.4289,Y=0 ...(truncated)"
    WeightMinimum=0
    Weight=@n1.OutputParam
   )
   > &RigUnit_ModifyTransforms(
    ItemToModify="((Item=(Type=Bone,Name=\"base_hairB_C_01\"),Transform=(Rotation=(X=0,Y=0,Z=0,W=-1),Translation=(X=0,Y=-3.7732,Z=0),Scale3D=(X=1,Y=1,Z=1))),(Item=(Type=Bone,Name=\"base_hairB_L_01\"),Transform=(Rotation=(X=0,Y=0,Z=0,W=1),Translation=(X=0,Y=-2.3426,Z=0),Scale3D= ...(truncated)"
    WeightMinimum=0
    Weight=&RigVMFunction_MathDoubleMul(B=RigVMVariableNode() {variable=Crouching}.Value, A=@n1.OutputParam).Result
   )
   > n2 = &RigUnit_SphericalPoseReader(
    ActiveRegionScaleFactors=(PositiveWidth=1,NegativeWidth=1,PositiveHeight=2.4,NegativeHeight=1.5)
    ActiveRegionSize=0.05
    Debug=(bDrawDebug=false,bDraw2D=false,bDrawLocalAxes=true,DebugScale=25,DebugSegments=20,DebugThickness=0.25)
    DriverAxis=(X=1,Y=0.5,Z=0)
    DriverItem=(Type=Bone,Name="head")
    FalloffRegionScaleFactors=(PositiveWidth=1,NegativeWidth=1,PositiveHeight=1,NegativeHeight=1.6)
    FalloffSize=0.15
    FlipHeightScaling=false
    FlipWidthScaling=true
    OptionalParentItem=(Type=Bone,Name="spine_04")
    RotationOffset=(X=95.6242,Y=-39.8683,Z=-50)
   )
   n3 = &RigVMFunction_AlphaInterp(
    InterpSpeedDecreasing=10
    InterpSpeedIncreasing=10
    Scale=1
    bClampResult=true
    Value=&RigVMFunction_MathBoolToFloat(Value=RigVMVariableNode() {variable=isFrontEnd_CR}.Value).Result
   )
   > &RigUnit_ModifyTransforms(
    ItemToModify="((Item=(Type=Bone,Name=\"base_hairB_R_01\"),Transform=(Rotation=(X=0,Y=0,Z=0,W=-1),Translation=(X=-4.0315,Y=0,Z=0),Scale3D=(X=1,Y=1,Z=1))),(Item=(Type=Bone,Name=\"dyn_hairB_FR_02\"),Transform=(Rotation=(X=0,Y=0,Z=0.1581,W=-0.9874),Translation=(X=1.4057,Y=-0.46 ...(truncated)"
    WeightMinimum=0
    Weight=&RigVMFunction_MathDoubleMul(
     B=&RigVMFunction_MathDoubleRemap(SourceMaximum=1, SourceMinimum=0, TargetMaximum=0, TargetMinimum=1, Value=@n3.Result).Result
     A=@n2.OutputParam
    ).Result
   )
   > n4 = &RigUnit_SphericalPoseReader(
    ActiveRegionScaleFactors=(PositiveWidth=0.3,NegativeWidth=0.3,PositiveHeight=0.6,NegativeHeight=1)
    ActiveRegionSize=0.15
    Debug=(bDrawDebug=false,bDraw2D=false,bDrawLocalAxes=true,DebugScale=25,DebugSegments=20,DebugThickness=0.25)
    DriverAxis=(X=1,Y=0,Z=0)
    DriverItem=(Type=Bone,Name="head")
    FalloffRegionScaleFactors=(PositiveWidth=0.8,NegativeWidth=0.4,PositiveHeight=1.3,NegativeHeight=1)
    FalloffSize=0.3
    FlipHeightScaling=false
    FlipWidthScaling=true
    OptionalParentItem=(Type=Bone,Name="spine_04")
    RotationOffset=(X=31.8313,Y=-30.3081,Z=-90)
   )
   > &RigUnit_ModifyTransforms(
    ItemToModify="((Item=(Type=Bone,Name=\"base_hairB_C_01\"),Transform=(Rotation=(X=-0.0101,Y=0.1388,Z=-0.0867,W=0.9865),Translation=(X=0,Y=0,Z=0),Scale3D=(X=1,Y=1,Z=1))),(Item=(Type=Bone,Name=\"base_hairB_L_01\"),Transform=(Rotation=(X=0.0014,Y=-0.0373,Z=0.0378,W=-0.9986),Tra ...(truncated)"
    WeightMinimum=0
    Weight=@n4.OutputParam
   )
   > n5 = &RigUnit_SphericalPoseReader(
    ActiveRegionScaleFactors=(PositiveWidth=1,NegativeWidth=2,PositiveHeight=1,NegativeHeight=1)
    ActiveRegionSize=0.05
    Debug=(bDrawDebug=false,bDraw2D=false,bDrawLocalAxes=true,DebugScale=25,DebugSegments=20,DebugThickness=0.25)
    DriverAxis=(X=-1,Y=0,Z=0)
    DriverItem=(Type=Bone,Name="upperarm_r")
    FalloffRegionScaleFactors=(PositiveWidth=1,NegativeWidth=1.6,PositiveHeight=1,NegativeHeight=1)
    FalloffSize=0.2
    FlipHeightScaling=false
    FlipWidthScaling=true
    OptionalParentItem=(Type=Bone,Name="spine_05")
    RotationOffset=(X=440.3224,Y=-158.8937,Z=212.3421)
   )
   > &RigUnit_ModifyTransforms(
    ItemToModify="((Item=(Type=Bone,Name=\"dyn_hairB_FR_02\"),Transform=(Rotation=(X=-0.0004,Y=0.0053,Z=0.0663,W=-0.9978),Translation=(X=-3.4822,Y=-0.3448,Z=-0.0373),Scale3D=(X=0.9337,Y=0.9337,Z=0.9337))),(Item=(Type=Bone,Name=\"base_hairB_R_01\"),Transform=(Rotation=(X=0.0841, ...(truncated)"
    WeightMinimum=0
    Weight=@n5.OutputParam
   )
   n6 = &RigUnit_SphericalPoseReader(
    ActiveRegionScaleFactors=(PositiveWidth=1,NegativeWidth=1.5,PositiveHeight=1.4,NegativeHeight=1)
    ActiveRegionSize=0.1
    Debug=(bDrawDebug=false,bDraw2D=false,bDrawLocalAxes=true,DebugScale=25,DebugSegments=20,DebugThickness=0.25)
    DriverAxis=(X=1,Y=0,Z=0)
    DriverItem=(Type=Bone,Name="upperarm_l")
    FalloffRegionScaleFactors=(PositiveWidth=1,NegativeWidth=1.3,PositiveHeight=1.8,NegativeHeight=1)
    FalloffSize=0.3
    FlipHeightScaling=false
    FlipWidthScaling=true
    OptionalParentItem=(Type=Bone,Name="spine_05")
    RotationOffset=(X=14.3661,Y=6.1444,Z=-0.4878)
    ExecutePin=@RigVMFunction_Sequence.C
   )
   n7 = RigVMVariableNode() {variable=Glider_CR}
   > &RigUnit_ModifyTransforms(
    ItemToModify="((Item=(Type=Bone,Name=\"dyn_hairB_FL_02\"),Transform=(Rotation=(X=0,Y=0,Z=0.2988,W=-0.9543),Translation=(X=0,Y=0,Z=0),Scale3D=(X=0.6825,Y=0.6825,Z=0.6825))),(Item=(Type=Bone,Name=\"base_hairB_L_01\"),Transform=(Rotation=(X=0,Y=0,Z=0.0416,W=-0.9991),Translatio ...(truncated)"
    WeightMinimum=0
    Weight=&RigVMFunction_MathDoubleMul(A=@n6.OutputParam, B=@n7.Value).Result
   )
   n8 = &RigVMFunction_AlphaInterp(
    InterpSpeedDecreasing=10
    InterpSpeedIncreasing=10
    Scale=1
    bClampResult=true
    Value=&RigVMFunction_MathBoolToFloat(Value=RigVMVariableNode() {variable=isEmoting_CR}.Value).Result
   )
   > &RigUnit_ModifyTransforms(
    ItemToModify=((Item=(Type=Bone,Name="dyn_hairB_FL_02"),Transform=(Rotation=(X=0,Y=-0.1511,Z=0,W=0.9885),Translation=(X=-1.9646,Y=-0.0001,Z=-0.5425),Scale3D=(X=0.8867,Y=0.8867,Z=0.8908))))
    WeightMinimum=0
    Weight=&RigVMFunction_MathDoubleMul(
     B=&RigVMFunction_MathDoubleRemap(SourceMaximum=1, SourceMinimum=0, TargetMaximum=0, TargetMinimum=1, Value=@n8.Result).Result
     A=@n6.OutputParam
    ).Result
   )
   > n9 = &RigUnit_SphericalPoseReader(
    ActiveRegionScaleFactors=(PositiveWidth=1,NegativeWidth=1.5,PositiveHeight=1.4,NegativeHeight=1)
    ActiveRegionSize=0.1
    Debug=(bDrawDebug=false,bDraw2D=false,bDrawLocalAxes=true,DebugScale=25,DebugSegments=20,DebugThickness=0.25)
    DriverAxis=(X=-1,Y=0,Z=0)
    DriverItem=(Type=Bone,Name="upperarm_r")
    FalloffRegionScaleFactors=(PositiveWidth=1,NegativeWidth=1,PositiveHeight=1,NegativeHeight=1)
    FalloffSize=0.3
    FlipHeightScaling=false
    FlipWidthScaling=true
    OptionalParentItem=(Type=Bone,Name="spine_05")
    RotationOffset=(X=171.5939,Y=-7.1189,Z=182.9834)
   )
   Multiply = &RigVMFunction_MathDoubleMul(A=@n9.OutputParam, B=@n7.Value)
   > &RigUnit_ModifyTransforms(
    ItemToModify="((Item=(Type=Bone,Name=\"dyn_hairB_FR_02\"),Transform=(Rotation=(X=0,Y=0,Z=0,W=1),Translation=(X=0,Y=0,Z=0),Scale3D=(X=0.8007,Y=0.8007,Z=0.8007))),(Item=(Type=Bone,Name=\"base_hairB_R_01\"),Transform=(Rotation=(X=0,Y=0,Z=-0.0981,W=-0.9952),Translation=(X=-0.97 ...(truncated)"
    WeightMinimum=0
    Weight=@Multiply.Result
   )
   RigVMFunction_MathBoolToFloat = &RigVMFunction_MathBoolToFloat(Value=RigVMVariableNode() {variable=isEmoting_CR}.Value)
   n10 = &RigVMFunction_AlphaInterp(InterpSpeedDecreasing=10, InterpSpeedIncreasing=10, Scale=1, bClampResult=true, Value=@RigVMFunction_MathBoolToFloat.Result)
   > &RigUnit_ModifyTransforms(
    ItemToModify="((Item=(Type=Bone,Name=\"dyn_hairB_FL_02\"),Transform=(Rotation=(X=0,Y=0,Z=0,W=1),Translation=(X=0,Y=-4.5235,Z=-1.8021),Scale3D=(X=1,Y=1,Z=1))),(Item=(Type=Bone,Name=\"dyn_hairB_FR_02\"),Transform=(Rotation=(X=0,Y=0,Z=0,W=1),Translation=(X=0.0002,Y=-2.0781,Z=0 ...(truncated)"
    WeightMinimum=0
    Weight=@n10.Result
   )
   n11 = &RigVMFunction_AlphaInterp(
    InterpSpeedDecreasing=10
    InterpSpeedIncreasing=10
    Scale=1
    bClampResult=true
    Value=RigVMVariableNode() {variable=CelebrationEmoteNoRBAN_CR}.Value
   )
   > &RigUnit_ModifyTransforms(
    ItemToModify="((Item=(Type=Bone,Name=\"base_hairB_R_01\"),Transform=(Rotation=(X=0.0448,Y=-0.0013,Z=0.029,W=0.9986),Translation=(X=-2.192,Y=-1.5889,Z=-0.0631),Scale3D=(X=1,Y=1,Z=1))),(Item=(Type=Bone,Name=\"dyn_hairB_FR_02\"),Transform=(Rotation=(X=0,Y=0,Z=0,W=-1),Translati ...(truncated)"
    WeightMinimum=0
    Weight=@n11.Result
   )
   RigVMFunctionReturnNode(ExecuteContext=@RigVMFunction_Sequence.D)
  ]}
 RigVMFunctionLibrary {nodes [
   ConstraintsBackpackNoRBAN = RigVMCollapseNode(ExecuteContext=())
   Delta Rotation Y = RigVMCollapseNode(ExecuteContext=())
   Delta Rotation Z = RigVMCollapseNode(ExecuteContext=())
   Hair = RigVMCollapseNode(ExecuteContext=())
  ]}
 RigVMFunction_Sequence_1_ContainedGraph {nodes [
   n1 = RigVMFunctionEntryNode()
   RigVMFunction_Sequence = &RigVMFunction_Sequence(ExecuteContext=@n1.ExecuteContext)
   n2 = &RigVMFunction_Sequence(ExecuteContext=@RigVMFunction_Sequence.B)
   n3 = &RigVMFunction_Sequence(ExecuteContext=@n2.B)
   n4 = &RigVMFunction_Sequence(ExecuteContext=@n3.B)
   RigVMFunctionReturnNode(A=@RigVMFunction_Sequence.A, B=@n2.A, C=@n3.A, D=@n4.A, E=@n4.B)
  ]}
 RigVMFunction_Sequence_1_ContainedGraph {nodes [
   n1 = RigVMFunctionEntryNode()
   RigVMFunction_Sequence = &RigVMFunction_Sequence(ExecuteContext=@n1.ExecuteContext)
   n2 = &RigVMFunction_Sequence(ExecuteContext=@RigVMFunction_Sequence.B)
   n3 = &RigVMFunction_Sequence(ExecuteContext=@n2.B)
   RigVMFunctionReturnNode(A=@RigVMFunction_Sequence.A, B=@n2.A, C=@n3.A, D=@n3.B)
  ]}
 RigVMModel {nodes [
   RigUnit_BeginExecution = &RigUnit_BeginExecution()
   RigVMFunction_MathBoolToFloat = &RigVMFunction_MathBoolToFloat(Value=RigVMVariableNode() {variable=Crouching_CR}.Value)
   n1 = &RigVMFunction_AlphaInterp(InterpSpeedDecreasing=10, InterpSpeedIncreasing=10, Scale=1, bClampResult=true, Value=@RigVMFunction_MathBoolToFloat.Result)
   n2 = RigVMVariableNode(Value=@n1.Result, ExecuteContext=@RigUnit_BeginExecution.ExecutePin) {variable=Crouching}
   RigVMFunction_ControlFlowBranch = &RigVMFunction_ControlFlowBranch(Condition=RigVMVariableNode() {variable=WearingBackblingAndNoRBAN}.Value, ExecuteContext=@n2.ExecuteContext)
   Hair = RigVMFunctionReferenceNode(ExecuteContext=@RigVMFunction_ControlFlowBranch.False)
   ConstraintsBackpackNoRBAN = RigVMFunctionReferenceNode(ExecuteContext=@RigVMFunction_ControlFlowBranch.True)
  ]}
]
hierarchy [
 root {loc=<0 0 0>, rot=<0 0 0 1>}
 pelvis {parent=root, loc=<0 -0.5301 98.0571>, rot=<0.0127 0.707 -0.0127 -0.707>}
 spine_01 {parent=pelvis, loc=<6.1208 -0.9328 0>, rot=<0 0 0.034 -0.9994>}
 spine_02 {parent=spine_01, loc=<6.5997 -0.2095 0>, rot=<0 0 0.0238 -0.9997>}
 spine_03 {parent=spine_02, loc=<6.4914 -0.1237 0>, rot=<0 0 0.0149 -0.9999>}
 spine_04 {parent=spine_03, loc=<6.9598 0.0794 0>, rot=<0 0 -0.0132 -0.9999>}
 spine_05 {parent=spine_04, loc=<6.9987 0 0>, rot=<0 0 -0.0015 -1>}
 neck_01 {parent=spine_05, loc=<11.3138 0.1529 0>, rot=<0 0 -0.2003 -0.9797>}
 neck_02 {parent=neck_01, loc=<4.5627 0 0>, rot=<0 0 0.0504 -0.9987>}
 head {parent=neck_02, loc=<5.3914 0 0>, rot=<0 0 0.115 -0.9934>}
 base_hairB_R_01 {parent=head, loc=<7.1721 -5.5783 6.1865>, rot=<0.0805 -0.9968 0 0>}
 dyn_hairB_R_01 {parent=base_hairB_R_01, loc=<0 0 0>, rot=<0 0 0 1>}
 dyn_hairB_R_02 {parent=dyn_hairB_R_01, loc=<8.2749 0.6765 -0.0053>, rot=<0 0 0.0734 -0.9973>}
 dyn_hairB_R_03 {parent=dyn_hairB_R_02, loc=<8.1762 -0.6758 -0.64>, rot=<0 0 0.0344 -0.9994>}
 dyn_hairB_R_05 {parent=dyn_hairB_R_03, loc=<0.949 -2.717 -1.2856>, rot=<-0.0946 -0.0242 -0.1684 -0.9809>}
 dyn_hairB_R_06 {parent=dyn_hairB_R_03, loc=<9.3749 -1.6434 -0.2587>, rot=<-0.1456 -0.0276 -0.184 -0.9717>}
 dyn_hairB_R_04 {parent=dyn_hairB_R_02, loc=<3.6793 -0.1444 -1.5672>, rot=<-0.4648 -0.0552 -0.1246 -0.8748>}
 dyn_hairB_FR_02 {parent=dyn_hairB_R_01, loc=<7.0988 6.8944 -3.135>, rot=<-0.2202 -0.0637 0.0824 -0.9699>}
 dyn_hairB_FR_03 {parent=dyn_hairB_FR_02, loc=<9.9537 -5.0566 -0.2808>, rot=<-0.172 0.0037 -0.1491 -0.9737>}
 base_hairB_C_01 {parent=head, loc=<7.1957 -8.1033 0.3833>, rot=<0.061 -0.9981 0 0>}
 dyn_hairB_C_01 {parent=base_hairB_C_01, loc=<0 0 0>, rot=<0 0 0 1>}
 dyn_hairB_C_02 {parent=dyn_hairB_C_01, loc=<8.2212 0 0>, rot=<0 0 0.0828 -0.9966>}
 dyn_hairB_C_03 {parent=dyn_hairB_C_02, loc=<8.2594 0 0>, rot=<0 0 0.0338 -0.9994>}
 dyn_hairB_C_04 {parent=dyn_hairB_C_03, loc=<10.4733 -0.1569 1.3682>, rot=<-0.1088 0.0001 -0.0558 -0.9925>}
 dyn_hairB_C_05 {parent=dyn_hairB_C_03, loc=<10.1474 0.8654 -3.4404>, rot=<-0.2265 -0.049 -0.1308 -0.964>}
 base_hairB_L_01 {parent=head, loc=<7.3192 -5.295 -5.9044>, rot=<0.0956 -0.9954 0 0>}
 dyn_hairB_L_01 {parent=base_hairB_L_01, loc=<0 0 0>, rot=<0 0 0 1>}
 dyn_hairB_L_02 {parent=dyn_hairB_L_01, loc=<8.2168 0 0>, rot=<0 0 0.0901 -0.9959>}
 dyn_hairB_L_03 {parent=dyn_hairB_L_02, loc=<8.4053 0 0.5102>, rot=<0 0 0.0288 -0.9996>}
 dyn_hairB_L_04 {parent=dyn_hairB_L_03, loc=<3.3572 -0.8351 6.2035>, rot=<0.4908 0.1339 -0.1547 -0.8469>}
 dyn_hairB_L_05 {parent=dyn_hairB_L_03, loc=<9.2944 0.0626 0.8203>, rot=<-0.1033 0.0169 -0.2039 -0.9734>}
 dyn_hairB_FL_02 {parent=dyn_hairB_L_01, loc=<7.2583 7.825 3.1661>, rot=<0.3257 0.0795 0.0009 -0.9421>}
 dyn_hairB_FL_03 {parent=dyn_hairB_FL_02, loc=<9.0709 -6.0239 -0.5256>, rot=<-0.007 -0.028 0.0033 -0.9996>}
 dyn_bangs_04 {parent=head, loc=<6.4238 4.1999 -7.3229>, rot=<-0.0027 0.6453 0.7639 -0.0032>}
 dyn_bangs_03 {parent=head, loc=<8.9212 11.1363 -3.889>, rot=<-0.0011 0.2712 0.9625 -0.004>}
 dyn_bangs_02 {parent=head, loc=<7.1984 12.5959 2.2453>, rot=<0.0005 -0.1167 0.9932 -0.0041>}
 dyn_bangs_01 {parent=head, loc=<7.5751 9.9394 5.451>, rot=<0.0011 -0.2591 0.9658 -0.004>}
 dyn_earring_r {parent=head, loc=<1.5088 -0.0026 6.447>, rot=<0.001 -0.2425 0.9702 -0.004>}
 dyn_earring_l {parent=head, loc=<1.5088 -0.0026 -6.4468>, rot=<-0.001 0.2382 0.9712 -0.004>}
 clavicle_l {parent=spine_05, loc=<5.4831 -0.1053 -3.2946>, rot=<0.0465 -0.7289 -0.009 -0.6829>}
 upperarm_l {parent=clavicle_l, loc=<9.8658 -1.2136 -0.4105>, rot=<0.0164 -0.3245 -0.0269 -0.9454>}
 lowerarm_l {parent=upperarm_l, loc=<26.465 -0.8345 0.259>, rot=<0.0036 0.0161 -0.2298 -0.9731>}
 clavicle_r {parent=spine_05, loc=<5.4831 -0.1053 3.2946>, rot=<0.7289 0.0465 -0.6829 0.009>}
 upperarm_r {parent=clavicle_r, loc=<-9.8658 1.2136 0.4105>, rot=<0.0164 -0.3245 -0.0269 -0.9454>}
 lowerarm_r {parent=upperarm_r, loc=<-26.465 0.8345 -0.259>, rot=<0.0036 0.0161 -0.2298 -0.9731>}
]
variables [
 CelebrationEmoteNoRBAN_CR {type=double, default=0}
 CelebrationEmoteRBAN_CR {type=double, default=0}
 Crouching {type=double, default=0}
 Crouching_CR {type=bool, default=False}
 FallingSwim_CR {type=double, default=0}
 Floating_CR {type=bool, default=False}
 Glider_CR {type=double, default=0}
 Moving2D_CR {type=bool, default=False}
 OnGroundAlpha {type=double, default=0}
 Pistol_CR {type=double, default=0}
 Skydiving_CR {type=bool, default=False}
 WearingBackblingAndNoRBAN {type=bool, default=False}
 isAttackingPickaxe {type=bool, default=False}
 isDBNO_CR {type=bool, default=False}
 isEmoting_CR {type=bool, default=False}
 isFrontEnd_CR {type=bool, default=False}
 isOnRail_CR {type=bool, default=False}
]
