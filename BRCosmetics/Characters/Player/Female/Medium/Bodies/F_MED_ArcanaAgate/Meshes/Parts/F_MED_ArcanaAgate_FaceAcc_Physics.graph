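# Physics asset for the F_MED_ArcanaAgate female character's FaceAcc part, defining physics for three back-hair bone chains (center, left, right).
# Each chain links three dyn_hairB capsule bodies in sequence from a tiny base_hairB anchor.
# Constraints lock twist and limit swing asymmetrically, much wider on swing1 than swing2.
# Capsules on head, neck, upper spine, clavicles and upper arms have no constraints, forming body shapes around the hair.
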PhysicsAsset /BRCosmetics/Characters/Player/Female/Medium/Bodies/F_MED_ArcanaAgate/Meshes/Parts/F_MED_ArcanaAgate_FaceAcc_Physics
bodies {
 base_hairB_C_01 [{shapeType=Capsule, length=0.505, radius=0.505, shapeName=base_hairB_C_01_capsule, rotation=<rot 0 -90 0>}]
 base_hairB_L_01 [{shapeType=Capsule, length=0.505, radius=0.505, shapeName=base_hairB_L_01_capsule, rotation=<rot 0 -90 0>}]
 base_hairB_R_01 [{shapeType=Capsule, length=0.505, radius=0.505, shapeName=base_hairB_R_01_capsule, rotation=<rot 0 -89.9995 -0.5517>}]
 clavicle_l [{shapeType=Capsule, length=10.6, radius=7, shapeName=clavicle_l_capsule, center=<6.2526 -0.0283 -2.2573>, rotation=<rot 0 -90 90>}]
 clavicle_r [
  {shapeType=Capsule, length=10.6, radius=5, shapeName=clavicle_l_capsule, center=<-6.2526 -0.1462 2.2573>, rotation=<rot 0 -90 90>}
  {shapeType=Capsule, length=7.3475, radius=4.2, shapeName=clavicle_r_capsule, center=<-6.1037 -0.3004 -2.9779>, rotation=<rot 90 0 0>}
 ]
 dyn_hairB_C_01 [{shapeType=Capsule, length=6.0314, radius=1, shapeName=dyn_hairB_C_01_capsule, center=<3.9112 0 0>, rotation=<rot 90 0 0>}]
 dyn_hairB_C_02 [{shapeType=Capsule, length=3.1828, radius=2.5094, shapeName=dyn_hairB_C_02_capsule, center=<3.855 0 0>, rotation=<rot 90 0 0>}]
 dyn_hairB_C_03 [
  {shapeType=Capsule, length=3.4505, radius=2.5764, shapeName=dyn_hairB_C_03_capsule, center=<4.6603 0.8837 0>, rotation=<rot 90 10.4957 0>}
  {shapeType=Capsule, length=5.0019, radius=3.5372, shapeName=dyn_hairB_C_03_capsule, center=<8.8917 0.6005 0>}
 ]
 dyn_hairB_L_01 [{shapeType=Capsule, length=6.1739, radius=1, shapeName=dyn_hairB_L_01_capsule, center=<3.7963 0 0>, rotation=<rot 90 0 0>}]
 dyn_hairB_L_02 [{shapeType=Capsule, length=4.5647, radius=2.5402, shapeName=dyn_hairB_L_02_capsule, center=<4.1798 -1.1066 0>, rotation=<rot 90 0 0>}]
 dyn_hairB_L_03 [{shapeType=Capsule, length=3.4023, radius=3.2924, shapeName=dyn_hairB_L_03_capsule, center=<3.4327 0 1.1568>, rotation=<rot -0.1388 0.001 0>}]
 dyn_hairB_R_01 [{shapeType=Capsule, length=6.5288, radius=1, shapeName=dyn_hairB_R_01_capsule, center=<3.6783 0 0>, rotation=<rot 90 0 0>}]
 dyn_hairB_R_02 [{shapeType=Capsule, length=4, radius=3, shapeName=dyn_hairB_R_02_capsule, center=<3.899 -0.1481 0>, rotation=<rot 90 0 0>}]
 dyn_hairB_R_03 [{
   shapeType=Capsule
   length=4.4769
   radius=3.6838
   shapeName=dyn_hairB_R_03_capsule
   center=<2.732 -0.9085 0>
   rotation=<rot 0.5212 -179.983 179.9996>
  }]
 head [{shapeType=Capsule, length=2.6403, radius=7.3597, center=<2.5678 1.1032 0>, rotation=<rot 90 0 0>}]
 neck_01 [{shapeType=Capsule, length=12.5, radius=6, shapeName=neck_01_capsule, center=<1.6583 0.949 0>}]
 spine_02 [{shapeType=Capsule, length=16.7, radius=9.3, shapeName=spine_02_capsule, center=<6.0567 1.2153 0>}]
 spine_03 [{shapeType=Capsule, length=16.7, radius=9.3, shapeName=spine_02_capsule, center=<6.0567 1.1414 0>}]
 spine_04 [
  {shapeType=Capsule, length=17.8, radius=9.1, shapeName=spine_04_capsule, center=<4.0135 1.6435 0>, rotation=<rot 0 -90 0>}
  {shapeType=Capsule, length=18.5, radius=9, shapeName=spine_04_capsule, center=<8.6672 0.6315 0>}
 ]
 spine_05 [{shapeType=Capsule, length=17.2, radius=8, shapeName=spine_05_capsule, center=<2.8857 0.3837 0>, rotation=<rot 0 -90 0>}]
 upperarm_l [{shapeType=Capsule, length=13.5146, radius=5.3875, shapeName=upperarm_l_capsule, center=<11.2314 -0.3627 0>, rotation=<rot 90 0 0>}]
 upperarm_r [{shapeType=Capsule, length=13.5146, radius=5.3875, shapeName=upperarm_r_capsule, center=<-11.2314 0.3627 0>, rotation=<rot 90 180 0>}]
}
constraints [
 {bone1Name=dyn_hairB_C_01, bone2Name=base_hairB_C_01, swing1LimitDegrees=25, swing2LimitDegrees=8, twistLimitDegrees=1, twistMotion=Locked}
 {bone1Name=dyn_hairB_C_02, bone2Name=dyn_hairB_C_01, swing1LimitDegrees=30, swing2LimitDegrees=7, twistLimitDegrees=1, twistMotion=Locked}
 {bone1Name=dyn_hairB_C_03, bone2Name=dyn_hairB_C_02, swing1LimitDegrees=40, swing2LimitDegrees=10, twistLimitDegrees=1, twistMotion=Locked}
 {bone1Name=dyn_hairB_L_01, bone2Name=base_hairB_L_01, swing1LimitDegrees=25, swing2LimitDegrees=5, twistLimitDegrees=1, twistMotion=Locked}
 {bone1Name=dyn_hairB_L_02, bone2Name=dyn_hairB_L_01, swing1LimitDegrees=20, swing2LimitDegrees=4, twistLimitDegrees=1, twistMotion=Locked}
 {bone1Name=dyn_hairB_L_03, bone2Name=dyn_hairB_L_02, swing1LimitDegrees=30, swing2LimitDegrees=7, twistLimitDegrees=1, twistMotion=Locked}
 {bone1Name=dyn_hairB_R_01, bone2Name=base_hairB_R_01, swing1LimitDegrees=15, swing2LimitDegrees=8, twistLimitDegrees=1, twistMotion=Locked}
 {bone1Name=dyn_hairB_R_02, bone2Name=dyn_hairB_R_01, swing1LimitDegrees=20, swing2LimitDegrees=7, twistLimitDegrees=1, twistMotion=Locked}
 {bone1Name=dyn_hairB_R_03, bone2Name=dyn_hairB_R_02, swing1LimitDegrees=25, swing2LimitDegrees=10, twistLimitDegrees=1, twistMotion=Locked}
]
