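# Post-physics AnimBP for the ArcanaAgate female face accessory part, driving secondary motion on hair and bangs.
# Parent and body-mesh poses feed a Control Rig driven by player state, then head-relative RigidBody setups for front end, DBNO, skydiving or normal play.
# Backbling, capes or disabled rigid bodies select the rig-only pose; Anim Dynamics finishes bangs and hair bones.
# The event graph updates state values and a parameter-asset gravity override.
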AnimBlueprint /BRCosmetics/Characters/Player/Female/Medium/Bodies/F_MED_ArcanaAgate/Meshes/Parts/F_MED_ArcanaAgate_FaceAcc_AnimBP
target_skeleton=&/F_MED_ArcanaAgate_FaceAcc_Skeleton
graphs [
 AnimGraph {nodes [
   n1 = Use cached pose 'SetupNormalCache'()
   n2 = Use cached pose 'SetupSkydiving'()
   n3 = Get bIsSkyDiving()
   n4 = Blend Poses by bool(
    BlendPose_0=Use cached pose 'FrontEndCache'()
    BlendPose_1=Blend Poses by bool(
     BlendPose_0=Use cached pose 'DBNOCache'()
     BlendPose_1=Blend Poses by bool(BlendPose_0=@n2, BlendPose_1=@n1, bActiveValue=@n3) {blendType=Linear}
     bActiveValue=Get bIsDBNO()
    ) {blendType=Linear}
    bActiveValue=Get bIsInFrontEnd()
   ) {blendType=Linear}
   n5 = Use cached pose 'ControlRig Cache'()
   n6 = Get isWearingBackbling / Cape / NoRBAN() -> (isWearingBackbling / Cape / NoRBAN:bool)
   n7 = "Transform (Modify) Bone - Bone: dyn_hairB_C_04"!(
    ComponentPose=Local To Component(LocalPose=Blend Poses by bool(BlendPose_0=@n5, BlendPose_1=@n4, bActiveValue=@n6) {blendType=Linear})
    bAlphaBoolEnabled:bool=True
    Translation="0,2,0"
    Rotation="0.294,-12.334,0.154"
    Scale="1,1,1"
   ) -> (Pose) {
    alphaInputType=Bool
    rotationMode=BMM_Additive
    rotationSpace=BCS_BoneSpace
    scaleMode=BMM_Additive
    scaleSpace=BCS_BoneSpace
    translationMode=BMM_Additive
    translationSpace=BCS_BoneSpace
    alphaBoolBlend {blendInTime=0.2, blendOption=Linear, blendOutTime=0.3}
    alphaScaleBiasClamp {bInterpResult=true, interpSpeedIncreasing=20}
    boneToModify {boneName=dyn_hairB_C_04}
    rotation=<rot 0.294 -12.334 0.154>
    translation {y=2}
   }
   n8 = "Transform (Modify) Bone - Bone: dyn_hairB_C_05"!(ComponentPose=@n7, bAlphaBoolEnabled:bool=True, Translation="0,0,0", Rotation="0.170,-40.614,0.223", Scale="1,1,1") -> (Pose) {
    alphaInputType=Bool
    rotationMode=BMM_Additive
    rotationSpace=BCS_BoneSpace
    scaleMode=BMM_Additive
    scaleSpace=BCS_BoneSpace
    translationMode=BMM_Additive
    translationSpace=BCS_BoneSpace
    alphaBoolBlend {blendInTime=0.2, blendOption=Linear, blendOutTime=0.3}
    alphaScaleBiasClamp {bInterpResult=true, interpSpeedIncreasing=20}
    boneToModify {boneName=dyn_hairB_C_05}
    rotation=<rot 0.17 -40.614 0.223>
   }
   n9 = Anim Dynamics - Bone: dyn_bangs_03(ComponentPose=@n8, Alpha=1, ExternalForce="0,0,0") {
    angularDampingOverride=1
    angularSpringConstant=20000
    bAngularSpring=true
    bOverrideAngularDamping=true
    bUseGravityOverride=true
    gravityScale=0.35
    simulationSpace=BoneRelative
    alphaBoolBlend {blendOption=Linear}
    boundBone {boneName=dyn_bangs_03}
    componentAppliedLinearAccClamp=<1000 1000 1000>
    componentLinearAccScale=<0.3 0.3 0.3>
    componentLinearVelScale=<1 0.4 0.4>
    gravityOverride {z=-980}
    physicsBodyDefinitions [{
      boundBone {boneName=dyn_bangs_03}
      boxExtents=<20 30 30>
      constraintSetup {angularLimitsMax=<0 30 0>, angularLimitsMin=<0 -30 0>, angularTarget=<1 0 0>}
      localJointOffset=<2 0 0>
     }]
    relativeSpaceBone {boneName=root}
   }
   n10 = Anim Dynamics - Bone: dyn_bangs_02(ComponentPose=@n9, Alpha=1, ExternalForce="0,0,0") {
    angularDampingOverride=1
    angularSpringConstant=20000
    bAngularSpring=true
    bOverrideAngularDamping=true
    bUseGravityOverride=true
    gravityScale=0.35
    simulationSpace=BoneRelative
    alphaBoolBlend {blendOption=Linear}
    boundBone {boneName=dyn_bangs_02}
    componentAppliedLinearAccClamp=<1000 1000 1000>
    componentLinearAccScale=<0.3 0.3 0.3>
    componentLinearVelScale=<1 0.4 0.4>
    gravityOverride {z=-980}
    physicsBodyDefinitions [{
      boundBone {boneName=dyn_bangs_02}
      boxExtents=<30 35 35>
      constraintSetup {angularLimitsMax=<0 30 0>, angularLimitsMin=<0 -30 0>, angularTarget=<1 0 0>}
      localJointOffset=<2 0 0>
     }]
    relativeSpaceBone {boneName=root}
   }
   n11 = Anim Dynamics - Bone: dyn_bangs_01(ComponentPose=@n10, Alpha=1, ExternalForce="0,0,0") {
    angularDampingOverride=1
    angularSpringConstant=20000
    bAngularSpring=true
    bOverrideAngularDamping=true
    bUseGravityOverride=true
    gravityScale=0.35
    simulationSpace=BoneRelative
    alphaBoolBlend {blendOption=Linear}
    boundBone {boneName=dyn_bangs_01}
    componentAppliedLinearAccClamp=<1000 1000 1000>
    componentLinearAccScale=<0.3 0.3 0.3>
    componentLinearVelScale=<1 0.4 0.4>
    gravityOverride {z=-980}
    physicsBodyDefinitions [{
      boundBone {boneName=dyn_bangs_01}
      boxExtents=<30 35 35>
      constraintSetup {angularLimitsMax=<0 30 0>, angularLimitsMin=<0 -30 0>, angularTarget=<1 0 0>}
      localJointOffset=<2 0 0>
     }]
    relativeSpaceBone {boneName=root}
   }
   n12 = Anim Dynamics - Bone: dyn_hairB_C_04!(ComponentPose=@n11, bAlphaBoolEnabled:bool=True, ExternalForce="0,0,0") -> (Pose) {
    alphaInputType=Bool
    angularDampingOverride=1
    angularSpringConstant=20000
    bAngularSpring=true
    bOverrideAngularDamping=true
    bUseGravityOverride=true
    gravityScale=0.35
    simulationSpace=BoneRelative
    alphaBoolBlend {blendOption=Linear}
    boundBone {boneName=dyn_hairB_C_04}
    componentAppliedLinearAccClamp=<2000 2000 2000>
    componentLinearAccScale=<0.3 0.3 0.3>
    componentLinearVelScale=<0.4 0.4 0.4>
    gravityOverride {z=-980}
    physicsBodyDefinitions [{
      boundBone {boneName=dyn_hairB_C_04}
      boxExtents=<20 20 20>
      constraintSetup {angularLimitsMax=<0 30 0>, angularLimitsMin=<0 -30 0>, angularTarget=<1 0 0>}
      localJointOffset=<2 0 0>
     }]
    relativeSpaceBone {boneName=root}
   }
   n13 = Anim Dynamics - Bone: dyn_hairB_C_05!(ComponentPose=@n12, bAlphaBoolEnabled:bool=True, ExternalForce="0,0,0") -> (Pose) {
    alphaInputType=Bool
    angularDampingOverride=1
    angularSpringConstant=20000
    bAngularSpring=true
    bOverrideAngularDamping=true
    bUseGravityOverride=true
    gravityScale=0.35
    simulationSpace=BoneRelative
    alphaBoolBlend {blendOption=Linear}
    boundBone {boneName=dyn_hairB_C_05}
    componentAppliedLinearAccClamp=<2000 2000 2000>
    componentLinearAccScale=<0.3 0.3 0.3>
    componentLinearVelScale=<0.4 0.4 0.4>
    gravityOverride {z=-980}
    physicsBodyDefinitions [{
      boundBone {boneName=dyn_hairB_C_05}
      boxExtents=<20 20 20>
      constraintSetup {angularLimitsMax=<0 30 0>, angularLimitsMin=<0 -30 0>, angularTarget=<1 0 0>}
      localJointOffset=<2 0 0>
     }]
    relativeSpaceBone {boneName=root}
   }
   n14 = Anim Dynamics - Bone: dyn_hairB_R_06(ComponentPose=@n13, Alpha=1, ExternalForce="0,0,0") {
    angularDampingOverride=1
    angularSpringConstant=20000
    bAngularSpring=true
    bOverrideAngularDamping=true
    bUseGravityOverride=true
    gravityScale=0.35
    simulationSpace=BoneRelative
    alphaBoolBlend {blendOption=Linear}
    boundBone {boneName=dyn_hairB_R_06}
    componentAppliedLinearAccClamp=<2000 2000 2000>
    componentLinearAccScale=<0.3 0.3 0.3>
    componentLinearVelScale=<0.4 0.4 0.4>
    gravityOverride {z=-980}
    physicsBodyDefinitions [{
      boundBone {boneName=dyn_hairB_R_06}
      boxExtents=<15 15 15>
      constraintSetup {angularLimitsMax=<0 30 0>, angularLimitsMin=<0 -30 0>, angularTarget=<1 0 0>}
      localJointOffset=<2 0 0>
     }]
    relativeSpaceBone {boneName=root}
   }
   n15 = Anim Dynamics - Bone: dyn_hairB_L_04(ComponentPose=@n14, Alpha=1, ExternalForce="0,0,0") {
    angularDampingOverride=1
    angularSpringConstant=20000
    bAngularSpring=true
    bOverrideAngularDamping=true
    bUseGravityOverride=true
    gravityScale=0.35
    simulationSpace=BoneRelative
    alphaBoolBlend {blendOption=Linear}
    boundBone {boneName=dyn_hairB_L_04}
    componentAppliedLinearAccClamp=<2000 2000 2000>
    componentLinearAccScale=<0.3 0.3 0.3>
    componentLinearVelScale=<0.4 0.4 0.4>
    gravityOverride {z=-980}
    physicsBodyDefinitions [{
      boundBone {boneName=dyn_hairB_L_04}
      boxExtents=<15 15 15>
      constraintSetup {angularLimitsMax=<0 30 0>, angularLimitsMin=<0 -30 0>, angularTarget=<1 0 0>}
      localJointOffset=<2 0 0>
     }]
    relativeSpaceBone {boneName=root}
   }
   n16 = Anim Dynamics - Bone: dyn_hairB_FL_03(ComponentPose=@n15, Alpha=1, ExternalForce="0,0,0") {
    angularDampingOverride=1
    angularSpringConstant=20000
    bAngularSpring=true
    bOverrideAngularDamping=true
    bUseGravityOverride=true
    gravityScale=0.35
    simulationSpace=BoneRelative
    alphaBoolBlend {blendOption=Linear}
    boundBone {boneName=dyn_hairB_FL_03}
    componentAppliedLinearAccClamp=<2000 2000 2000>
    componentLinearAccScale=<0.3 0.3 0.3>
    componentLinearVelScale=<0.4 0.4 0.4>
    gravityOverride {z=-980}
    physicsBodyDefinitions [{
      boundBone {boneName=dyn_hairB_FL_03}
      boxExtents=<20 20 20>
      constraintSetup {angularLimitsMax=<0 0 30>, angularLimitsMin=<0 0 -30>, angularTarget=<1 0 0>}
      localJointOffset=<2 0 0>
     }]
    relativeSpaceBone {boneName=root}
   }
   n17 = Anim Dynamics - Bone: dyn_hairB_FR_03(ComponentPose=@n16, Alpha=1, ExternalForce="0,0,0") {
    angularDampingOverride=1
    angularSpringConstant=20000
    bAngularSpring=true
    bOverrideAngularDamping=true
    bUseGravityOverride=true
    gravityScale=0.35
    simulationSpace=BoneRelative
    alphaBoolBlend {blendOption=Linear}
    boundBone {boneName=dyn_hairB_FR_03}
    componentAppliedLinearAccClamp=<2000 2000 2000>
    componentLinearAccScale=<0.3 0.3 0.3>
    componentLinearVelScale=<0.4 0.4 0.4>
    gravityOverride {z=-980}
    physicsBodyDefinitions [{
      boundBone {boneName=dyn_hairB_FR_03}
      boxExtents=<20 20 20>
      constraintSetup {angularLimitsMax=<0 0 30>, angularLimitsMin=<0 0 -30>, angularTarget=<1 0 0>}
      localJointOffset=<2 0 0>
     }]
    relativeSpaceBone {boneName=root}
   }
   Output Pose(Result=Component To Local(ComponentPose=@n17)) {layerGroup=None}
   n18 = Layered blend per bone(
    BasePose=Copy Pose From Mesh() {bUseAttachedParent=true}
    BlendPoses_0=Copy Pose From Mesh(SourceMeshComponent=Get Body_Mesh()) {bUseAttachedParent=true}
   ) {layerSetup [{branchFilters [{boneName=clavicle_l}, {boneName=clavicle_r}, {boneName=upperarm_r}, {boneName=upperarm_l}]}]}
   n19 = Control Rig!(
    Source=@n18
    Alpha
    Crouching_CR:bool=Get bIsCrouching()
    Pistol_CR:double=Get Pistol() -> (Pistol:double)
    Floating_CR:bool=Get bIsSkydiveFloating()
    Moving2D_CR:bool=Get bIsMoving2D()
    Glider_CR:double=Get HangGlidingAlpha() -> (HangGlidingAlpha:double)
    FallingSwim_CR:double=Get FallingSwimming() -> (FallingSwimming:double)
    WearingBackblingAndNoRBAN:bool=Get isWearingBackbling / Cape / NoRBAN()
    Skydiving_CR:bool=Get bIsSkyDiving()
    CelebrationEmoteRBAN_CR:double=Get PlayingCelebrationRban() -> (PlayingCelebrationRban:double)
    CelebrationEmoteNoRBAN_CR:double=Get PlayingCelebrationNoRban() -> (PlayingCelebrationNoRban:double)
    isAttackingPickaxe:bool=Get bIsPlayingMeleeAnim()
    isEmoting_CR:bool=Get bIsPlayingEmote()
    isDBNO_CR:bool=Get bIsDBNO()
    isFrontEnd_CR:bool=Get bIsInFrontEnd()
    isOnRail_CR:bool=Get bIsOnGrindRail()
   ) -> (Pose) {
    alphaBoolBlend {blendOption=Linear}
    controlRigAssetReference {blueprintRigClass=&/F_MED_ArcanaAgate_FaceAcc_CtrlRig.F_MED_ArcanaAgate_FaceAcc_CtrlRig_C}
   }
   Save cached pose 'ControlRig Cache'(Pose=@n19) {cacheName=ControlRig Cache}
   n20 = RigidBody!(
    ComponentPose=Local To Component(LocalPose=Use cached pose 'ControlRig Cache'())
    Alpha
    OverrideWorldGravity:Vector=Get Gravity Override()
    SimSpaceSettings:Sim Space Settings Structure=(WorldAlpha=0.3,VelocityScaleZ=1,DampingAlpha=1,MaxLinearVelocity=700,MaxAngularVelocity=15,MaxLinearAcceleration=300,MaxAngularAcceleration=1000,ExternalLinearDragV=(X=0,Y=0,Z=0),ExternalLinearVelocity=(X=0,Y=0,Z=0),ExternalAngularVelocity=(X=0,Y=0,Z=0))
   ) -> (Pose) {
    bClampLinearTranslationLimitToRefPose=true
    bOverrideWorldGravity=true
    overridePhysicsAsset=&/F_MED_ArcanaAgate_FaceAcc_Skydive_Physics
    simulationSpace=BaseBoneSpace
    alphaBoolBlend {blendOption=Linear}
    baseBoneRef {boneName=head}
    componentAppliedLinearAccClamp=<3000 3000 3000>
    simSpaceSettings {maxAngularAcceleration=1000, maxAngularVelocity=15, maxLinearAcceleration=300, maxLinearVelocity=700, worldAlpha=0.3}
   }
   n21 = "Transform (Modify) Bone - Bone: base_hairB_R_01"(ComponentPose=@n20, Alpha=1, Translation="-2.489,-3.743,-2.937", Rotation="-0.847,-8.421,13.046", Scale="1,1,1") {
    rotationMode=BMM_Additive
    rotationSpace=BCS_BoneSpace
    scaleMode=BMM_Additive
    scaleSpace=BCS_BoneSpace
    translationMode=BMM_Additive
    translationSpace=BCS_BoneSpace
    alphaBoolBlend {blendOption=Linear}
    alphaScaleBiasClamp {bInterpResult=true, interpSpeedDecreasing=3}
    boneToModify {boneName=base_hairB_R_01}
    rotation=<rot -0.847 -8.421 13.046>
    translation=<-2.489 -3.743 -2.937>
   }
   n22 = "Transform (Modify) Bone - Bone: base_hairB_L_01"(ComponentPose=@n21, Alpha=1, Translation="0,0,0", Rotation="1.541,-16.343,-8.127", Scale="1,1,1") {
    rotationMode=BMM_Additive
    rotationSpace=BCS_BoneSpace
    scaleMode=BMM_Additive
    scaleSpace=BCS_BoneSpace
    translationMode=BMM_Additive
    translationSpace=BCS_BoneSpace
    alphaBoolBlend {blendOption=Linear}
    alphaScaleBiasClamp {bInterpResult=true, interpSpeedDecreasing=3}
    boneToModify {boneName=base_hairB_L_01}
    rotation=<rot 1.541 -16.343 -8.127>
   }
   Save cached pose 'SetupSkydiving'(Pose=Component To Local(ComponentPose=@n22)) {cacheName=SetupSkydiving}
   n23 = RigidBody(
    ComponentPose=Local To Component(LocalPose=Use cached pose 'ControlRig Cache'())
    OverrideWorldGravity=Get Gravity Override()
    SimSpaceSettings=(WorldAlpha=0.3,VelocityScaleZ=1,DampingAlpha=1,MaxLinearVelocity=700,MaxAngularVelocity=15,MaxLinearAcceleration=300,MaxAngularAcceleration=1000,ExternalLinearDragV=(X=0,Y=0,Z=0),ExternalLinearVelocity=(X=0,Y=0,Z=0),ExternalAngularVelocity=(X=0,Y=0,Z=0))
   ) {
    bClampLinearTranslationLimitToRefPose=true
    bOverrideWorldGravity=true
    overridePhysicsAsset=&/F_MED_ArcanaAgate_FaceAcc_Physics
    simulationSpace=BaseBoneSpace
    alphaBoolBlend {blendOption=Linear}
    baseBoneRef {boneName=head}
    componentAppliedLinearAccClamp=<3000 3000 3000>
    simSpaceSettings {maxAngularAcceleration=1000, maxAngularVelocity=15, maxLinearAcceleration=300, maxLinearVelocity=700, worldAlpha=0.3}
   }
   n24 = "Transform (Modify) Bone - Bone: base_hairB_C_01"(ComponentPose=@n23, Alpha=1, Translation="0,0,0", Rotation="0,-25,0", Scale="1,1,1") {
    rotationMode=BMM_Additive
    rotationSpace=BCS_BoneSpace
    scaleMode=BMM_Additive
    scaleSpace=BCS_BoneSpace
    translationMode=BMM_Additive
    translationSpace=BCS_BoneSpace
    alphaBoolBlend {blendOption=Linear}
    alphaScaleBiasClamp {bInterpResult=true, interpSpeedIncreasing=2}
    boneToModify {boneName=base_hairB_C_01}
    rotation {yaw=-25}
   }
   n25 = "Transform (Modify) Bone - Bone: base_hairB_L_01"(ComponentPose=@n24, Alpha=1, Translation="0,0,0", Rotation="0,-25,0", Scale="1,1,1") {
    rotationMode=BMM_Additive
    rotationSpace=BCS_BoneSpace
    scaleMode=BMM_Additive
    scaleSpace=BCS_BoneSpace
    translationMode=BMM_Additive
    translationSpace=BCS_BoneSpace
    alphaBoolBlend {blendOption=Linear}
    alphaScaleBiasClamp {bInterpResult=true, interpSpeedIncreasing=2}
    boneToModify {boneName=base_hairB_L_01}
    rotation {yaw=-25}
   }
   n26 = "Transform (Modify) Bone - Bone: base_hairB_R_01"(ComponentPose=@n25, Alpha=1, Translation="0,0,0", Rotation="0,-25,0", Scale="1,1,1") {
    rotationMode=BMM_Additive
    rotationSpace=BCS_BoneSpace
    scaleMode=BMM_Additive
    scaleSpace=BCS_BoneSpace
    translationMode=BMM_Additive
    translationSpace=BCS_BoneSpace
    alphaBoolBlend {blendOption=Linear}
    alphaScaleBiasClamp {bInterpResult=true, interpSpeedIncreasing=2}
    boneToModify {boneName=base_hairB_R_01}
    rotation {yaw=-25}
   }
   n27 = "Transform (Modify) Bone - Bone: dyn_hairB_C_03"(ComponentPose=@n26, Alpha=1, Translation="0,-1,0", Rotation="0,-80,0", Scale="1,1,1") {
    rotationMode=BMM_Additive
    rotationSpace=BCS_BoneSpace
    scaleMode=BMM_Additive
    scaleSpace=BCS_BoneSpace
    translationMode=BMM_Additive
    translationSpace=BCS_BoneSpace
    alphaBoolBlend {blendOption=Linear}
    alphaScaleBiasClamp {bInterpResult=true, interpSpeedIncreasing=1}
    boneToModify {boneName=dyn_hairB_C_03}
    rotation {yaw=-80}
    translation {y=-1}
   }
   n28 = "Transform (Modify) Bone - Bone: dyn_hairB_L_03"(ComponentPose=@n27, Alpha=1, Translation="0,-1,0", Rotation="0,-80,0", Scale="1,1,1") {
    rotationMode=BMM_Additive
    rotationSpace=BCS_BoneSpace
    scaleMode=BMM_Additive
    scaleSpace=BCS_BoneSpace
    translationMode=BMM_Additive
    translationSpace=BCS_BoneSpace
    alphaBoolBlend {blendOption=Linear}
    alphaScaleBiasClamp {bInterpResult=true, interpSpeedIncreasing=1}
    boneToModify {boneName=dyn_hairB_L_03}
    rotation {yaw=-80}
    translation {y=-1}
   }
   n29 = "Transform (Modify) Bone - Bone: dyn_hairB_R_03"(ComponentPose=@n28, Alpha=1, Translation="0,-1,0", Rotation="0,-80,0", Scale="1,1,1") {
    rotationMode=BMM_Additive
    rotationSpace=BCS_BoneSpace
    scaleMode=BMM_Additive
    scaleSpace=BCS_BoneSpace
    translationMode=BMM_Additive
    translationSpace=BCS_BoneSpace
    alphaBoolBlend {blendOption=Linear}
    alphaScaleBiasClamp {bInterpResult=true, interpSpeedIncreasing=1}
    boneToModify {boneName=dyn_hairB_R_03}
    rotation {yaw=-80}
    translation {y=-1}
   }
   Save cached pose 'SetupNormalCache'(Pose=Component To Local(ComponentPose=@n29)) {cacheName=SetupNormalCache}
   n30 = RigidBody(
    ComponentPose=Local To Component(LocalPose=Use cached pose 'ControlRig Cache'())
    OverrideWorldGravity=Get Gravity Override()
    SimSpaceSettings=(WorldAlpha=0.6,VelocityScaleZ=1,DampingAlpha=1,MaxLinearVelocity=700,MaxAngularVelocity=15,MaxLinearAcceleration=300,MaxAngularAcceleration=50,ExternalLinearDragV=(X=1,Y=0,Z=1),ExternalLinearVelocity=(X=0,Y=0,Z=0),ExternalAngularVelocity=(X=0,Y=0,Z=0))
   ) {
    bClampLinearTranslationLimitToRefPose=true
    bOverrideWorldGravity=true
    overridePhysicsAsset=&F_MED_ArcanaAgate_FaceAcc_Skydive_Physics
    simulationSpace=BaseBoneSpace
    alphaBoolBlend {blendOption=Linear}
    baseBoneRef {boneName=head}
    componentAppliedLinearAccClamp=<3000 3000 3000>
    simSpaceSettings {
     maxAngularAcceleration=50
     maxAngularVelocity=15
     maxLinearAcceleration=300
     maxLinearVelocity=700
     worldAlpha=0.6
     externalLinearDragV {x=1, z=1}
    }
   }
   Save cached pose 'DBNOCache'(Pose=Component To Local(ComponentPose=@n30)) {cacheName=DBNOCache}
   n31 = RigidBody!(ComponentPose=Local To Component(LocalPose=Use cached pose 'ControlRig Cache'()), Alpha, OverrideWorldGravity=Get Gravity Override()) -> (Pose) {
    bClampLinearTranslationLimitToRefPose=true
    bOverrideWorldGravity=true
    overridePhysicsAsset=&F_MED_ArcanaAgate_FaceAcc_Physics
    simulationSpace=BaseBoneSpace
    alphaBoolBlend {blendOption=Linear}
    baseBoneRef {boneName=head}
    componentAppliedLinearAccClamp=<3000 3000 3000>
    simSpaceSettings {maxAngularAcceleration=1000, maxAngularVelocity=15, maxLinearAcceleration=300, maxLinearVelocity=700, worldAlpha=0.6}
   }
   n32 = "Transform (Modify) Bone - Bone: dyn_hairB_L_03"!(ComponentPose=@n31, bAlphaBoolEnabled:bool=True, Translation="0,0,0", Rotation="27.640,0.033,-0.005", Scale="1,1,1") -> (Pose) {
    alphaInputType=Bool
    rotationMode=BMM_Additive
    rotationSpace=BCS_BoneSpace
    scaleMode=BMM_Additive
    scaleSpace=BCS_BoneSpace
    translationMode=BMM_Additive
    translationSpace=BCS_BoneSpace
    alphaBoolBlend {blendOption=Linear}
    alphaScaleBiasClamp {bInterpResult=true, interpSpeedIncreasing=1}
    boneToModify {boneName=dyn_hairB_L_03}
    rotation=<rot 27.64 0.033 -0.005>
   }
   n33 = "Transform (Modify) Bone - Bone: dyn_hairB_L_03"(ComponentPose=@n32, bAlphaBoolEnabled=True, Translation="0,2,0", Rotation="0,0,0", Scale="1,1,1") {
    alphaInputType=Bool
    rotationMode=BMM_Additive
    rotationSpace=BCS_BoneSpace
    scaleMode=BMM_Additive
    scaleSpace=BCS_BoneSpace
    translationMode=BMM_Additive
    translationSpace=BCS_BoneSpace
    alphaBoolBlend {blendOption=Linear}
    alphaScaleBiasClamp {bInterpResult=true, interpSpeedIncreasing=1}
    boneToModify {boneName=dyn_hairB_L_03}
    translation {y=2}
   }
   n34 = "Transform (Modify) Bone - Bone: dyn_hairB_C_03"!(ComponentPose=@n33, bAlphaBoolEnabled:bool=True, Translation="0,3,0", Rotation="0,3,0", Scale="1,1,1") -> (Pose) {
    alphaInputType=Bool
    rotationMode=BMM_Additive
    rotationSpace=BCS_BoneSpace
    scaleMode=BMM_Additive
    scaleSpace=BCS_BoneSpace
    translationMode=BMM_Additive
    translationSpace=BCS_BoneSpace
    alphaBoolBlend {blendOption=Linear}
    alphaScaleBiasClamp {bInterpResult=true, interpSpeedIncreasing=1}
    boneToModify {boneName=dyn_hairB_C_03}
    rotation {yaw=3}
    translation {y=3}
   }
   n35 = "Transform (Modify) Bone - Bone: dyn_hairB_R_03"!(ComponentPose=@n34, bAlphaBoolEnabled:bool=True, Translation="0,0,0", Rotation="0,26.674,0", Scale="1,1,1") -> (Pose) {
    alphaInputType=Bool
    rotationMode=BMM_Additive
    rotationSpace=BCS_BoneSpace
    scaleMode=BMM_Additive
    scaleSpace=BCS_BoneSpace
    translationMode=BMM_Additive
    translationSpace=BCS_BoneSpace
    alphaBoolBlend {blendOption=Linear}
    alphaScaleBiasClamp {bInterpResult=true, interpSpeedIncreasing=1}
    boneToModify {boneName=dyn_hairB_R_03}
    rotation {yaw=26.674}
   }
   Save cached pose 'FrontEndCache'(Pose=Component To Local(ComponentPose=@n35)) {cacheName=FrontEndCache}
  ]}
 EventGraph {nodes [
   Event OnCharacterCustomizationLoaded()
   > n1 = Cast To PlayerPawn_Athena(Object=TryGetPawnOwner())
   > Set Body_Mesh(Body_Mesh=GetSkeletalMeshForPartType(self=@n1))
   Event BlueprintUpdateAnimation()
   > n2 = Sequence!(execute) -> (then_0, then_1, then_2:Exec, then_3:Exec, then_4:Exec)
   > n3 = "Gravity Override State Function (Parameter Asset) (Nativized)"(
    CharacterPartAnimInstance=Self-Reference()
    GameTimeInSeconds=GetGameTimeInSeconds()
    DebugJointName=spine_05
    Parameters=&/M_MED_ArcanaAgate_FaceAcc_Parameters
   )
   > Set Gravity Override(Gravity Override=@n3)
   > Set RBAN Settings Out(RBAN Settings Out=@n3.OutRigidBodyAnimNodeParameters)
   n4 = NOT Boolean(A=Get bIsPlayingEmote())
   n5 = NOT Boolean(A=Get bIsSkyDiving())
   n6 = Get bIsFalling()
   Set FallingGround(execute=@n2.then_1, FallingGround:double="To Float (Boolean)"(InBool=AND Boolean!(A=@n6, B=@n5, C:bool=@n4) -> (ReturnValue))) -> (Output_Get:double)
   n7 = Get bIsMoving2D()
   n8 = Get bIsCrouching()
   > Set Moving2dCrouch(Moving2dCrouch:double="To Float (Boolean)"(InBool=AND Boolean!(A=@n7, B=@n8) -> (ReturnValue))) -> (Output_Get:double)
   n9 = NOT Boolean(A=Get bIsPlayingEmote())
   n10 = NOT Boolean(A=Get bIsOnGround() -> (bIsOnGround:bool))
   n11 = NOT Boolean(A=Get bIsSkyDiving())
   n12 = Get bIsFalling()
   > Set FallingSwimming(FallingSwimming:double="To Float (Boolean)"(InBool=AND Boolean!(A=@n12, B=@n11, C:bool=@n9, D:bool=@n10) -> (ReturnValue))) -> (Output_Get:double)
   n13 = Montage_IsPlaying(self=GetAnimInstance(self=GetOwnerSkeletalMeshComponent()), Montage=/BRCosmetics/Animation/Game/MainPlayer/Emotes/Emote_Celebration_Female_M)
   n14 = Get isWearingBackbling / Cape / NoRBAN()
   > Set PlayingCelebrationNoRban(PlayingCelebrationNoRban:double="To Float (Boolean)"(InBool=AND Boolean!(A=@n13, B=@n14) -> (ReturnValue))) -> (Output_Get:double)
   n15 = NOT Boolean(A=Get isWearingBackbling / Cape / NoRBAN())
   > Set PlayingCelebrationRban(PlayingCelebrationRban:double="To Float (Boolean)"(InBool=AND Boolean(A=@n13, B=@n15))) -> (Output_Get:double)
   n16 = Sequence!(execute=@n2.then_3) -> (then_0, then_1)
   n17 = GetOwnerSkeletalMeshComponent()
   n18 = BreakVector(InVec=BreakTransform(InTransform=GetDeltaTransformFromRefPose(self=@n17, BoneName=pelvis, BaseName=root)))
   n19 = Get bIsSkydiveFloating()
   > n20 = Branch(Condition=AND Boolean(A="float < float"(A=@n18.Z, B=-7), B=@n19))
   > Set Alpha_RightFloating(Alpha_RightFloating:double=1) -> (Output_Get:double)
   Set Alpha_RightFloating(execute=@n20.else, Alpha_RightFloating=0)
   n21 = Branch(execute=@n16.then_1, Condition=AND Boolean(A=@n19, B="float > float"(A=@n18.Z, B=7)))
   > Set Alpha_LeftFloating(Alpha_LeftFloating:double=1) -> (Output_Get:double)
   Set Alpha_LeftFloating(execute=@n21.else, Alpha_LeftFloating=0)
   n22 = Get bRigidBodyEnabled()
   n23 = CheckForBackbling()
   Set isWearingBackbling / Cape / NoRBAN(
    execute=@n2.then_2
    isWearingBackbling / Cape / NoRBAN:bool=OR Boolean!(A=@n23, B=@n23.bIsBackblingACape, C:bool=@n23.bIsBackblingLong, D:bool=NOT Boolean(A=@n22)) -> (ReturnValue)
   ) -> (Output_Get:bool)
   > Set isSwingingTwoHanded(isSwingingTwoHanded:bool=AND Boolean(A=Get bIsPlayingMeleeAnim(), B=Get TwoHandedPickaxe() -> (TwoHandedPickaxe:bool))) -> (Output_Get:bool)
   > Set notPlayingEmote(notPlayingEmote:bool=NOT Boolean(A=Get bIsPlayingEmote())) -> (Output_Get:bool)
   Event OnParachuteOpened()
   > Switch on EGliderType(Selection=Get PlayerGliderType() -> (PlayerGliderType=HangGlider))
   > Set HangGlidingAlpha(HangGlidingAlpha:double=1) -> (Output_Get:double)
   Event OnParachuteClosed()
   > Set HangGlidingAlpha(HangGlidingAlpha=0)
   Event BlueprintInitializeAnimation()
   > Set Skel Mesh(Skel Mesh=GetOwnerSkeletalMeshComponent())
   Event OnWeaponChanged()
   > n24 = Cast To FortPlayerAnimInstance(Object=GetAnimInstance(self=GetOwnerSkeletalMeshComponent()))
   > Set Pistol(Pistol:double=0) -> (Output_Get:double)
   > Set TwoHandedPickaxe(TwoHandedPickaxe:bool=false) -> (Output_Get:bool)
   > n25 = Switch on EFortWeaponCoreAnimation!(execute, Selection:EFortWeaponCoreAnimation Enum=Get WeaponCoreAnim(self=@n24)) -> (
    Melee:Exec
    Pistol:Exec
    Shotgun:Exec
    PaperBlueprint:Exec
    Rifle:Exec
    MeleeOneHand:Exec
    MachinePistol:Exec
    RocketLauncher:Exec
    GrenadeLauncher:Exec
    GoingCommando:Exec
    AssaultRifle:Exec
    TacticalShotgun:Exec
    SniperRifle:Exec
    TrapPlacement:Exec
    ShoulderLauncher:Exec
    AbilityDecoTool:Exec
    Crossbow:Exec
    C4:Exec
    RemoteControl:Exec
    DualWield:Exec
    AR_BullPup:Exec
    AR_ForwardGrip:Exec
    MedPackPaddles:Exec
    SMG_P90:Exec
    AR_DrumGun:Exec
    Consumable_Small:Exec
    Consumable_Large:Exec
    Balloon:Exec
    MountedTurret:Exec
    CreativeTool:Exec
    ExplosiveBow:Exec
    AshtonIndigo:Exec
    AshtonChicago:Exec
    MeleeDualWield:Exec
    Unarmed:Exec
   )
   > Set TwoHandedPickaxe(TwoHandedPickaxe=true)
   Set Pistol(execute=@n25.Pistol, Pistol=1)
   Get bRigidBodyEnabled()
   Get bRigidBodyEnabled()
  ]}
]
variables [
 Alpha_LeftFloating {type=number}
 Alpha_RightFloating {type=number}
 Body_Mesh {type=SkeletalMeshComponent}
 FallingGround {type=number}
 FallingSwimming {type=number}
 Gravity Override {type=Vector}
 HangGlidingAlpha {type=number}
 isSwingingTwoHanded {type=boolean}
 isWearingBackbling / Cape / NoRBAN {type=boolean}
 Moving2dCrouch {type=number}
 notPlayingEmote {type=boolean}
 Pistol {type=number}
 PlayingCelebrationNoRban {type=number}
 PlayingCelebrationRban {type=number}
 RBAN Settings Out {type=FortRigidBodyAnimNodeParameters}
 Skel Mesh {type=SkeletalMeshComponent}
 TwoHandedPickaxe {type=boolean}
]
